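# Translucent UI material for a backing shape, named for Creative objective positioning, that blends between a ringed circle and a slightly skewed square via CircleSquare.
# Dense exponential radial gradients form the circle mask and an inner disc that separates RingColor from BackingColor.
# The square mask multiplies four sharpened, clamped UV edge ramps, each turned by a Rotator with constant Time, using DotProduct as a scalar multiply.
# CircleSquare also fades the ring color into solid BackingColor.
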Material /Game/Athena/HUD/LTM/Playground/Materials/M_UI_CreativeObjectivePositioning_Backing
graph {
 properties {blendMode=BLEND_Translucent, materialDomain=MD_UI, pixelDepthOffsetMode=PDOM_Legacy, usedShadingModels=Default Lit}
 nodes [
  n1 = Clamp(None=MaterialFunctionCall(UVs=TextureCoordinate(), CenterPosition=Constant2Vector() {g=0.5, r=0.5}, Radius=Constant() {r=0.495}, Density=Constant() {r=32}) {materialFunction=&RadialGradientExponential}.RadialGradientExponential)
  n2 = Clamp(None=MaterialFunctionCall(UVs=TextureCoordinate(), CenterPosition=Constant2Vector() {g=0.5, r=0.5}, Radius=Constant() {r=0.425}, Density=Constant() {r=32}) {materialFunction=&RadialGradientExponential}.RadialGradientExponential)
  DotProduct(A=@n1, B=OneMinus(None=@n2))
  n3 = Add(A=Rotator(Coordinate=TextureCoordinate(), Time=Constant() {r=-0.255}) {speed=6.2832}, B=Constant() {r=-0.05})
  n4 = Clamp(None=Multiply(A=ComponentMask(None=@n3) {r=true}) {constB=42})
  n5 = Add(A=Rotator(Coordinate=TextureCoordinate(), Time=Constant() {r=0.2535}) {speed=6.2832}, B=Constant() {r=-0.05})
  n6 = Clamp(None=Multiply(A=ComponentMask(None=@n5) {r=true}) {constB=42})
  n7 = Add(A=Rotator(Coordinate=TextureCoordinate(), Time=Constant() {r=0.01}) {speed=6.2832}, B=Constant() {r=-0.05})
  n8 = Clamp(None=Multiply(A=ComponentMask(None=@n7) {r=true}) {constB=42})
  n9 = Add(A=Rotator(Coordinate=TextureCoordinate(), Time=Constant() {r=0.4925}) {speed=6.2832}, B=Constant() {r=-0.05})
  n10 = Clamp(None=Multiply(A=ComponentMask(None=@n9) {r=true}) {constB=42})
  n11 = ScalarParameter() {parameterName=CircleSquare}
  LinearInterpolate(A=@n1, B=DotProduct(A=DotProduct(A=@n4, B=@n6), B=DotProduct(A=@n8, B=@n10)), Alpha=@n11)
  n12 = VectorParameter() {parameterName=BackingColor, defaultValue=<rgba 0 0.521 1 1>}
  LinearInterpolate(
   A=LinearInterpolate(A=VectorParameter() {parameterName=RingColor, defaultValue=<rgba 0 0.2159 1 1>}.RGB, B=@n12.RGB, Alpha=@n2)
   B=@n12.RGB
   Alpha=@n11
  )
 ]
}
