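# UI-domain translucent material for a Creative objective positioning icon in the Playground HUD.
# An Icon texture parameter, sampled with texture coordinates and defaulting to the objective attack icon, supplies the image.
# An IconColor vector parameter provides a color, light cyan by default, so instances can swap both icon and color.
# The graph shows no math nodes or connections to the material outputs.
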Material /Game/Athena/HUD/LTM/Playground/Materials/M_UI_CreativeObjective_Positioning_Icon
graph {
 properties {blendMode=BLEND_Translucent, materialDomain=MD_UI, pixelDepthOffsetMode=PDOM_Legacy, usedShadingModels=Default Lit}
 nodes [
  TextureSampleParameter2D(UVs=TextureCoordinate()) {parameterName=Icon, texture=&/T_UI_CreativeObjectiveAttack}
  VectorParameter() {parameterName=IconColor, defaultValue=<rgba 0.5 1 1 1>}
 ]
}
