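# UI material for the backing of the Playground objective numeric HUD view, drawing a two-color progress fill.
# The horizontal UV is offset by PercentFill minus one, then ceiled and clamped into a hard-edged mask without any texture.
# That mask lerps from BackingColor to FillColor, so the fill grows from the high-U (right) edge as PercentFill rises.
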Material /Game/Athena/HUD/LTM/Playground/Materials/M_UI_HUD_ObjectiveNumericView_Backing
graph {
 properties {blendMode=BLEND_Translucent, materialDomain=MD_UI, pixelDepthOffsetMode=PDOM_Legacy, usedShadingModels=Default Lit}
 nodes [
  Constant() {r=1}
  n1 = TextureCoordinate()
  n2 = Add(A=Add(A=ComponentMask(None=@n1) {r=true}, B=Constant() {r=-1}), B=ScalarParameter() {defaultValue=0.666, parameterName=PercentFill})
  LinearInterpolate(
   A=VectorParameter() {parameterName=BackingColor, defaultValue=<rgba 0.004 0.0232 0.2159 1>}.RGB
   B=VectorParameter() {parameterName=FillColor, defaultValue=<rgba 0 0.2159 1 1>}.RGB
   Alpha=Clamp(None=Ceil(None=@n2))
  )
 ]
}
