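# Translucent UI material in the Playground LTM HUD folder that draws a sheen band for an outline.
# DrawLine-2D creates a soft horizontal line mask spanning past both UV edges, with parameters for height (YValue), width and edge falloff.
# The mask blends ColorA to ColorB in RGB and separately lifts a scalar from partial to full, likely for opacity.
# It has no time node, so the band moves only when YValue changes.
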Material /Game/Athena/HUD/LTM/Playground/Materials/M_UI_OutlineSheen_V
graph {
 properties {blendMode=BLEND_Translucent, materialDomain=MD_UI, pixelDepthOffsetMode=PDOM_Legacy, usedShadingModels=Default Lit}
 nodes [
  Constant() {r=1}
  n1 = ScalarParameter() {defaultValue=0.5, parameterName=YValue}
  n2 = MaterialFunctionCall!(
   Coordinates=TextureCoordinate()
   P0=AppendVector(A=Constant() {r=-0.5}, B=@n1)
   P1=AppendVector(A=Constant() {r=1.5}, B=@n1)
   Line Width=ScalarParameter() {defaultValue=0.05, parameterName=LineWidth}
   Edge Falloff=ScalarParameter() {defaultValue=0.5, parameterName=Edge}
   Square Corners?
  ) {materialFunction=/Engine/Functions/Engine_MaterialFunctions01/Coordinates/DrawLine-2D}
  LinearInterpolate(
   A=VectorParameter() {parameterName=ColorA, defaultValue=<rgba 0.1529 0.6157 0.6745 0.4275>}.RGB
   B=VectorParameter() {parameterName=ColorB, defaultValue=<rgba 0 0.7451 1 1>}.RGB
   Alpha=@n2.Result
  )
  LinearInterpolate(A=Constant() {r=0.47}, B=Constant() {r=1}, Alpha=@n2.Result)
 ]
}
