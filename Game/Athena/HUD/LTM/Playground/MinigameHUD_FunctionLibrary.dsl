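; Blueprint function library for minigame HUDs that updates visibility of the game phase HUD element, desktop and mobile tags, in Playground modes.
; UpdateGamePhaseVisibility loops local gameplay controllers outside Creative Royale, setting a "Minigame"-context visibility flag from Creative/VKPlay playlist checks and ShouldUseGamePhaseTimer.
; That helper tests whether the player's relevant safe zone interface is not Battle Royale game phase logic.
; Creative Royale is detected by playlist context tag.
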(fn IsCreativeRoyalePlaylist
  (params (InGameStateAthena :type Fort%20Game%20State%20Athena%20Object%20Reference))
  ;[0.003181 1 0 0.5] Check the playlist tags to see if this is a Creative Royale playlist.
  (Utilities|IsValid InGameStateAthena
    (:Is%20Valid
      (return (GameplayTags|HasTag (Class|FortGameStateAthena|GetAthenaPlaylistContextTags InGameStateAthena) "(TagName=\"Athena.Playlist.CreativeRoyale\")")))
    (:Is%20Not%20Valid
      (Development|PrintString 0 "MinigameHUD_FunctionLibrary::IsCreativeRoyalePlaylist Error: InGameStateAthena is NOT valid." true true "(R=1.000000,G=0.269408,B=0.000000,A=1.000000)" 1e+01)
      (return false))))

(fn ShouldUseGamePhaseTimer
  (params (Controller :type Controller%20Object%20Reference))
  (bind _playerstate (Variables|/Script/Engine.Controller|GetPlayerState Controller)
    (:then
      (return (not (.bSuccess (Utilities|Casting|CastToFortGameStateComponent_BattleRoyaleGamePhaseLogic (SafeZoneInterface|GetRelevantSafeZoneInterface _playerstate))))))
    (:else
      (return false))))

(fn UpdateGamePhaseVisibility
  (params (WorldContextObject :type Object%20Reference))
  (bind _worldcontextobject (Variables|GetWorldContextObject))
  (bind _returnvalue (Game|GetGameState _worldcontextobject))
  (bind (_asfort_game_state_athena _bsuccess) (Utilities|Casting|CastToFortGameStateAthena _returnvalue))
  (bind _worldcontextobject_2 (Variables|GetWorldContextObject))
  (Utilities|IsValid WorldContextObject
    (:Is%20Valid
      (for _array_element (Fortnite|GetLocalFortPlayerControllers true)
        (if (Math|Boolean|ANDBoolean (.bSuccess (Utilities|Casting|CastToFortPlayerControllerGameplay _array_element)) _bsuccess (not (CallFunction|IsCreativeRoyalePlaylist _asfort_game_state_athena _worldcontextobject_2)))
          (HUD|SetHUDElementVisibilitywithContext (Utilities|Casting|CastToFortPlayerControllerGameplay _array_element) (GameplayTags|MakeLiteralGameplayTagContainer "(GameplayTags=((TagName=\"MobileHUD.Stats.GamePhase\"),(TagName=\"HUD.Athena.GamePhase\")))") (and (or (Class|FortGameStateAthena|IsAnyCreativePlaylist _asfort_game_state_athena) (Class|FortGameStateAthena|IsVKPlayPlaylist _asfort_game_state_athena)) (not (CallFunction|ShouldUseGamePhaseTimer _array_element _worldcontextobject_2))) "Minigame" ""))))))
