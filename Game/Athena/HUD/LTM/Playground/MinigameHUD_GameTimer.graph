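# Minigame HUD round-timer widget showing an icon and M:SS text, counting down or up per the current minigame.
# OnDisplayMinigameInProgressUI shows it outside curated hubs when counting up or time-limited, starting a one-second looping update, otherwise collapsing it and clearing that timer.
# Each tick refreshes the text and plays warning cues at 30 and 15 seconds left.
# HideStormTimer ignores its ShouldHide input, only calling UpdateGamePhaseVisibility; Play Timer Sound immediately re-arms both DoOnce gates.
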WidgetBlueprint /Game/Athena/HUD/LTM/Playground/MinigameHUD_GameTimer
graphs [
 Check Timer {nodes [
   n1 = Check Timer!() -> (then, Get Time:double=~) {is=title K2Node_FunctionEntry}
   n2 = Truncate(A=@n1)
   n3 = "Equal (Integer)"(A=@n2, B=15) {is=title K2Node_CallFunction}
   > Return Node!(execute, Is Valid Time Up:bool=OR Boolean(A="Equal (Integer)"(A=@n2, B=30), B=@n3), Is 15 Seconds:bool=@n3)
  ]}
 EventGraph {nodes [
   n1 = OnUpdateTime() {is=title K2Node_CustomEvent}
   > UpdateTime()
   > n2 = Branch(Condition=Get bTimerCountsDown!(self:Fort Minigame Object Reference=Get CurrentMinigame() -> (CurrentMinigame:Fort Minigame Object Reference)) -> (bTimerCountsDown:bool))
   > n3 = Check Timer!(execute, self, Get Time:double=GetTimeRemaining!(self:Fort Minigame Object Reference=Get CurrentMinigame()) -> (ReturnValue:float)) -> (then, Is Valid Time Up:bool, Is 15 Seconds:bool) {is=title K2Node_CallFunction}
   > Branch(Condition=@n3)
   > Do Once(Reset=@n6.then_0)
   > Play Timer Sound!(execute, self, Play 15 Seconds Sound:bool=@n3.Is 15 Seconds) -> (then)
   n4 = Get CurrentMinigame()
   n5 = Check Timer(
    execute=@n2.else
    Get Time=float - float(
     A=Get TimeLimit!(self:Fort Minigame Object Reference=@n4) -> (TimeLimit:float)
     B=GetRuntime!(self:Fort Minigame Object Reference=@n4) -> (ReturnValue:float)
    ) {is=title K2Node_CallFunction}
   )
   > Branch(Condition=@n5)
   > Do Once(Reset=@n6.then_1)
   > Play Timer Sound(Play 15 Seconds Sound=@n5.Is 15 Seconds)
   Reset Time Sound() {is=title K2Node_CustomEvent}
   > n6 = Sequence()
   n7 = Event OnDisplayMinigameInProgressUI!() -> (OutputDelegate, then, bDisplay:bool) {is=title K2Node_Event}
   n8 = Get CurrentMinigame()
   > n9 = Is Valid(InputObject=@n8)
   > n10 = Branch(Condition=NOT Boolean(A=IsCuratedHub!(self:Fort Minigame Object Reference=@n8) -> (ReturnValue:bool)))
   n11 = Get TimeLimit(self=@n8)
   n12 = Get bTimerCountsDown(self=@n8)
   n13 = AND Boolean(
    A=@n7.bDisplay
    B=OR Boolean(A="Not Equal (Float)"!(A:double=@n11, B:double=0) -> (ReturnValue:bool) {is=title K2Node_CallFunction}, B=NOT Boolean(A=@n12))
   )
   > HideStormTimer!(execute, self, ShouldHide:bool=@n13) -> (then)
   > n14 = Branch(Condition=@n13)
   > SetVisibility(InVisibility=HitTestInvisible)
   > UpdateTime()
   > n15 = Set Timer by Event(Delegate=@n1, Time=1, bLooping=true)
   > Set TimerHandle(TimerHandle:Timer Handle Structure=@n15) -> (Output_Get:Timer Handle Structure)
   SetVisibility(execute=[@n14.else, @n10.else, @n9.Is Not Valid], InVisibility=Collapsed)
   > Clear and Invalidate Timer by Handle(Handle=Get TimerHandle() -> (TimerHandle:Timer Handle Structure))
   > Reset Time Sound() {is=title K2Node_CallFunction}
  ]}
 HideStormTimer {nodes [
   HideStormTimer!() -> (then, ShouldHide:bool=~) {is=title K2Node_FunctionEntry}
   > UpdateGamePhaseVisibility!(execute, WorldContextObject:Object Reference=Self-Reference()) -> (then)
  ]}
 MakeTime {nodes [
   n1 = Round(A=Inputs!() -> (InSeconds:double=~)) {is=title K2Node_CallFunction}
   n2 = "% (Integer)"(A=@n1, B=60) -> (ReturnValue=0)
   Outputs!(Result:Text=Format Text!(
     Format="LOCTEXT(\"{MM}:{SS}\")"
     MM:int=int / int!(A:int=@n1, B:int=60) -> (ReturnValue:int=0) {is=title K2Node_CallFunction}
     SS:Text="To Text (Integer)"(Value=@n2, MinimumIntegralDigits=2, MaximumIntegralDigits=2)
    ) -> (Result))
  ]}
 Play Timer Sound {nodes [
   n1 = Play Timer Sound!() -> (then, Play 15 Seconds Sound:bool=~) {is=title K2Node_FunctionEntry}
   > PlaySound2D(Sound=Select(
     Option 0:Sound Base Object Reference=/Game/Sounds/Creative/Modes/Match/Match_TimeWarning_01_Cue
     Option 1:Sound Base Object Reference=/Game/Sounds/Creative/Modes/Match/Match_TimeWarning_02_Cue
     Index=@n1
    ) -> (ReturnValue:Sound Base Object Reference))
   > Reset Time Sound()
  ]}
 UpdateTime {nodes [
   UpdateTime() {is=title K2Node_FunctionEntry}
   n1 = Get CurrentMinigame()
   > Is Valid(InputObject=@n1)
   n2 = GetTimeRemaining(self=@n1)
   n3 = GetRuntime(self=@n1)
   n4 = Get bTimerCountsDown(self=@n1)
   > "SetText (Text)"(
    self=Get TimeText() -> (TimeText:Text Object Reference)
    InText=Make Time!(InSeconds:double=Select(Option 0:double=@n3, Option 1:double=@n2, Index=@n4) -> (ReturnValue:double)) -> (Result:Text) {is=title K2Node_MacroInstance}
   )
  ]}
]
variables [TimerHandle {type=TimerHandle}]
widgets {
 info {
  inheritedWidgetCount=1
  namedSlotCount=2
  parentClass=/Script/FortniteUI.MinigameWidgetBase
  rootWidgetClass=&HorizontalBox
  widgetCount=6
 }
 widgets [
  HorizontalBox_0 {widgetClassPath=&HorizontalBox, widget=HorizontalBox_0}
  FortHUDElementWrapper_Icon {
   widgetClassPath=/Game/Athena/HUD/FortHUDElementWrapper
   bIsVariable=true
   parent=HorizontalBox_0
   widget=FortHUDElementWrapper_Icon
   properties {
    elementTag {tagName=HUD.Athena.Player.Timer.Icon}
    hUDElementTag {gameplayTags [{tagName=Athena.HideForCapture}, {tagName=Athena.Creative.HUDController.HideRoundTimer}]}
   }
   slot {class=HorizontalBoxSlot, properties {horizontalAlignment=HAlign_Center, verticalAlignment=VAlign_Center}}
  }
  Icon {
   widgetClassPath=&Image
   bIsVariable=true
   namedSlotHost=FortHUDElementWrapper_Icon
   widget=Icon
   properties {brush {
     bIsDynamicallyLoaded=false
     drawAs=Image
     imageType=NoImage
     mirroring=NoMirror
     resourceName=None
     resourceObject=/Game/Athena/HUD/Phase/MI_GameRoundTimer
     tiling=NoTile
     imageSize=<36 36>
     margin=<ltrb 0 0 0 0>
     outlineSettings {
      bUseBrushTransparency=false
      roundingType=HalfHeightRadius
      width=0
      color {colorUseRule=UseColor_Specified, specifiedColor=<rgba 0 0 0 0>}
      cornerRadii=<0 0 0 0>
     }
     tintColor {colorUseRule=UseColor_Specified, specifiedColor=<rgba 1 1 1 1>}
     uVRegion {bIsValid=false, max=<0 0>, min=<0 0>}
    }}
  }
  FortHUDElementWrapper_Time {
   widgetClassPath=&FortHUDElementWrapper
   bIsVariable=true
   parent=HorizontalBox_0
   widget=FortHUDElementWrapper_Time
   properties {
    elementTag {tagName=HUD.Athena.Player.Timer.Time}
    hUDElementTag {gameplayTags [{tagName=Athena.HideForCapture}, {tagName=Athena.Creative.HUDController.HideRoundTimer}]}
   }
   slot {class=HorizontalBoxSlot, properties {verticalAlignment=VAlign_Center, padding=<ltrb 4 0 0 0>}}
  }
  TimeText {
   widgetClassPath=&TextBlock
   bIsVariable=true
   namedSlotHost=FortHUDElementWrapper_Time
   widget=TimeText
   properties {
    bSimpleTextMode=true
    justification=Center
    minDesiredWidth=50
    text=0:00
    font {
     bForceMonospaced=false
     bMaterialIsStencil=false
     fontMaterial=null
     fontObject=&NotoSans
     letterSpacing=0
     monospacedWidth=1
     size=20
     skewAmount=0
     typefaceFontName=NotoSans-Bold
     outlineSettings {
      bApplyOutlineToDropShadows=true
      bMiteredCorners=false
      bSeparateFillAlpha=false
      outlineMaterial=null
      outlineSize=1
      outlineColor=<rgba 0 0 0 0.7>
     }
    }
   }
  }
  TutorialHighlight {widgetClassPath=&TutorialHighlight}
 ]
}
