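# Minigame HUD widget for the pregame lobby, showing player count against the playlist maximum and seconds until auto-start.
# When the display event fires and the new creative start screen is disabled, a one-second looping timer refreshes values from CurrentMinigame and GameState.
# During the countdown it hides tagged HUD elements and plays tick sounds, louder near zero, then an end cue.
# Turning it off or destruction restores the HUD.
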WidgetBlueprint /Game/Athena/HUD/LTM/Playground/MinigameHUD_Joining
animations [
 IntroAnim {
  binding_count=3
  length_seconds=0.5
  object=&self:IntroAnim
  sequence {bindings [
    CountDownOverlay {
     id=#1
     display_name=CommonBorder_0
     possessed_class=Overlay
     tracks [{
       type=MovieScene2DTransformTrack
       sections [{
         type=MovieScene2DTransformSection
         end_frame=10
         start_frame=0
         channels [
          Translation.X {type=MovieSceneScriptingFloatChannel, const [0,2]}
          Translation.Y {type=MovieSceneScriptingFloatChannel, keys [[0,-250],[5.9103,23.0401],[8,0],[10,0]]}
          Angle {type=MovieSceneScriptingFloatChannel}
          Scale.X {type=MovieSceneScriptingFloatChannel}
          Scale.Y {type=MovieSceneScriptingFloatChannel}
          Shear.X {type=MovieSceneScriptingFloatChannel}
          Shear.Y {type=MovieSceneScriptingFloatChannel}
         ]
        }]
      }]
    }
    CountdownBox {
     id=#2
     possessed_class=VerticalBox
     tracks [{
       type=MovieScene2DTransformTrack
       sections [{
         type=MovieScene2DTransformSection
         end_frame=10
         start_frame=0
         channels [
          Translation.X {type=MovieSceneScriptingFloatChannel}
          Translation.Y {type=MovieSceneScriptingFloatChannel, keys [[0,0],[5.7833,0],[7.8213,9.174],[10,0]]}
          Angle {type=MovieSceneScriptingFloatChannel}
          Scale.X {type=MovieSceneScriptingFloatChannel}
          Scale.Y {type=MovieSceneScriptingFloatChannel}
          Shear.X {type=MovieSceneScriptingFloatChannel}
          Shear.Y {type=MovieSceneScriptingFloatChannel}
         ]
        }]
      }]
    }
    HorizontalBox_1 {
     id=#3
     possessed_class=HorizontalBox
     tracks [
      {
       type=MovieScene2DTransformTrack
       sections [{
         type=MovieScene2DTransformSection
         end_frame=8
         start_frame=5
         channels [
          Translation.X {type=MovieSceneScriptingFloatChannel}
          Translation.Y {type=MovieSceneScriptingFloatChannel}
          Angle {type=MovieSceneScriptingFloatChannel}
          Scale.X {type=MovieSceneScriptingFloatChannel}
          Scale.Y {type=MovieSceneScriptingFloatChannel}
          Shear.X {type=MovieSceneScriptingFloatChannel}
          Shear.Y {type=MovieSceneScriptingFloatChannel}
         ]
        }]
      }
      {
       type=MovieSceneFloatTrack
       display_name=Render Opacity
       sections [{type=MovieSceneFloatSection, end_frame=8, start_frame=0, channels [None {type=MovieSceneScriptingFloatChannel}]}]
      }
     ]
    }
   ]}
 }
 PlayerCountChanged {
  binding_count=1
  length_seconds=0.25
  object=&self:PlayerCountChanged
  sequence {bindings [TotalPlayers {
     id=#4
     possessed_class=CommonTextBlock
     tracks [{
       type=MovieScene2DTransformTrack
       sections [{
         type=MovieScene2DTransformSection
         end_frame=5
         start_frame=0
         channels [
          Translation.X {type=MovieSceneScriptingFloatChannel}
          Translation.Y {type=MovieSceneScriptingFloatChannel}
          Angle {type=MovieSceneScriptingFloatChannel}
          Scale.X {type=MovieSceneScriptingFloatChannel, keys [[0,0.8],[3,1.15],[5,1]]}
          Scale.Y {type=MovieSceneScriptingFloatChannel, keys [[0,0.8],[3,1.15],[5,1]]}
          Shear.X {type=MovieSceneScriptingFloatChannel}
          Shear.Y {type=MovieSceneScriptingFloatChannel}
         ]
        }]
      }]
    }]}
 }
]
graphs [
 EventGraph {nodes [
   n1 = Update Message() {is=title K2Node_CustomEvent}
   > UpdateText()
   n2 = Get TimeToDisplay() -> (TimeToDisplay:int=0)
   > n3 = Branch(Condition="integer >= integer"!(A:int=@n2, B:int=1) -> (ReturnValue:bool) {is=title K2Node_CallFunction})
   > Set bWasVisible(bWasVisible:bool=true) -> (Output_Get:bool)
   > n4 = SetHUDVis!(execute, self, bHideElements:bool=true) -> (then)
   n5 = Decrement Int(" "=[@n4.then, @n3.else], Value=@n2)
   > "SetText (Text)"(
    self=Get CountdownText() -> (CountdownText:Common Text Object Reference)
    InText=Format Text!(Format="LOCTEXT(\"{seconds}\")", seconds:int="Max (Integer)"(A=@n5, B=0)) -> (Result)
   )
   n6 = Get TimeToDisplay() -> (TimeToDisplay=0)
   > n7 = Branch(Condition="integer <= integer"!(A:int=@n6, B:int=0) -> (ReturnValue:bool) {is=title K2Node_CallFunction})
   > Branch(Condition=Get bWasVisible() -> (bWasVisible:bool))
   > PlaySound2D(Sound=/Game/Sounds/Creative/Toys/Floor_Siren/Timer_UI_Tick_Minigame_End_Cue)
   n8 = Branch(execute=@n7.else, Condition="integer > integer"(A=@n6, B=20) {is=title K2Node_CallFunction})
   PlaySound2D(
    execute=@n8.else
    Sound=/Game/Sounds/Creative/Toys/Floor_Siren/Timer_UI_Tick_Minigame_Cue
    VolumeMultiplier=Select(Option 0:double=1, Option 1:double=0.5, Index="integer > integer"(A=@n6, B=5)) -> (ReturnValue:double)
   )
   TestThisShit() {is=title K2Node_CustomEvent}
   > SetVisibility(self=Get CountDownOverlay() -> (CountDownOverlay:Overlay Object Reference), InVisibility=SelfHitTestInvisible)
   > PlayAnimation(InAnimation=Get IntroAnim() -> (IntroAnim:Widget Animation Object Reference))
   UpdateText() {is=title K2Node_CustomEvent}
   n9 = Get CurrentMinigame() -> (CurrentMinigame:Fort Minigame Object Reference)
   > n10 = Branch(Condition="float > float"(A=GetTimeRemainingUntilAutoStart!(self:Fort Minigame Object Reference=@n9) -> (ReturnValue:float), B=0) {is=title K2Node_CallFunction})
   n11 = Get CountDownOverlay()
   > SetVisibility(self=@n11, InVisibility=SelfHitTestInvisible)
   n12 = Get CurrentMinigame()
   > "SetText (Text)"(
    self=Get TotalPlayers() -> (TotalPlayers:Common Text Object Reference)
    InText="To Text (Integer)"(Value=GetPlayerCount!(self:Fort Minigame Object Reference=@n12, bIncludeSpectators:bool=true) -> (ReturnValue:int=0))
   )
   > n13 = Cast To FortGameStateAthena!(execute, Object=GetGameState!() -> (ReturnValue:Game State Base Object Reference)) -> (then, CastFailed, AsFort Game State Athena:Fort Game State Athena Object Reference) {is=title K2Node_DynamicCast}
   > n14 = "SetText (Text)"(
    self=Get MaxPlayers() -> (MaxPlayers:Common Text Object Reference)
    InText="To Text (Integer)"(Value=GetMaxPlaylistPlayers!(self:Fort Game State Athena Object Reference=@n13) -> (ReturnValue:int=0))
   )
   n15 = Set TimeToDisplay(execute=[@n14.then, @n13.CastFailed], TimeToDisplay:int=Round(A=GetTimeRemainingUntilAutoStart(self=@n12)) {is=title K2Node_CallFunction}) -> (Output_Get:int=0)
   > Branch(Condition="integer >= integer"(A=@n15, B=1))
   > "SetText (Text)"(self=Get CountdownText(), InText=Format Text(Format="LOCTEXT(\"{seconds}\")", seconds=@n15))
   SetVisibility(execute=@n10.else, self=@n11, InVisibility=Collapsed)
   n16 = SetHUDVis!() -> (OutputDelegate, then, bHideElements:bool=~) {is=title K2Node_CustomEvent}
   > n17 = Cast To FortPlayerControllerAthena(Object=GetOwningPlayer())
   > SetHUDElementVisibilityWithContext!(
    execute
    self:Fort Player Controller Object Reference=@n17
    HUDElementTags:"Gameplay Tag Container Structure (by ref)"=MakeGameplayTagContainerFromTag!(SingleTag:Gameplay Tag Structure=(TagName="HUD.Hide.CreativeJoiningLobby")) -> (ReturnValue:Gameplay Tag Container Structure)
    bHideElements:bool=@n16.bHideElements
    ReasonName:Name=None
    LogContext:String
   ) -> (then)
   n18 = Event OnDisplayMinigamePregameCountdownUI!() -> (OutputDelegate, then, bDisplay:bool) {is=title K2Node_Event}
   > n19 = Branch(Condition=IsNewCreativeStartScreenEnabled!() -> (ReturnValue:bool))
   > SetVisibility(InVisibility=Collapsed)
   n20 = Branch(execute=@n19.else, Condition=@n18.bDisplay)
   > n21 = Set Timer by Event(Delegate=@n1, Time=1, bLooping=true)
   > Set TimerHandle(TimerHandle:Timer Handle Structure=@n21) -> (Output_Get:Timer Handle Structure)
   > Update Message() {is=title K2Node_CallFunction}
   > Branch(Condition=IsVisible(self=Get CountDownOverlay()))
   > PlaySound2D(Sound=&Timer_UI_Tick_Minigame_Cue)
   n22 = Branch(execute=@n20.else, Condition=Get bWasVisible())
   > n23 = SetHUDVis(bHideElements=false) {is=title K2Node_CallFunction}
   Clear and Invalidate Timer by Handle(execute=[@n23.then, @n22.else], Handle=Get TimerHandle() -> (TimerHandle:Timer Handle Structure))
   > SetVisibility(self=Get CountDownOverlay(), InVisibility=Collapsed)
   > Set bWasVisible(bWasVisible=false)
   Event Destruct()
   > SetHUDVis(bHideElements=false)
  ]}
 FormatTextForTimer {nodes [
   n1 = Get HighlightTag() -> (HighlightTag:String)
   Outputs!(Text:Text=Format Text!(
     Format="LOCTEXT(\"{Data}{HighlightTag}{time}</>\")"
     Data:Text=Get DisplayedText() -> (DisplayedText:Text)
     HighlightTag:Text="To Text (String)"(InString=@n1)
     time:int=Inputs!() -> (Time:int)
    ) -> (Result))
   Get Game Text() -> (Game Text:Text)
  ]}
]
variables [
 bWasVisible {type=boolean}
 DisplayedText {type=string, default=""}
 Game Text {type=string, default="<keyword>Game Starting</>\r\nIn: "}
 HighlightTag {type=string, default="<enemy>"}
 TimerHandle {type=TimerHandle}
 TimeToDisplay {type=integer}
]
widgets {
 info {inheritedWidgetCount=1, parentClass=/Script/FortniteUI.MinigameWidgetBase, rootWidgetClass=&Overlay, widgetCount=16}
 widgets [
  CountDownOverlay {widgetClassPath=&Overlay, bIsVariable=true, widget=CountDownOverlay}
  CommonBorder_0 {
   widgetClassPath=&CommonBorder
   parent=CountDownOverlay
   widget=CommonBorder_0
   properties {
    background {
     bIsDynamicallyLoaded=false
     drawAs=Image
     imageType=NoImage
     mirroring=NoMirror
     resourceName=None
     resourceObject=/Game/UI/Foundation/Modals/ErrorWindow/Materials/M_UI_HorizontalFade_Border
     tiling=NoTile
     imageSize=<24 24>
     margin=<ltrb 0.0306 0.1162 0.1162 0.0917>
     outlineSettings {
      bUseBrushTransparency=false
      roundingType=HalfHeightRadius
      width=0
      color {colorUseRule=UseColor_Specified, specifiedColor=<rgba 0 0 0 0>}
      cornerRadii=<0 0 0 0>
     }
     tintColor {colorUseRule=UseColor_Specified, specifiedColor=<rgba 1 1 1 1>}
     uVRegion {bIsValid=false, max=<0 0>, min=<0 0>}
    }
    brushColor=<rgba 0 0.0021 0.0176 0.2>
    padding=<ltrb 24 8 24 8>
   }
   slot {class=OverlaySlot, properties {horizontalAlignment=HAlign_Center, verticalAlignment=VAlign_Bottom, padding=<ltrb 0 80 0 80>}}
  }
  CountdownBox {
   widgetClassPath=&VerticalBox
   bIsVariable=true
   parent=CommonBorder_0
   widget=CountdownBox
   slot {class=BorderSlot, properties {padding=<ltrb 24 8 24 8>}}
  }
  HorizontalBox_1 {
   widgetClassPath=&HorizontalBox
   parent=CountdownBox
   widget=HorizontalBox_1
   slot {class=VerticalBoxSlot, properties {padding=<ltrb 6 0 0 0>}}
  }
  TimerTitle {
   widgetClassPath=&CommonTextBlock
   parent=HorizontalBox_1
   widget=TimerTitle
   properties {
    text=Waiting for Players
    textTransformPolicy=ToUpper
    colorAndOpacity {colorUseRule=UseColor_Specified, specifiedColor=<rgba 0.2 0.6667 1 1>}
    font {
     bForceMonospaced=false
     bMaterialIsStencil=false
     fontMaterial=null
     fontObject=&BurbankSmall
     letterSpacing=0
     monospacedWidth=1
     size=24
     skewAmount=0
     typefaceFontName=Black
     outlineSettings {
      bApplyOutlineToDropShadows=false
      bMiteredCorners=false
      bSeparateFillAlpha=false
      outlineMaterial=null
      outlineSize=1
      outlineColor=<rgba 0 0.0833 1 0.25>
     }
    }
    renderTransform {angle=0, scale=<1 1>, shear=<-8 0>, translation=<0 0>}
    shadowColorAndOpacity=<rgba 0 0.0046 0.075 0.25>
    shadowOffset=<3 3>
   }
   slot {class=HorizontalBoxSlot, properties {verticalAlignment=VAlign_Center}}
  }
  Throbber_42 {widgetClassPath=/Script/UMG.Throbber, bIsVariable=true, parent=HorizontalBox_1, widget=Throbber_42, slot {class=HorizontalBoxSlot}}
  HorizontalBox_2 {
   widgetClassPath=&HorizontalBox
   parent=CountdownBox
   widget=HorizontalBox_2
   slot {class=VerticalBoxSlot, properties {horizontalAlignment=HAlign_Center, padding=<ltrb 0 -10 0 -10>}}
  }
  TotalPlayers {
   widgetClassPath=&CommonTextBlock
   bIsVariable=true
   parent=HorizontalBox_2
   widget=TotalPlayers
   properties {
    text=1
    colorAndOpacity {colorUseRule=UseColor_Specified, specifiedColor=<rgba 0.2 1 1 1>}
    font {
     bForceMonospaced=false
     bMaterialIsStencil=false
     fontMaterial=null
     fontObject=&BurbankSmall
     letterSpacing=0
     monospacedWidth=1
     size=32
     skewAmount=0
     typefaceFontName=Black
     outlineSettings {
      bApplyOutlineToDropShadows=false
      bMiteredCorners=false
      bSeparateFillAlpha=false
      outlineMaterial=null
      outlineSize=0
      outlineColor=<rgba 0 0 0 1>
     }
    }
    shadowColorAndOpacity=<rgba 0 0.15 0.15 0.2471>
    shadowOffset=<3 3>
   }
   slot {class=HorizontalBoxSlot}
  }
  CommonTextBlock_1 {
   widgetClassPath=&CommonTextBlock
   bIsVariable=true
   parent=HorizontalBox_2
   widget=CommonTextBlock_1
   properties {
    text=/
    colorAndOpacity {colorUseRule=UseColor_Specified, specifiedColor=<rgba 0.9047 0.9047 0.9047 1>}
    font {
     bForceMonospaced=false
     bMaterialIsStencil=false
     fontMaterial=null
     fontObject=&BurbankSmall
     letterSpacing=0
     monospacedWidth=1
     size=32
     skewAmount=0
     typefaceFontName=Black
     outlineSettings {
      bApplyOutlineToDropShadows=false
      bMiteredCorners=false
      bSeparateFillAlpha=false
      outlineMaterial=null
      outlineSize=0
      outlineColor=<rgba 0 0 0 1>
     }
    }
    shadowColorAndOpacity=<rgba 0 0.15 0.15 0.2471>
    shadowOffset=<3 3>
   }
   slot {class=HorizontalBoxSlot}
  }
  MaxPlayers {
   widgetClassPath=&CommonTextBlock
   bIsVariable=true
   parent=HorizontalBox_2
   widget=MaxPlayers
   properties {
    text=16
    colorAndOpacity {colorUseRule=UseColor_Specified, specifiedColor=<rgba 0.9047 0.9047 0.9047 1>}
    font {
     bForceMonospaced=false
     bMaterialIsStencil=false
     fontMaterial=null
     fontObject=&BurbankSmall
     letterSpacing=0
     monospacedWidth=1
     size=32
     skewAmount=0
     typefaceFontName=Black
     outlineSettings {
      bApplyOutlineToDropShadows=false
      bMiteredCorners=false
      bSeparateFillAlpha=false
      outlineMaterial=null
      outlineSize=0
      outlineColor=<rgba 0 0 0 1>
     }
    }
    shadowColorAndOpacity=<rgba 0 0.15 0.15 0.2471>
    shadowOffset=<3 3>
   }
   slot {class=HorizontalBoxSlot}
  }
  HorizontalBox_0 {
   widgetClassPath=&HorizontalBox
   parent=CountdownBox
   widget=HorizontalBox_0
   slot {class=VerticalBoxSlot, properties {horizontalAlignment=HAlign_Center, verticalAlignment=VAlign_Center}}
  }
  CommonTextBlock_190 {
   widgetClassPath=&CommonTextBlock
   parent=HorizontalBox_0
   widget=CommonTextBlock_190
   properties {
    text=Autostarts in
    textTransformPolicy=ToUpper
    colorAndOpacity {colorUseRule=UseColor_Specified, specifiedColor=<rgba 0.2016 0.6724 1 1>}
    font {
     bForceMonospaced=false
     bMaterialIsStencil=false
     fontMaterial=null
     fontObject=&BurbankSmall
     letterSpacing=0
     monospacedWidth=1
     size=16
     skewAmount=0
     typefaceFontName=Black
     outlineSettings {
      bApplyOutlineToDropShadows=false
      bMiteredCorners=false
      bSeparateFillAlpha=false
      outlineMaterial=null
      outlineSize=1
      outlineColor=<rgba 0 0.0833 1 0.25>
     }
    }
    shadowOffset=<0 0>
   }
   slot {class=HorizontalBoxSlot, properties {verticalAlignment=VAlign_Center}}
  }
  CountdownText {
   widgetClassPath=&CommonTextBlock
   bIsVariable=true
   parent=HorizontalBox_0
   widget=CountdownText
   properties {
    justification=Center
    minDesiredWidth=53.5955
    text=60
    colorAndOpacity {colorUseRule=UseColor_Specified, specifiedColor=<rgba 0 0.1667 1 1>}
    font {
     bForceMonospaced=false
     bMaterialIsStencil=false
     fontMaterial=null
     fontObject=&BurbankSmall
     letterSpacing=0
     monospacedWidth=1
     size=24
     skewAmount=0
     typefaceFontName=Black
     outlineSettings {
      bApplyOutlineToDropShadows=false
      bMiteredCorners=false
      bSeparateFillAlpha=false
      outlineMaterial=null
      outlineSize=0
      outlineColor=<rgba 0 0 0 1>
     }
    }
    shadowOffset=<0 0>
   }
   slot {class=HorizontalBoxSlot, properties {horizontalAlignment=HAlign_Center}}
  }
  CanvasPanel_239 {
   widgetClassPath=&CanvasPanel
   parent=CountDownOverlay
   widget=CanvasPanel_239
   slot {class=OverlaySlot, properties {horizontalAlignment=HAlign_Fill, verticalAlignment=VAlign_Fill}}
  }
  HamburgerButton {
   widgetClassPath=/Game/Athena/HUD/Bacchus/HamburgerButton
   bIsVariable=true
   parent=CanvasPanel_239
   widget=HamburgerButton
   properties {elementTag {tagName=HUD.Bacchus.Hamburger}, hUDElementTag {gameplayTags []}}
   slot {
    class=CanvasPanelSlot
    properties {bAutoSize=true, layoutData {alignment=<0 0>, anchors {maximum=<1 0>, minimum=<1 0>}, offsets=<ltrb -195 -75 168 160>}}
   }
  }
  TutorialHighlight {widgetClassPath=&TutorialHighlight}
 ]
}
