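# Heist LTM HUD widget showing comm messages as an animated radio with a speech bubble and sound.
# It binds to game state message requests, sets team material parameters, hides bottom-left HUD, and plays Intro and TalkAnim, with AngryStyle for critical messages.
# The empty Delay animation acts as hold timer, its finish triggering Outro.
# While busy, only critical messages queue; Outro completion restores the HUD and replays the next.
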WidgetBlueprint /Game/Athena/HUD/LTM/Heist/HeistHUD_Radio
animations [
 Intro {
  binding_count=5
  display_label=""
  length_seconds=0.6705
  object=&self:Intro
  sequence {
   bindings [
    Overlay_Container {
     id=#1
     possessed_class=Overlay
     tracks [{
       type=MovieSceneFloatTrack
       display_name=Render Opacity
       sections [{type=MovieSceneFloatSection, end_frame=13, start_frame=0, channels [None {type=MovieSceneScriptingFloatChannel, keys [[0,0],[3.564,1]]}]}]
      }]
    }
    Image_RadioBG {
     id=#2
     possessed_class=Image
     tracks [
      {
       type=MovieScene2DTransformTrack
       sections [{
         type=MovieScene2DTransformSection
         end_frame=25
         start_frame=0
         channels [
          Translation.X {type=MovieSceneScriptingFloatChannel}
          Translation.Y {type=MovieSceneScriptingFloatChannel}
          Angle {type=MovieSceneScriptingFloatChannel}
          Scale.X {type=MovieSceneScriptingFloatChannel, keys [[0,0],[9,1.0346],[12.436,1]]}
          Scale.Y {type=MovieSceneScriptingFloatChannel, keys [[0,0],[9,1.0346],[12.436,1]]}
          Shear.X {type=MovieSceneScriptingFloatChannel}
          Shear.Y {type=MovieSceneScriptingFloatChannel}
         ]
        }]
      }
      {
       type=MovieSceneWidgetMaterialTrack
       sections [{
         type=MovieSceneParameterSection
         end_frame=25
         start_frame=0
         channels [
          ShadowScaleR {type=MovieSceneScriptingFloatChannel, keys [[0.033,1.8696],[9,1.9142]]}
          ShadowScaleG {type=MovieSceneScriptingFloatChannel, keys [[0.033,1.8696],[9,0.9793]]}
          ShadowScaleB {type=MovieSceneScriptingFloatChannel}
          ShadowScaleA {type=MovieSceneScriptingFloatChannel}
          ShadowOffsetR {type=MovieSceneScriptingFloatChannel, keys [[5,-0.4594],[12.092,-0.2952],[16,-0.3266],[18.297,-0.2952]]}
          ShadowOffsetG {type=MovieSceneScriptingFloatChannel, keys [[5,0.0608],[12.092,0.1143],[16,0.1369],[18.297,0.1143]]}
          ShadowOffsetB {type=MovieSceneScriptingFloatChannel}
          ShadowOffsetA {type=MovieSceneScriptingFloatChannel}
         ]
        }]
      }
     ]
    }
    SizeBox_Message {
     id=#3
     possessed_class=SizeBox
     tracks [{
       type=MovieScene2DTransformTrack
       sections [{
         type=MovieScene2DTransformSection
         end_frame=37
         start_frame=16
         channels [
          Translation.X {type=MovieSceneScriptingFloatChannel}
          Translation.Y {type=MovieSceneScriptingFloatChannel}
          Angle {type=MovieSceneScriptingFloatChannel}
          Scale.X {type=MovieSceneScriptingFloatChannel, keys [[24.997,1],[26.632,1.1],[29.967,1]]}
          Scale.Y {type=MovieSceneScriptingFloatChannel, keys [[24.997,1],[26.632,1.1],[29.967,1]]}
          Shear.X {type=MovieSceneScriptingFloatChannel}
          Shear.Y {type=MovieSceneScriptingFloatChannel}
         ]
        }]
      }]
    }
    Image_Radio {
     id=#4
     possessed_class=Image
     tracks [{
       type=MovieScene2DTransformTrack
       sections [{
         type=MovieScene2DTransformSection
         end_frame=27
         start_frame=2
         channels [
          Translation.X {type=MovieSceneScriptingFloatChannel}
          Translation.Y {type=MovieSceneScriptingFloatChannel}
          Angle {type=MovieSceneScriptingFloatChannel}
          Scale.X {type=MovieSceneScriptingFloatChannel, keys [[2,0],[12,0.7955],[14.436,0.6],[18.297,0.6]]}
          Scale.Y {type=MovieSceneScriptingFloatChannel, keys [[2,0],[12,0.7955],[14.436,0.6],[18.297,0.6]]}
          Shear.X {type=MovieSceneScriptingFloatChannel}
          Shear.Y {type=MovieSceneScriptingFloatChannel}
         ]
        }]
      }]
    }
    Overlay_Dialog {
     id=#5
     possessed_class=Overlay
     tracks [
      {
       type=MovieScene2DTransformTrack
       sections [{
         type=MovieScene2DTransformSection
         end_frame=40
         start_frame=0
         channels [
          Translation.X {type=MovieSceneScriptingFloatChannel}
          Translation.Y {type=MovieSceneScriptingFloatChannel}
          Angle {type=MovieSceneScriptingFloatChannel, keys [[15.593,0],[23.605,0.1397],[27.306,0]]}
          Scale.X {type=MovieSceneScriptingFloatChannel, keys [[14.796,0],[24.592,1.1282],[27.232,1],[30.354,1]]}
          Scale.Y {type=MovieSceneScriptingFloatChannel, keys [[14.796,0],[24.592,1.1282],[27.232,1],[30.354,1]]}
          Shear.X {type=MovieSceneScriptingFloatChannel}
          Shear.Y {type=MovieSceneScriptingFloatChannel}
         ]
        }]
      }
      {
       type=MovieSceneFloatTrack
       display_name=Render Opacity
       sections [{type=MovieSceneFloatSection, end_frame=40, start_frame=0, channels [None {type=MovieSceneScriptingFloatChannel, const [1,1]}]}]
      }
     ]
    }
   ]
   display_rate {numerator=60}
  }
 }
 Outro {
  binding_count=3
  display_label=""
  length_seconds=0.4441
  object=&self:Outro
  sequence {
   bindings [
    Image_RadioBG {
     id=#2
     possessed_class=Image
     tracks [{
       type=MovieScene2DTransformTrack
       sections [{
         type=MovieScene2DTransformSection
         end_frame=26
         start_frame=0
         channels [
          Translation.X {type=MovieSceneScriptingFloatChannel}
          Translation.Y {type=MovieSceneScriptingFloatChannel}
          Angle {type=MovieSceneScriptingFloatChannel}
          Scale.X {type=MovieSceneScriptingFloatChannel, keys [[0,1.0346],[6.09,1],[9.612,0]]}
          Scale.Y {type=MovieSceneScriptingFloatChannel, keys [[0,1.0346],[6.09,1],[9.612,0]]}
          Shear.X {type=MovieSceneScriptingFloatChannel}
          Shear.Y {type=MovieSceneScriptingFloatChannel}
         ]
        }]
      }]
    }
    Image_Radio {
     id=#4
     possessed_class=Image
     tracks [{
       type=MovieScene2DTransformTrack
       sections [{
         type=MovieScene2DTransformSection
         end_frame=26
         start_frame=0
         channels [
          Translation.X {type=MovieSceneScriptingFloatChannel}
          Translation.Y {type=MovieSceneScriptingFloatChannel}
          Angle {type=MovieSceneScriptingFloatChannel}
          Scale.X {type=MovieSceneScriptingFloatChannel, keys [[0,0.7076],[6.09,0.6],[9.612,0]]}
          Scale.Y {type=MovieSceneScriptingFloatChannel, keys [[0,0.7076],[6.09,0.6],[9.612,0]]}
          Shear.X {type=MovieSceneScriptingFloatChannel}
          Shear.Y {type=MovieSceneScriptingFloatChannel}
         ]
        }]
      }]
    }
    Overlay_Dialog {
     id=#5
     possessed_class=Overlay
     tracks [
      {
       type=MovieScene2DTransformTrack
       sections [{
         type=MovieScene2DTransformSection
         end_frame=26
         start_frame=0
         channels [
          Translation.X {type=MovieSceneScriptingFloatChannel}
          Translation.Y {type=MovieSceneScriptingFloatChannel}
          Angle {type=MovieSceneScriptingFloatChannel}
          Scale.X {type=MovieSceneScriptingFloatChannel, keys [[0,1.0346],[6.09,1],[9.612,0]]}
          Scale.Y {type=MovieSceneScriptingFloatChannel, keys [[0,1.0346],[6.09,1],[9.612,0]]}
          Shear.X {type=MovieSceneScriptingFloatChannel}
          Shear.Y {type=MovieSceneScriptingFloatChannel}
         ]
        }]
      }
      {
       type=MovieSceneFloatTrack
       display_name=Render Opacity
       sections [{type=MovieSceneFloatSection, end_frame=10, start_frame=0, channels [None {type=MovieSceneScriptingFloatChannel, keys [[0,1],[4.908,0]]}]}]
      }
     ]
    }
   ]
   display_rate {numerator=60}
  }
 }
 Delay {
  binding_count=0
  display_label=""
  length_seconds=5.0089
  object=&self:Delay
  sequence {bindings [], tracks [{type=MovieSceneEventTrack, sections [{type=MovieSceneEventSection}]}]}
 }
 TalkAnim {
  binding_count=6
  display_label=""
  length_seconds=0.8055
  object=&self:TalkAnim
  sequence {
   bindings [
    Image_DialogBG {
     id=#6
     possessed_class=Image
     tracks [{
       type=MovieSceneWidgetMaterialTrack
       sections [{
         type=MovieSceneParameterSection
         end_frame=0
         start_frame=0
         channels [
          BoltAmount {type=MovieSceneScriptingFloatChannel, const [0,1]}
          AccentColorR {type=MovieSceneScriptingFloatChannel, const [0,1]}
          AccentColorG {type=MovieSceneScriptingFloatChannel, const [0.9216,1]}
          AccentColorB {type=MovieSceneScriptingFloatChannel, const [1,1]}
          AccentColorA {type=MovieSceneScriptingFloatChannel, const [1,1]}
         ]
        }]
      }]
    }
    Image_Callout {
     id=#7
     possessed_class=Image
     tracks [{
       type=MovieSceneWidgetMaterialTrack
       sections [{
         type=MovieSceneParameterSection
         end_frame=48
         start_frame=0
         channels [
          BoltEdge {type=MovieSceneScriptingFloatChannel, keys [[9.304,0.5814],[21.545,0.4794],[28.996,0.549],[45.72,0.9191]]}
          BoltReveal {type=MovieSceneScriptingFloatChannel, keys [[7.808,1.0882],[29.272,-0.1979]]}
          BoltOutro {type=MovieSceneScriptingFloatChannel, keys [[18.068,1.3466],[44.325,-0.15]]}
         ]
        }]
      }]
    }
    Image_Radio {
     id=#8
     possessed_class=Image
     tracks [
      {
       type=MovieSceneWidgetMaterialTrack
       sections [{
         type=MovieSceneParameterSection
         end_frame=34
         start_frame=0
         channels [
          HeightRatio {type=MovieSceneScriptingFloatChannel, keys [[6.631,0],[17.429,3.0226],[27.751,0]]}
          FullBodyDistort {type=MovieSceneScriptingFloatChannel, keys [[6.631,0],[17.429,0.9527],[27.751,0]]}
         ]
        }]
      }
      {
       type=MovieScene2DTransformTrack
       sections [{
         type=MovieScene2DTransformSection
         end_frame=39
         start_frame=0
         channels [
          Translation.X {type=MovieSceneScriptingFloatChannel}
          Translation.Y {type=MovieSceneScriptingFloatChannel, keys [[0,-50],[12.535,-69.9842],[25.164,-50]]}
          Angle {type=MovieSceneScriptingFloatChannel}
          Scale.X {type=MovieSceneScriptingFloatChannel, keys [[0,0.6],[3.246,0.55],[12.535,0.6],[25.164,0.6]]}
          Scale.Y {type=MovieSceneScriptingFloatChannel, keys [[0,0.6],[3.246,0.4767],[12.582,0.6622],[26,0.5469],[33.831,0.6],[39.006,0.6]]}
          Shear.X {type=MovieSceneScriptingFloatChannel}
          Shear.Y {type=MovieSceneScriptingFloatChannel}
         ]
        }]
      }
     ]
    }
    OverlayCallout {
     id=#9
     possessed_class=Overlay
     tracks [
      {
       type=MovieScene2DTransformTrack
       sections [{
         type=MovieScene2DTransformSection
         end_frame=48
         start_frame=0
         channels [
          Translation.X {type=MovieSceneScriptingFloatChannel}
          Translation.Y {type=MovieSceneScriptingFloatChannel}
          Angle {type=MovieSceneScriptingFloatChannel}
          Scale.X {type=MovieSceneScriptingFloatChannel, keys [[0,1],[35.992,1.2]]}
          Scale.Y {type=MovieSceneScriptingFloatChannel, keys [[0,1],[35.992,1.2]]}
          Shear.X {type=MovieSceneScriptingFloatChannel}
          Shear.Y {type=MovieSceneScriptingFloatChannel}
         ]
        }]
      }
      {
       type=MovieSceneFloatTrack
       display_name=Render Opacity
       sections [{type=MovieSceneFloatSection, end_frame=47, start_frame=0, channels [None {type=MovieSceneScriptingFloatChannel, keys [[0,0],[9.965,1]]}]}]
      }
     ]
    }
    Image_CalloutShadow {
     id=#10
     possessed_class=Image
     tracks [{
       type=MovieSceneWidgetMaterialTrack
       sections [{
         type=MovieSceneParameterSection
         end_frame=48
         start_frame=0
         channels [
          BoltReveal {type=MovieSceneScriptingFloatChannel, keys [[7.808,1.08],[29.272,-0.19]]}
          BoltOutro {type=MovieSceneScriptingFloatChannel, keys [[18.068,1.34],[44.325,-0.15]]}
         ]
        }]
      }]
    }
    Image_RadioBG {
     id=#11
     possessed_class=Image
     tracks [{
       type=MovieSceneWidgetMaterialTrack
       sections [{
         type=MovieSceneParameterSection
         end_frame=48
         start_frame=0
         channels [
          ShadowOffsetR {
           type=MovieSceneScriptingFloatChannel
           keys [[0,-0.2952],[4,-0.3161],[11.501,-0.2422],[17.146,-0.2005],[24,-0.2952],[25.861,-0.2906],[29.348,-0.2952]]
          }
          ShadowOffsetG {type=MovieSceneScriptingFloatChannel, keys [[0,0.1143],[4,0.1414],[11.501,0.1371],[17.146,0.1164],[24,0.1143],[29.348,0.1143]]}
          ShadowOffsetB {type=MovieSceneScriptingFloatChannel}
          ShadowOffsetA {type=MovieSceneScriptingFloatChannel}
          ShadowScaleR {type=MovieSceneScriptingFloatChannel, keys [[0,1.9142],[11.501,1.7504],[17,1.7113],[24,1.934],[29.348,1.911]]}
          ShadowScaleG {type=MovieSceneScriptingFloatChannel, keys [[0,0.9793],[11.569,0.9393],[17,0.9393],[23,0.9987],[29,0.9793]]}
          ShadowScaleB {type=MovieSceneScriptingFloatChannel}
          ShadowScaleA {type=MovieSceneScriptingFloatChannel}
         ]
        }]
      }]
    }
   ]
   display_rate {numerator=60}
  }
 }
 AngryStyle {
  binding_count=2
  display_label=""
  length_seconds=0.5833
  object=&self:AngryStyle
  sequence {bindings [
    Image_DialogTail {
     id=#12
     possessed_class=Image
     tracks [{
       type=MovieSceneWidgetMaterialTrack
       sections [{
         type=MovieSceneParameterSection
         end_frame=11
         start_frame=0
         channels [
          AccentColorR {type=MovieSceneScriptingFloatChannel, keys [[0,0],[11.664,1]]}
          AccentColorG {type=MovieSceneScriptingFloatChannel, keys [[0,0.9216],[11.664,0.04]]}
          AccentColorB {type=MovieSceneScriptingFloatChannel, keys [[0,1],[11.664,0]]}
          AccentColorA {type=MovieSceneScriptingFloatChannel, const [1,2]}
         ]
        }]
      }]
    }
    Image_dialogBG {
     id=#13
     possessed_class=Image
     tracks [{
       type=MovieSceneWidgetMaterialTrack
       sections [{
         type=MovieSceneParameterSection
         end_frame=11
         start_frame=0
         channels [
          BoltAmount {type=MovieSceneScriptingFloatChannel, keys [[0,0],[4.862,0.8646],[6.339,0.3964],[9.4313,0.7013],[11.6543,0.62]]}
          AccentColorR {type=MovieSceneScriptingFloatChannel, keys [[0,0],[9.98,1]]}
          AccentColorG {type=MovieSceneScriptingFloatChannel, keys [[0,0.9216],[9.98,0.04]]}
          AccentColorB {type=MovieSceneScriptingFloatChannel, keys [[0,1],[9.98,0]]}
          AccentColorA {type=MovieSceneScriptingFloatChannel, const [1,2]}
         ]
        }]
      }]
    }
   ]}
 }
]
graphs [
 EventGraph {nodes [
   Delay(Duration=3)
   n1 = Get Intro()
   > PlayAnimation(InAnimation=@n1)
   > PlayAnimation(
    InAnimation=Get AngryStyle() -> (AngryStyle:Widget Animation Object Reference)
    StartAtTime=float + float(A=GetEndTime(self=@n1), B=3) {is=title K2Node_CommutativeAssociativeBinaryOperator}
   )
   > Delay(Duration=3)
   > PlayAnimation(InAnimation=Get TalkAnim() -> (TalkAnim:Widget Animation Object Reference))
   n2 = OutroComplete() {is=title K2Node_CustomEvent}
   > HUDToggle!(execute, self, Hide:bool=false) -> (then)
   > Set CanStartMessage(CanStartMessage:bool=true) -> (Output_Get:bool)
   n3 = Get QueuedMessages() -> (QueuedMessages:Array of Athena Game Message Data Structures)
   > Branch(Condition="integer > integer"(A=Length(TargetArray:Array of Athena Game Message Data Structures=@n3), B=0) {is=title K2Node_CallFunction})
   > Set QueuedMsg(QueuedMsg:Athena Game Message Data Structure="Get (a copy)"(Array:Array of Athena Game Message Data Structures=Get QueuedMessages()) -> (Output:Athena Game Message Data Structure)) -> (Output_Get:Athena Game Message Data Structure)
   > Remove Index!(execute, TargetArray:Array of Athena Game Message Data Structures=Get QueuedMessages(), IndexToRemove:int=0) -> (then)
   > OnMessageRequested!(execute, self, MsgData:Athena Game Message Data Structure=Get QueuedMsg() -> (QueuedMsg:Athena Game Message Data Structure)) -> (then)
   n4 = DelayComplete() {is=title K2Node_CustomEvent}
   n5 = Get Outro() -> (Outro:Widget Animation Object Reference)
   > StopAnimation(InAnimation=@n5)
   > PlayAnimation(InAnimation=@n5)
   n6 = OnMessageRequested!() -> (OutputDelegate, then, MsgData:Athena Game Message Data Structure) {is=title K2Node_CustomEvent}
   n7 = Break Athena Game Message Data!(AthenaGameMessageData:"Athena Game Message Data Structure (by ref)"=@n6.MsgData) -> (
    MsgType:EAthenaGameMsgType Enum
    MsgText:Text
    MsgSound:Sound Cue Object Reference
    MsgMediaSource:Media Source Object Reference
    MsgDelay:float=~
    bIsTeamBased:bool=~
    TeamIndex:int
    DisplayTime:float=~
    MessageFlags:EAthenaGameMessageDataFlags Enum
    TargetPlayerController:Player Controller Object Reference
   ) {is=title K2Node_BreakStruct}
   > n8 = Branch(Condition=@n7.bIsTeamBased)
   > n9 = GetDynamicMaterial(self=Get Image_Radio() -> (Image_Radio:Image Object Reference))
   > SetScalarParameterValue(self=@n9, ParameterName=TeamActive, Value=1)
   > n10 = SetScalarParameterValue(self=@n9, ParameterName=TeamType, Value="To Float (Integer)"(InInt=@n7.TeamIndex))
   n11 = Branch(execute=[@n10.then, @n8.else], Condition=Get CanStartMessage() -> (CanStartMessage:bool))
   n12 = Break Athena Game Message Data(AthenaGameMessageData=@n6.MsgData) -> (MsgDelay=~, bIsTeamBased=~, DisplayTime=~)
   > n13 = Switch on EAthenaGameMsgType!(execute, Selection:EAthenaGameMsgType Enum=@n12) -> (
    None:Exec
    DefaultIntro:Exec
    DefaultMessage:Exec
    DefaultCriticalMessage:Exec
    CommIntro:Exec
    CommMessage:Exec
    CommCriticalMessage:Exec
    CornerIntro:Exec
    CornerMessage:Exec
    CornerCriticalMessage:Exec
    RespawnTurningOffWarning:Exec
    RespawnOffWarning:Exec
    CenterMessage:Exec
    CenterImportantMessage:Exec
    CenterCriticalMessage:Exec
   ) {is=title K2Node_SwitchEnum}
   n14 = PlayAnimation(execute=@n13.CommCriticalMessage, InAnimation=Get AngryStyle())
   Set CanStartMessage(execute=[@n14.then, @n13.CommIntro, @n13.CommMessage], CanStartMessage=false)
   > SetText(self=Get RichTextBlock_Message() -> (RichTextBlock_Message:Rich Text Block Object Reference), InText=@n12.MsgText)
   > SetVisibility(self=Get Overlay_Container() -> (Overlay_Container:Overlay Object Reference), InVisibility=HitTestInvisible)
   n15 = Get Intro()
   > StopAnimation(InAnimation=@n15)
   > n16 = Sequence()
   > PlaySound2D(Sound=@n12.MsgSound, VolumeMultiplier=0.25)
   > HUDToggle(Hide=true)
   PlayAnimation(execute=@n16.then_1, InAnimation=@n15)
   > PlayAnimation(InAnimation=Get TalkAnim(), NumLoopsToPlay=2)
   n17 = Get Delay() -> (Delay:Widget Animation Object Reference)
   > StopAnimation(InAnimation=@n17)
   > PlayAnimation(InAnimation=@n17)
   Branch(
    execute=@n11.else
    Condition="Equal (Enum)"(
     A:EAthenaGameMsgType Enum=Break Athena Game Message Data(AthenaGameMessageData=@n6.MsgData) -> (MsgDelay=~, bIsTeamBased=~, DisplayTime=~)
     B:EAthenaGameMsgType Enum=CommCriticalMessage
    )
   )
   > Add(
    TargetArray:Array of Athena Game Message Data Structures=Get QueuedMessages()
    NewItem:"Athena Game Message Data Structure (by ref)"=@n6.MsgData
   ) {is=title K2Node_CallArrayFunction}
   Event Construct()
   > n18 = Cast To FortGameStateAthena!(execute, Object=GetGameState!() -> (ReturnValue:Game State Base Object Reference)) -> (then, CastFailed, AsFort Game State Athena:Fort Game State Athena Object Reference) {is=title K2Node_DynamicCast}
   > Bind Event to Game Mode Message Request(self:Fort Game State Athena Object Reference=@n18, Delegate=Create Event()) {is=title K2Node_AddDelegate}
   > BindToAnimationFinished!(execute, self:Widget Animation Object Reference=Get Delay(), Widget:User Widget Object Reference=Self-Reference(), Delegate=@n4) -> (then)
   > BindToAnimationFinished(self=Get Outro(), Widget=Self-Reference(), Delegate=@n2)
   > SetVisibility(self=Get Overlay_Container(), InVisibility=Collapsed)
  ]}
 HUDToggle {nodes [
   n1 = HUDToggle!() -> (then, Hide:bool=~) {is=title K2Node_FunctionEntry}
   > n2 = Cast To FortPlayerControllerAthena(Object=GetOwningPlayer())
   > SetHUDElementVisibilityWithContext!(
    execute
    self:Fort Player Controller Object Reference=@n2
    HUDElementTags:"Gameplay Tag Container Structure (by ref)"=MakeGameplayTagContainerFromTag!(SingleTag:Gameplay Tag Structure=(TagName="HUD.Athena.Hide.Bottom.Left")) -> (ReturnValue:Gameplay Tag Container Structure)
    bHideElements:bool=@n1
    ReasonName:Name=None
    LogContext:String
   ) -> (then)
  ]}
]
variables [
 CanStartMessage {type=boolean, default=true}
 QueuedMessages {type="array<AthenaGameMessageData>", default []}
 QueuedMsg {
  type=AthenaGameMessageData
  default {bIsTeamBased=false, displayTime=5, messageFlags=None, msgDelay=0, msgMediaSource=null, msgSound=null, msgText="", msgType=None, teamIndex=0}
 }
]
widgets {
 info {inheritedWidgetCount=0, parentClass=&UserWidget, rootWidgetClass=&ScaleBox, widgetCount=14}
 widgets [
  ScaleBox_1 {widgetClassPath=&ScaleBox, widget=ScaleBox_1, properties {stretch=None}}
  Overlay_Container {
   widgetClassPath=&Overlay
   bIsVariable=true
   parent=ScaleBox_1
   widget=Overlay_Container
   properties {visibility=HitTestInvisible}
   slot {class=ScaleBoxSlot, properties {horizontalAlignment=HAlign_Left, verticalAlignment=VAlign_Bottom}}
  }
  Image_RadioBG {
   widgetClassPath=&Image
   parent=Overlay_Container
   widget=Image_RadioBG
   properties {
    brush {
     bIsDynamicallyLoaded=false
     drawAs=Image
     imageType=NoImage
     mirroring=NoMirror
     resourceName=None
     resourceObject=/Game/Athena/HUD/LTM/Heist/Materials/M_UI_HeistRadio_BG_Inst
     tiling=NoTile
     imageSize=<307.1884 277.3373>
     margin=<ltrb 0 0 0 0>
     outlineSettings {
      bUseBrushTransparency=false
      roundingType=HalfHeightRadius
      width=0
      color {colorUseRule=UseColor_Specified, specifiedColor=<rgba 0 0 0 0>}
      cornerRadii=<0 0 0 0>
     }
     tintColor {colorUseRule=UseColor_Specified, specifiedColor=<rgba 1 1 1 1>}
     uVRegion {bIsValid=false, max=<0 0>, min=<0 0>}
    }
    renderTransformPivot=<0 1>
   }
   slot {class=OverlaySlot, properties {verticalAlignment=VAlign_Bottom}}
  }
  Image_Radio {
   widgetClassPath=&Image
   bIsVariable=true
   parent=Overlay_Container
   widget=Image_Radio
   properties {
    brush {
     bIsDynamicallyLoaded=false
     drawAs=Image
     imageType=NoImage
     mirroring=NoMirror
     resourceName=None
     resourceObject=/Game/Athena/HUD/LTM/Heist/Materials/M_UI_RadioImage_Inst
     tiling=NoTile
     imageSize=<305 472>
     margin=<ltrb 0 0 0 0>
     outlineSettings {
      bUseBrushTransparency=false
      roundingType=HalfHeightRadius
      width=0
      color {colorUseRule=UseColor_Specified, specifiedColor=<rgba 0 0 0 0>}
      cornerRadii=<0 0 0 0>
     }
     tintColor {colorUseRule=UseColor_Specified, specifiedColor=<rgba 1 1 1 1>}
     uVRegion {bIsValid=false, max=<0 0>, min=<0 0>}
    }
    renderTransform {angle=0, scale=<0.6 0.6>, shear=<0 0>, translation=<63 -50>}
    renderTransformPivot=<0 1>
   }
   slot {class=OverlaySlot, properties {verticalAlignment=VAlign_Bottom}}
  }
  OverlayCallout {
   widgetClassPath=&Overlay
   parent=Overlay_Container
   widget=OverlayCallout
   properties {renderOpacity=0, renderTransform {angle=0, scale=<1 1>, shear=<0 0>, translation=<199 -156>}, renderTransformPivot=<-0.5317 0.7439>}
   slot {class=OverlaySlot, properties {verticalAlignment=VAlign_Bottom}}
  }
  Image_CalloutShadow {
   widgetClassPath=&Image
   parent=OverlayCallout
   widget=Image_CalloutShadow
   properties {
    brush {
     bIsDynamicallyLoaded=false
     drawAs=Image
     imageType=NoImage
     mirroring=NoMirror
     resourceName=None
     resourceObject=/Game/Athena/HUD/LTM/Heist/Materials/M_UI_Height_RadioCallouts_Inst
     tiling=NoTile
     imageSize=<109 215>
     margin=<ltrb 0 0 0 0>
     outlineSettings {
      bUseBrushTransparency=false
      roundingType=HalfHeightRadius
      width=0
      color {colorUseRule=UseColor_Specified, specifiedColor=<rgba 0 0 0 0>}
      cornerRadii=<0 0 0 0>
     }
     tintColor {colorUseRule=UseColor_Specified, specifiedColor=<rgba 0.0052 0.0052 0.0056 0.384>}
     uVRegion {bIsValid=false, max=<0 0>, min=<0 0>}
    }
    renderTransform {angle=0, scale=<1 1>, shear=<0 0>, translation=<6 6>}
   }
   slot {class=OverlaySlot}
  }
  Image_Callout {
   widgetClassPath=&Image
   parent=OverlayCallout
   widget=Image_Callout
   properties {brush {
     bIsDynamicallyLoaded=false
     drawAs=Image
     imageType=NoImage
     mirroring=NoMirror
     resourceName=None
     resourceObject=&M_UI_Height_RadioCallouts_Inst
     tiling=NoTile
     imageSize=<109 215>
     margin=<ltrb 0 0 0 0>
     outlineSettings {
      bUseBrushTransparency=false
      roundingType=HalfHeightRadius
      width=0
      color {colorUseRule=UseColor_Specified, specifiedColor=<rgba 0 0 0 0>}
      cornerRadii=<0 0 0 0>
     }
     tintColor {colorUseRule=UseColor_Specified, specifiedColor=<rgba 0.2423 1 0.9647 1>}
     uVRegion {bIsValid=false, max=<0 0>, min=<0 0>}
    }}
   slot {class=OverlaySlot}
  }
  Overlay_Dialog {
   widgetClassPath=&Overlay
   bIsVariable=true
   parent=Overlay_Container
   widget=Overlay_Dialog
   properties {renderTransform {angle=0, scale=<1 1>, shear=<0 0>, translation=<304 -190>}, renderTransformPivot=<0 0.6439>}
   slot {class=OverlaySlot, properties {verticalAlignment=VAlign_Bottom}}
  }
  Overlay_1 {
   widgetClassPath=&Overlay
   parent=Overlay_Dialog
   widget=Overlay_1
   slot {class=OverlaySlot, properties {horizontalAlignment=HAlign_Fill, verticalAlignment=VAlign_Fill}}
  }
  Image_DialogTail {
   widgetClassPath=&Image
   parent=Overlay_1
   widget=Image_DialogTail
   properties {
    visibility=HitTestInvisible
    brush {
     bIsDynamicallyLoaded=false
     drawAs=Image
     imageType=NoImage
     mirroring=NoMirror
     resourceName=None
     resourceObject=/Game/Athena/HUD/LTM/Heist/Materials/M_UI_RadioDialogBG3_Tail
     tiling=NoTile
     imageSize=<130.4032 99.9464>
     margin=<ltrb 0 0 0 0>
     outlineSettings {
      bUseBrushTransparency=false
      roundingType=HalfHeightRadius
      width=0
      color {colorUseRule=UseColor_Specified, specifiedColor=<rgba 0 0 0 0>}
      cornerRadii=<0 0 0 0>
     }
     tintColor {colorUseRule=UseColor_Specified, specifiedColor=<rgba 1 1 1 1>}
     uVRegion {bIsValid=false, max=<0 0>, min=<0 0>}
    }
    renderTransform {angle=0, scale=<1 1>, shear=<0 0>, translation=<-71 -45>}
    renderTransformPivot=<1 0.5>
   }
   slot {class=OverlaySlot, properties {verticalAlignment=VAlign_Bottom}}
  }
  Image_dialogBG {
   widgetClassPath=&Image
   parent=Overlay_1
   widget=Image_dialogBG
   properties {
    visibility=HitTestInvisible
    brush {
     bIsDynamicallyLoaded=false
     drawAs=Image
     imageType=NoImage
     mirroring=NoMirror
     resourceName=None
     resourceObject=/Game/Athena/HUD/LTM/Heist/Materials/M_UI_RadioDialogBG3_Inst
     tiling=NoTile
     imageSize=<256 128>
     margin=<ltrb 0 0.1358 0.1 0.2275>
     outlineSettings {
      bUseBrushTransparency=false
      roundingType=HalfHeightRadius
      width=0
      color {colorUseRule=UseColor_Specified, specifiedColor=<rgba 0 0 0 0>}
      cornerRadii=<0 0 0 0>
     }
     tintColor {colorUseRule=UseColor_Specified, specifiedColor=<rgba 1 1 1 1>}
     uVRegion {bIsValid=false, max=<0 0>, min=<0 0>}
    }
   }
   slot {class=OverlaySlot, properties {horizontalAlignment=HAlign_Fill, verticalAlignment=VAlign_Fill, padding=<ltrb 0 -40 -80 -80>}}
  }
  Border_1 {
   widgetClassPath=&Border
   parent=Overlay_1
   widget=Border_1
   properties {
    background {
     bIsDynamicallyLoaded=false
     drawAs=NoDrawType
     imageType=NoImage
     mirroring=NoMirror
     resourceName=None
     resourceObject=null
     tiling=NoTile
     imageSize=<32 32>
     margin=<ltrb 0 0 0 0>
     outlineSettings {
      bUseBrushTransparency=false
      roundingType=HalfHeightRadius
      width=0
      color {colorUseRule=UseColor_Specified, specifiedColor=<rgba 0 0 0 0>}
      cornerRadii=<0 0 0 0>
     }
     tintColor {colorUseRule=UseColor_Specified, specifiedColor=<rgba 1 1 1 1>}
     uVRegion {bIsValid=false, max=<0 0>, min=<0 0>}
    }
    padding=<ltrb 0 0 0 0>
   }
   slot {class=OverlaySlot, properties {padding=<ltrb 50 50 50 50>}}
  }
  SizeBox_Message {
   widgetClassPath=&SizeBox
   parent=Border_1
   widget=SizeBox_Message
   properties {maxDesiredWidth=360, renderTransformPivot=<0.5 1>}
   slot {class=BorderSlot, properties {padding=<ltrb 0 0 0 0>}}
  }
  RichTextBlock_Message {
   widgetClassPath=/Script/UMG.RichTextBlock
   bIsVariable=true
   parent=SizeBox_Message
   widget=RichTextBlock_Message
   properties {
    justification=Center
    lineHeightPercentage=0.9
    text="AN <enemy>ENEMY</> HAS ACQUIRED A <keyword>JEWEL</>!"
    textStyleSet=&/AthenaHeistRadioStyles
    wrapTextAt=360
    renderTransform {angle=0, scale=<1 1>, shear=<-12 0>, translation=<0 0>}
   }
   slot {class=SizeBoxSlot}
  }
 ]
}
