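# Playground minigame HUD widget that warns the player they are leaving the playspace and counts down the time remaining.
# OnDisplayLeavingVolumeUI, given a valid minigame volume, either shows the warning or plays the radio Outro and clears the timer.
# Showing sends MessageData to the embedded HeistHUD_Radio and starts a five-second countdown on a looping one-second timer.
# Each tick rewrites the radio's rich text through FormatTextForTimer, which appends the highlight-tagged time.
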WidgetBlueprint /Game/Athena/HUD/LTM/Playground/MinigameHUD_LeavingTheArea
graphs [
 EventGraph {nodes [
   n1 = Update Countdown Display() {is=title K2Node_CustomEvent}
   > n2 = Set TimeToDisplay(TimeToDisplay:int=int - int(A=Get TimeToDisplay() -> (TimeToDisplay:int=0), B=1) {is=title K2Node_CallFunction}) -> (Output_Get:int=0)
   > SetText(
    self=Get RichTextBlock_Message!(self:Heist HUD Radio Object Reference=Get HeistHUD_Radio() -> (HeistHUD_Radio:Heist HUD Radio Object Reference)) -> (RichTextBlock_Message:Rich Text Block Object Reference)
    InText=Format Text for Timer!(Time:int=@n2) -> (Text:Text) {is=title K2Node_MacroInstance}
   )
   n3 = Event OnDisplayLeavingVolumeUI!() -> (OutputDelegate, then, bDisplay:bool) {is=title K2Node_Event}
   n4 = Get CurrentMinigame() -> (CurrentMinigame:Fort Minigame Object Reference)
   > Is Valid(InputObject=@n4)
   > Is Valid(InputObject=Get Volume!(self:Fort Minigame Object Reference=@n4) -> (Volume:Actor Object Reference))
   > n5 = Branch(Condition=@n3.bDisplay)
   > SetVisibility(InVisibility=HitTestInvisible)
   > n6 = Set TimeToDisplay(TimeToDisplay=5) -> (Output_Get=0)
   n7 = Get HeistHUD_Radio()
   > OnMessageRequested!(
    execute
    self:Heist HUD Radio Object Reference=@n7
    MsgData:Athena Game Message Data Structure=Get MessageData() -> (MessageData:Athena Game Message Data Structure)
   ) -> (then)
   > SetText(self=Get RichTextBlock_Message(self=@n7), InText=Format Text for Timer(Time=@n6))
   > n8 = Set Timer by Event(Delegate=@n1, Time=1, bLooping=true)
   > Set TimerHandle(TimerHandle:Timer Handle Structure=@n8) -> (Output_Get:Timer Handle Structure)
   StopAllAnimations(execute=@n5.else, self:User Widget Object Reference=Get HeistHUD_Radio())
   n9 = Get HeistHUD_Radio()
   > PlayAnimation(self=@n9, InAnimation=Get Outro!(self:Heist HUD Radio Object Reference=@n9) -> (Outro:Widget Animation Object Reference))
   > Set TimeToDisplay(TimeToDisplay=5) -> (Output_Get=0)
   > Clear and Invalidate Timer by Handle(Handle=Get TimerHandle() -> (TimerHandle:Timer Handle Structure))
   Event PreConstruct()
  ]}
 FormatTextForTimer {nodes [
   n1 = Get MessageData()
   n2 = Get HighlightTag() -> (HighlightTag:String)
   Outputs!(Text:Text=Format Text!(
     Format="LOCTEXT(\"{Data}{HighlightTag}{time}</>\")"
     Data:Text=Break Athena Game Message Data!(AthenaGameMessageData:"Athena Game Message Data Structure (by ref)"=@n1) -> (
      MsgType:EAthenaGameMsgType Enum
      MsgText:Text
      MsgSound:Sound Cue Object Reference
      MsgMediaSource:Media Source Object Reference
      MsgDelay:float=~
      bIsTeamBased:bool=~
      TeamIndex:int
      DisplayTime:float=~
      MessageFlags:EAthenaGameMessageDataFlags Enum
      TargetPlayerController:Player Controller Object Reference
     ) {is=title K2Node_BreakStruct}.MsgText
     HighlightTag:Text="To Text (String)"(InString=@n2)
     time:int=Inputs!() -> (Time:int)
    ) -> (Result))
  ]}
]
variables [
 HighlightTag {type=string, default="<enemy>"}
 MessageData {
  type=AthenaGameMessageData
  default {
   bIsTeamBased=false
   displayTime=5
   messageFlags=None
   msgDelay=0
   msgMediaSource=null
   msgSound=/Game/Athena/Sounds/LTMs/Score/ScoreLTM_FinalCountdown_Warning_Cue
   msgText="<enemy>LEAVING</> THE <keyword>PLAYSPACE</>!\r\nTime Remaining: "
   msgType=CommMessage
   teamIndex=0
  }
 }
 TimerHandle {type=TimerHandle}
 TimeToDisplay {type=integer}
]
widgets {
 info {inheritedWidgetCount=1, parentClass=/Script/FortniteUI.MinigameWidgetBase, rootWidgetClass=&Overlay, widgetCount=3}
 widgets [
  Overlay_0 {widgetClassPath=&Overlay, widget=Overlay_0}
  HeistHUD_Radio {
   widgetClassPath=/Game/Athena/HUD/LTM/Heist/HeistHUD_Radio
   bIsVariable=true
   parent=Overlay_0
   widget=HeistHUD_Radio
   slot {class=OverlaySlot, properties {verticalAlignment=VAlign_Bottom}}
  }
  TutorialHighlight {widgetClassPath=&TutorialHighlight}
 ]
}
