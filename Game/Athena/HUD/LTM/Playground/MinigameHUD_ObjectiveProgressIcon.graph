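# Playground minigame HUD widget that shows one tracked destruction or capture objective as an icon with a progress-filled backplate.
# Initialization async-loads the preset icon and picks backplate materials and colors by objective type and friendly/enemy side.
# Dynamic material parameters drive fill, capture pulse direction and blocked state.
# The AnimateBarFill event track interpolates fill, while UMG animations show destruction damage hits and completion at zero percent.
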WidgetBlueprint /Game/Athena/HUD/LTM/Playground/MinigameHUD_ObjectiveProgressIcon
animations [
 AnimateBarFill {
  binding_count=0
  length_seconds=0.5
  object=&self:AnimateBarFill
  sequence {bindings [], tracks [{type=MovieSceneEventTrack, sections [{type=MovieSceneEventRepeaterSection, end_frame=10, start_frame=0}]}]}
 }
 Completed {
  binding_count=2
  length_seconds=1
  object=&self:Completed
  sequence {
   bindings [
    IndicatorIcon {
     id=#1
     possessed_class=Image
     tracks [
      {
       type=MovieSceneColorTrack
       display_name=Color and Opacity
       sections [{
         type=MovieSceneColorSection
         end_frame=60
         start_frame=0
         channels [
          Color.R {type=MovieSceneScriptingFloatChannel, thinned_from=5, keys [[0,1],[3,1],[24,0.2],[60.001,0.2]]}
          Color.G {type=MovieSceneScriptingFloatChannel, keys [[0,1],[3,1],[24,0.2],[60,0.2]]}
          Color.B {type=MovieSceneScriptingFloatChannel, keys [[0,1],[3,1],[24,0.2],[60,0.2]]}
          Color.A {type=MovieSceneScriptingFloatChannel, const [1,4]}
         ]
        }]
      }
      {
       type=MovieScene2DTransformTrack
       sections [{
         type=MovieScene2DTransformSection
         end_frame=60
         start_frame=0
         channels [
          Translation.X {type=MovieSceneScriptingFloatChannel, const [0,3]}
          Translation.Y {type=MovieSceneScriptingFloatChannel, keys [[0,0],[3,0],[9,-2.6],[15,4],[18,0],[60,0]]}
          Angle {type=MovieSceneScriptingFloatChannel, const [0,1]}
          Scale.X {type=MovieSceneScriptingFloatChannel, keys [[0,1],[3,1],[15,0.925],[18,1],[60,1]]}
          Scale.Y {type=MovieSceneScriptingFloatChannel, keys [[0,1],[3,1],[9,1.25],[15,0.8],[18,1],[60,1]]}
          Shear.X {type=MovieSceneScriptingFloatChannel, const [0,1]}
          Shear.Y {type=MovieSceneScriptingFloatChannel, const [0,1]}
         ]
        }]
      }
     ]
    }
    Image_BackplateProgress {
     id=#2
     possessed_class=Image
     tracks [
      {
       type=MovieSceneWidgetMaterialTrack
       sections [{
         type=MovieSceneParameterSection
         end_frame=60
         start_frame=0
         channels [
          UseMask {type=MovieSceneScriptingFloatChannel, keys [[0,0],[3,1],[60,1]]}
          AnimateLeftCenter {type=MovieSceneScriptingFloatChannel, keys [[0,0],[9,0],[15,1.75],[18,1],[60,1]]}
          AnimateRightCenter {type=MovieSceneScriptingFloatChannel, thinned_from=6, keys [[0,0],[13,0],[19,2.5],[22,1],[60,1]]}
          OverallAlpha {type=MovieSceneScriptingFloatChannel, thinned_from=4, keys [[0,1],[18,1],[45,0.5]]}
          UseBorder {type=MovieSceneScriptingFloatChannel, keys [[0,1],[3,1],[15,0]]}
         ]
        }]
      }
      {
       type=MovieScene2DTransformTrack
       sections [{
         type=MovieScene2DTransformSection
         end_frame=60
         start_frame=0
         channels [
          Translation.X {type=MovieSceneScriptingFloatChannel}
          Translation.Y {type=MovieSceneScriptingFloatChannel}
          Angle {type=MovieSceneScriptingFloatChannel}
          Scale.X {type=MovieSceneScriptingFloatChannel, thinned_from=5, keys [[0,1],[24,1],[30,0.95],[60,0.95]]}
          Scale.Y {type=MovieSceneScriptingFloatChannel, thinned_from=5, keys [[0,1],[24,1],[30,0.95],[60,0.95]]}
          Shear.X {type=MovieSceneScriptingFloatChannel}
          Shear.Y {type=MovieSceneScriptingFloatChannel}
         ]
        }]
      }
     ]
    }
   ]
   display_rate {numerator=60}
  }
 }
 FriendlyObjectiveDamaged {
  binding_count=2
  length_seconds=0.2
  object=&self:FriendlyObjectiveDamaged
  sequence {bindings [
    IndicatorIcon {
     id=#3
     possessed_class=Image
     tracks [
      {
       type=MovieSceneColorTrack
       display_name=Color and Opacity
       sections [{
         type=MovieSceneColorSection
         end_frame=4
         start_frame=0
         channels [
          Color.R {type=MovieSceneScriptingFloatChannel, keys [[0,1],[1,1.5],[3,0.2],[4,1]]}
          Color.G {type=MovieSceneScriptingFloatChannel, keys [[0,1],[1,1.5],[3,0.2],[4,1]]}
          Color.B {type=MovieSceneScriptingFloatChannel, keys [[0,1],[1,1.5],[3,0.2],[4,1]]}
          Color.A {type=MovieSceneScriptingFloatChannel, const [1,3]}
         ]
        }]
      }
      {
       type=MovieScene2DTransformTrack
       sections [{
         type=MovieScene2DTransformSection
         end_frame=4
         start_frame=0
         channels [
          Translation.X {type=MovieSceneScriptingFloatChannel, const [0,1]}
          Translation.Y {type=MovieSceneScriptingFloatChannel, const [0,1]}
          Angle {type=MovieSceneScriptingFloatChannel, const [0,1]}
          Scale.X {type=MovieSceneScriptingFloatChannel, keys [[0,1],[1,1.5],[3,0.75],[4,1]]}
          Scale.Y {type=MovieSceneScriptingFloatChannel, keys [[0,1],[1,1.5],[3,0.75],[4,1]]}
          Shear.X {type=MovieSceneScriptingFloatChannel, const [0,1]}
          Shear.Y {type=MovieSceneScriptingFloatChannel, const [0,1]}
         ]
        }]
      }
     ]
    }
    Image_BackplateProgress {
     id=#4
     possessed_class=Image
     tracks [{
       type=MovieScene2DTransformTrack
       sections [{
         type=MovieScene2DTransformSection
         end_frame=4
         start_frame=0
         channels [
          Translation.X {type=MovieSceneScriptingFloatChannel, const [0,1]}
          Translation.Y {type=MovieSceneScriptingFloatChannel, const [0,1]}
          Angle {type=MovieSceneScriptingFloatChannel, const [0,1]}
          Scale.X {type=MovieSceneScriptingFloatChannel, keys [[0,1],[1,1.25],[3,0.75],[4,1],[10,1]]}
          Scale.Y {type=MovieSceneScriptingFloatChannel, keys [[0,1],[1,1.25],[3,0.75],[4,1],[10,1]]}
          Shear.X {type=MovieSceneScriptingFloatChannel, const [0,1]}
          Shear.Y {type=MovieSceneScriptingFloatChannel, const [0,1]}
         ]
        }]
      }]
    }
   ]}
 }
 EnemyObjectiveDamaged {
  binding_count=2
  length_seconds=0.2
  object=&self:EnemyObjectiveDamaged
  sequence {bindings [
    Image_BackplateProgress {
     id=#4
     possessed_class=Image
     tracks [{
       type=MovieScene2DTransformTrack
       sections [{
         type=MovieScene2DTransformSection
         end_frame=4
         start_frame=0
         channels [
          Translation.X {type=MovieSceneScriptingFloatChannel, const [0,1]}
          Translation.Y {type=MovieSceneScriptingFloatChannel, const [0,1]}
          Angle {type=MovieSceneScriptingFloatChannel, const [0,1]}
          Scale.X {type=MovieSceneScriptingFloatChannel, keys [[0,1],[1,1.1],[3,0.8],[4,1],[10,1]]}
          Scale.Y {type=MovieSceneScriptingFloatChannel, keys [[0,1],[1,1.1],[3,0.8],[4,1],[10,1]]}
          Shear.X {type=MovieSceneScriptingFloatChannel, const [0,1]}
          Shear.Y {type=MovieSceneScriptingFloatChannel, const [0,1]}
         ]
        }]
      }]
    }
    IndicatorIcon {
     id=#5
     possessed_class=Image
     tracks [
      {
       type=MovieSceneColorTrack
       display_name=Color and Opacity
       sections [{
         type=MovieSceneColorSection
         end_frame=4
         start_frame=0
         channels [
          Color.R {type=MovieSceneScriptingFloatChannel, keys [[0,1],[1,1.5],[3,0.2],[4,1]]}
          Color.G {type=MovieSceneScriptingFloatChannel, keys [[0,1],[1,1.5],[3,0.2],[4,1]]}
          Color.B {type=MovieSceneScriptingFloatChannel, keys [[0,1],[1,1.5],[3,0.2],[4,1]]}
          Color.A {type=MovieSceneScriptingFloatChannel, const [1,3]}
         ]
        }]
      }
      {
       type=MovieScene2DTransformTrack
       sections [{
         type=MovieScene2DTransformSection
         end_frame=4
         start_frame=0
         channels [
          Translation.X {type=MovieSceneScriptingFloatChannel, const [0,1]}
          Translation.Y {type=MovieSceneScriptingFloatChannel, const [0,1]}
          Angle {type=MovieSceneScriptingFloatChannel, const [0,1]}
          Scale.X {type=MovieSceneScriptingFloatChannel, keys [[0,1],[1,1.25],[3,0.9],[4,1]]}
          Scale.Y {type=MovieSceneScriptingFloatChannel, keys [[0,1],[1,1.25],[3,0.75],[4,1]]}
          Shear.X {type=MovieSceneScriptingFloatChannel, const [0,1]}
          Shear.Y {type=MovieSceneScriptingFloatChannel, const [0,1]}
         ]
        }]
      }
     ]
    }
   ]}
 }
]
graphs [
 AnimateFulfilledState {nodes [AnimateFulfilledState() {is=title K2Node_FunctionEntry}]}
 CleanCaptureLockedState {nodes [
   CleanCaptureLockedState() {is=title K2Node_FunctionEntry}
   > n1 = GetDynamicMaterial(self=Get Image_BackplateProgress() -> (Image_BackplateProgress:Image Object Reference))
   > Set MID_Backplate(MID_Backplate:Material Instance Dynamic Object Reference=@n1) -> (Output_Get:Material Instance Dynamic Object Reference)
  ]}
 EventGraph {nodes [
   n1 = LoadIndicatorIcon!() -> (OutputDelegate, then, IconPresetIndex:int) {is=title K2Node_CustomEvent}
   > n2 = Async Load Asset(Asset="Get (a copy)"(
     Array:Array of Texture 2D Soft Object References=Get ObjectiveIndicatorIcons() -> (ObjectiveIndicatorIcons:Array of Texture 2D Soft Object References)
     Dimension 1=int + int(A=@n1.IconPresetIndex)
    ) -> (Output:Texture 2D Soft Object Reference))
   n3 = Cast To Texture2D(execute=@n2.Completed, Object=@n2)
   > SetBrush(
    self=Get IndicatorIcon() -> (IndicatorIcon:Image Object Reference)
    InBrush=MakeBrushFromTexture!(Texture:Texture 2D Object Reference=@n3, Width:int=24, Height:int=24) -> (ReturnValue:Slate Brush Structure)
   )
   > SetVisibility(self=Get IndicatorIcon(), InVisibility=SelfHitTestInvisible)
   K2Node_Event_1 = AnimateNewScore!() -> (OutputDelegate, then, NewScoreValue:double=~, bSkipAnimation:bool=~) {is=title K2Node_CustomEvent}
   > Set NewPercentage(NewPercentage:double=@K2Node_Event_1.NewScoreValue) -> (Output_Get:double)
   > n4 = Branch(Condition=@K2Node_Event_1.bSkipAnimation)
   > SetFillPercentage!(execute, self, Value:double=Get NewPercentage() -> (NewPercentage:double)) -> (then)
   PlayAnimation(execute=@n4.else, InAnimation=Get AnimateBarFill() -> (AnimateBarFill:Widget Animation Object Reference))
  ]}
 GetBarTickDuration {nodes [
   Inputs!()
   n1 = Get AnimateBarFill()
   Outputs!(ReturnValue:double=float - float(A=GetEndTime(self=@n1), B=GetStartTime!(self:Widget Animation Object Reference=@n1) -> (ReturnValue:float)) {is=title K2Node_CallFunction})
  ]}
 GetIntegralMin {nodes [
   n1 = Inputs!() -> (A:double=~, A2:double=~)
   n2 = "Min (Float)"(A:double="Absolute (Float)"(A=@n1), B:double="Absolute (Float)"(A=@n1.A2)) -> (ReturnValue:double) {is=title K2Node_CommutativeAssociativeBinaryOperator}
   Outputs(ReturnValue=int * float!(A:int=Ceil(A=@n2) {is=title K2Node_CallFunction}, B:double="Sign (Float)"!(A:double=@n1) -> (ReturnValue:double)) -> (ReturnValue:double))
  ]}
 GetScoreTickDuration {nodes [
   Inputs!()
   n1 = Get ScoreTickAnimation() -> (ScoreTickAnimation:Widget Animation Object Reference)
   Outputs(ReturnValue=float - float(A=GetEndTime(self=@n1), B=GetStartTime(self=@n1)))
  ]}
 InitializeIconFromObjective {nodes [
   n1 = InitializeIconFromObjective!() -> (then, Objective:Tracked Objective Structure, NewIsFriendly:bool=~) {is=title K2Node_FunctionEntry}
   > n2 = Sequence!(execute) -> (then_0, then_1, then_2:Exec, then_3:Exec, then_4:Exec)
   n3 = Break Tracked Objective!(TrackedObjective:"Tracked Objective Structure (by ref)"=@n1) -> (
    TrackedIndex:int
    TeamIndex:Byte
    OldTeamIndex:Byte
    IndicatorPresetIndex:int
    ObjectiveSortOrder:int
    ProgressPercentage:float=~
    TrackedState:Byte
    TrackedStateOwnerTeam:Byte
    OldTrackedState:Byte
    ObjectiveType:Byte
    bIsInteractionAllowed:bool=~
    bOldIsInteractionAllowed:bool=~
   ) {is=title K2Node_BreakStruct}
   > Set TrackedIndex(TrackedIndex:int=@n3) -> (Output_Get:int=0)
   n4 = Set IndicatorPresetIndex(execute=@n2.then_1, IndicatorPresetIndex:int=@n3.IndicatorPresetIndex) -> (Output_Get:int=0)
   > n5 = Branch(Condition="integer > integer"(A=@n4, B=-1))
   > LoadIndicatorIcon!(execute, self, IconPresetIndex:int=Get IndicatorPresetIndex() -> (IndicatorPresetIndex:int=0)) -> (then) {is=title K2Node_CallFunction}
   SetVisibility(execute=@n5.else, self=Get IndicatorIcon(), InVisibility=Hidden)
   UpdateFriendlyState!(execute=@n2.then_2, self, NewFriendly:bool=Get NewIsFriendly() -> (NewIsFriendly:bool)) -> (then)
   AnimateNewScore!(execute=@n2.then_4, self, NewScoreValue:double=@n3.ProgressPercentage, bSkipAnimation:bool=true) -> (then) {is=title K2Node_CallFunction}
   UpdateObjectiveType!(
    execute=@n2.then_3
    self
    ObjectiveType:EObjectiveType Enum=Byte to Enum EObjectiveType!(Byte:Byte=@n3.ObjectiveType) -> (ReturnValue:EObjectiveType Enum) {is=title K2Node_CastByteToEnum}
   ) -> (then)
   > UpdateObjectiveState!(
    execute
    self
    ObjectiveState:Byte=@n3.TrackedState
    OldObjectiveState:Byte=@n3.OldTrackedState
    TeamIndex:Byte=@n3.TeamIndex
    ObjectiveStateOwnerTeam:Byte=@n3.TrackedStateOwnerTeam
   ) -> (then)
  ]}
 ResetIconColors {nodes [
   ResetIconColors() {is=title K2Node_FunctionEntry}
   > n1 = GetDynamicMaterial(self=Get Image_BackplateProgress())
   > Set MID_Backplate(MID_Backplate=@n1)
   n2 = Get BackgroundColor_Friendly() -> (BackgroundColor_Friendly:Color Structure)
   n3 = Get BackgroundColor_Enemy() -> (BackgroundColor_Enemy:Color Structure)
   > SetVectorParameterValue(
    self=Get MID_Backplate() -> (MID_Backplate:Material Instance Dynamic Object Reference)
    ParameterName=Get MID_BackgroundColor() -> (MID_BackgroundColor:Name=None)
    Value=SelectColor(
     A="To LinearColor (Color)"!(InColor:Color Structure=@n2) -> (ReturnValue:Linear Color Structure)
     B="To LinearColor (Color)"(InColor=@n3)
     bPickA=Get bFriendly() -> (bFriendly:bool)
    )
   )
   n4 = Get FillColor_Friendly() -> (FillColor_Friendly:Color Structure)
   n5 = Get FillColor_Enemy() -> (FillColor_Enemy:Color Structure)
   > SetVectorParameterValue(
    self=Get MID_Backplate()
    ParameterName=Get MID_FillColor() -> (MID_FillColor:Name=None)
    Value=SelectColor(A="To LinearColor (Color)"(InColor=@n4), B="To LinearColor (Color)"(InColor=@n5), bPickA=Get bFriendly())
   )
  ]}
 SetFillPercentage {nodes [
   n1 = SetFillPercentage!() -> (then, Value:double=~) {is=title K2Node_FunctionEntry}
   > Set OriginalPercentage(OriginalPercentage:double=@n1) -> (Output_Get:double)
   > n2 = Set bCompleted(bCompleted:bool="float <= float"!(A:double=Get OriginalPercentage() -> (OriginalPercentage:double), B:double=0) -> (ReturnValue:bool) {is=title K2Node_CallFunction}) -> (Output_Get:bool)
   > n3 = Sequence!(execute) -> (then_0, then_1)
   n4 = Get Image_BackplateProgress()
   > Is Valid(InputObject=@n4)
   > n5 = GetDynamicMaterial(self=@n4)
   > Is Valid(InputObject=@n5)
   > SetScalarParameterValue(self=@n5, ParameterName=PercentFill, Value=Get OriginalPercentage())
   n6 = Select(
    Option 0:Widget Animation Object Reference=Get EnemyObjectiveDamaged() -> (EnemyObjectiveDamaged:Widget Animation Object Reference)
    Option 1:Widget Animation Object Reference=Get FriendlyObjectiveDamaged() -> (FriendlyObjectiveDamaged:Widget Animation Object Reference)
    Index=Get bFriendly()
   ) -> (ReturnValue:Widget Animation Object Reference)
   n7 = IsAnimationPlaying(InAnimation=@n6)
   n8 = Get bCompleted() -> (bCompleted:bool)
   n9 = Get Local_ObjectiveType() -> (Local_ObjectiveType:EObjectiveType Enum=DestructionObjective)
   > Branch(Condition=AND Boolean!(A=NOT Boolean(A=@n8), B=NOT Boolean(A=@n7), C:bool="Equal (Enum)"(A:EObjectiveType Enum=@n9, B:EObjectiveType Enum=DestructionObjective)) -> (ReturnValue))
   > PlayAnimation(InAnimation=@n6)
   n10 = Get Local_ObjectiveType() -> (Local_ObjectiveType=DestructionObjective)
   UpdatedCompletedState!(execute=@n3.then_1, self, bWantsCompleted:bool=AND Boolean!(A=@n2, B="Equal (Enum)"(A=@n10, B=DestructionObjective)) -> (ReturnValue)) -> (then)
  ]}
 SetIsInteractionAllowed {nodes [
   SetIsInteractionAllowed!() -> (then, bInIsAllowedInteraction:bool=~) {is=title K2Node_FunctionEntry}
   > n1 = GetDynamicMaterial(self=Get Image_BackplateProgress())
   > Set MID_Backplate(MID_Backplate=@n1)
   > n2 = Branch(Condition=Get bInIsAllowedInteraction() -> (bInIsAllowedInteraction:bool))
   > SetScalarParameterValue(self=Get MID_Backplate(), ParameterName=Get MID_BlockedParameter() -> (MID_BlockedParameter:Name=None), Value=0)
   > Branch(Condition="Not Equal (Integer)"!(A:int=Get IndicatorPresetIndex() -> (IndicatorPresetIndex=0), B:int=-1) -> (ReturnValue:bool) {is=title K2Node_CallFunction})
   > SetVisibility(self=Get IndicatorIcon(), InVisibility=SelfHitTestInvisible)
   SetScalarParameterValue(execute=@n2.else, self=Get MID_Backplate(), ParameterName=Get MID_BlockedParameter() -> (MID_BlockedParameter=None), Value=1)
   > ResetIconColors() {is=title K2Node_CallFunction}
   > Branch(Condition="Not Equal (Integer)"(A=Get IndicatorPresetIndex() -> (IndicatorPresetIndex=0), B=-1))
   > SetVisibility(self=Get IndicatorIcon(), InVisibility=Collapsed)
  ]}
 TickBarFill {nodes [
   TickBarFill() {is=title K2Node_FunctionEntry}
   n1 = Get NewPercentage()
   > n2 = Branch(Condition="Nearly Equal (Float)"(A=@n1, B=Get OriginalPercentage(), ErrorTolerance=0.01))
   > SetFillPercentage(Value=Get NewPercentage()) {is=title K2Node_CallFunction}
   n3 = float - float(A=@n1, B=Get OriginalPercentage())
   n4 = Get AnimateBarFill()
   n5 = "Clamp (Float)"(
    Value=float / float(
     A=GetAnimationCurrentTime!(self:User Widget Object Reference, InAnimation:Widget Animation Object Reference=@n4) -> (ReturnValue:float)
     B=Get Bar Tick Duration!() -> (ReturnValue:double=~) {is=title K2Node_MacroInstance}
    ) {is=title K2Node_CallFunction}
    Min=0
    Max=1
   )
   SetFillPercentage(
    execute=@n2.else
    Value=float + float(A=Get OriginalPercentage(), B=float * float(A=@n3, B=@n5)) {is=title K2Node_CommutativeAssociativeBinaryOperator}
   )
  ]}
 UpdateFriendlyState {nodes [
   n1 = UpdateFriendlyState!() -> (then, NewFriendly:bool=~) {is=title K2Node_FunctionEntry}
   > n2 = Set bFriendly(bFriendly:bool=@n1) -> (Output_Get:bool)
   > n3 = Branch(Condition=@n2)
   > SetContentColorAndOpacity!(
    execute
    self:Border Object Reference=Get Indicator_Color() -> (Indicator_Color:Border Object Reference)
    InContentColorAndOpacity:Linear Color Structure=(R=0.5029,G=1,B=1,A=1)
   ) -> (then)
   SetContentColorAndOpacity(execute=@n3.else, self=Get Indicator_Color(), InContentColorAndOpacity=(R=1,G=0.7913,B=0.5029,A=1))
  ]}
 UpdateInteractingIconColors {nodes [
   UpdateInteractingIconColors!() -> (then, ObjectiveState:Byte, ObjectiveStateOwner:Byte, ObjectiveOwner:Byte) {is=title K2Node_FunctionEntry}
   > n1 = GetDynamicMaterial(self=Get Image_BackplateProgress())
   > Set MID_Backplate(MID_Backplate=@n1)
   n2 = Byte to Enum ECaptureState!(Byte:Byte=Get ObjectiveState() -> (ObjectiveState:Byte=0)) -> (ReturnValue:ECaptureState Enum) {is=title K2Node_CastByteToEnum}
   > n3 = Branch(Condition=OR Boolean(A="Equal (Enum)"(A:ECaptureState Enum=@n2, B:ECaptureState Enum=Neutralizing), B="Equal (Enum)"(A=@n2, B=Deneutralizing)))
   n4 = Get BackgroundColor_Friendly()
   n5 = Get BackgroundColor_Enemy()
   > SetVectorParameterValue(
    self=Get MID_Backplate()
    ParameterName=Get MID_BackgroundColor() -> (MID_BackgroundColor=None)
    Value=SelectColor(A="To LinearColor (Color)"(InColor=@n4), B="To LinearColor (Color)"(InColor=@n5), bPickA=Get bFriendly())
   )
   n6 = Get FillColor_Friendly()
   n7 = Get FillColor_Enemy()
   n8 = GetTeamAffiliationToLocalPlayer!(Team:Byte=Get ObjectiveOwner() -> (ObjectiveOwner:Byte=0), bIncludeNonSpectators:bool=true, bIncludeSpectators:bool=false) -> (ReturnValue:EFortTeamAffiliation Enum=Friendly)
   > SetVectorParameterValue(
    self=Get MID_Backplate()
    ParameterName=Get MID_FillColor() -> (MID_FillColor=None)
    Value=SelectColor(
     A="To LinearColor (Color)"(InColor=@n6)
     B="To LinearColor (Color)"(InColor=@n7)
     bPickA="Equal (Enum)"(A:EFortTeamAffiliation Enum=@n8, B:EFortTeamAffiliation Enum=Friendly)
    )
   )
   n9 = Sequence(execute=@n3.else)
   > n10 = Branch(Condition="Equal (Byte)"!(
     A:Byte=Get ObjectiveOwner() -> (ObjectiveOwner=0)
     B:Byte=Literal enum EFortTeam!(Enum:EFortTeam Enum=Spectator) -> (ReturnValue:EFortTeam Enum) {is=title K2Node_EnumLiteral}
    ) -> (ReturnValue:bool))
   n11 = Get BackgroundColor_Friendly()
   n12 = Get BackgroundColor_Enemy()
   > SetVectorParameterValue(
    self=Get MID_Backplate()
    ParameterName=Get MID_BackgroundColor() -> (MID_BackgroundColor=None)
    Value=SelectColor(A="To LinearColor (Color)"(InColor=@n11), B="To LinearColor (Color)"(InColor=@n12), bPickA=Get bFriendly())
   )
   n13 = Get FillColor_Friendly()
   n14 = Get FillColor_Enemy()
   SetVectorParameterValue(
    execute=@n10.else
    self=Get MID_Backplate()
    ParameterName=Get MID_BackgroundColor() -> (MID_BackgroundColor=None)
    Value=SelectColor(A="To LinearColor (Color)"(InColor=@n13), B="To LinearColor (Color)"(InColor=@n14), bPickA=Get bFriendly())
   )
   n15 = Get FillColor_Friendly()
   n16 = Get FillColor_Enemy()
   n17 = GetTeamAffiliationToLocalPlayer(Team=Get ObjectiveStateOwner() -> (ObjectiveStateOwner:Byte=0), bIncludeNonSpectators=true, bIncludeSpectators=false) -> (ReturnValue=Friendly)
   SetVectorParameterValue(
    execute=@n9.then_1
    self=Get MID_Backplate()
    ParameterName=Get MID_FillColor() -> (MID_FillColor=None)
    Value=SelectColor(A="To LinearColor (Color)"(InColor=@n15), B="To LinearColor (Color)"(InColor=@n16), bPickA="Equal (Enum)"(A=@n17, B=Friendly))
   )
  ]}
 UpdateObjectiveState {nodes [
   UpdateObjectiveState!() -> (then, ObjectiveState:Byte, OldObjectiveState:Byte, TeamIndex:Byte, ObjectiveStateOwnerTeam:Byte) {is=title K2Node_FunctionEntry}
   > n1 = GetDynamicMaterial(self=Get Image_BackplateProgress())
   > Set MID_Backplate(MID_Backplate=@n1)
   n2 = Byte to Enum ECaptureState(Byte=Get ObjectiveState() -> (ObjectiveState=0))
   > n3 = Branch(Condition=OR Boolean!(
     A="Equal (Enum)"(A:ECaptureState Enum=@n2, B:ECaptureState Enum=Neutral)
     B="Equal (Enum)"(A=@n2, B=Captured)
     C:bool="Equal (Enum)"(A=@n2, B=Disabled)
    ) -> (ReturnValue))
   > SetScalarParameterValue(self=Get MID_Backplate(), ParameterName=Get MID_CaptureEffectParameter() -> (MID_CaptureEffectParameter:Name=None), Value=0)
   > ResetIconColors()
   n4 = Get ObjectiveState() -> (ObjectiveState=0)
   n5 = Branch(execute=@n3.else, Condition="Equal (Enum)"(A=Byte to Enum ECaptureState(Byte=@n4), B=Capturing))
   > n6 = SetScalarParameterValue(
    self=Get MID_Backplate()
    ParameterName=Get MID_CaptureEffectDirectionParameter() -> (MID_CaptureEffectDirectionParameter:Name=None)
    Value=0
   )
   SetScalarParameterValue(
    execute=[@n6.then, @n8.then]
    self=Get MID_Backplate()
    ParameterName=Get MID_CaptureEffectParameter() -> (MID_CaptureEffectParameter=None)
    Value=1
   )
   > UpdateInteractingIconColors!(
    execute
    self
    ObjectiveState:Byte=Get ObjectiveState() -> (ObjectiveState=0)
    ObjectiveStateOwner:Byte=Get ObjectiveStateOwnerTeam() -> (ObjectiveStateOwnerTeam:Byte=0)
    ObjectiveOwner:Byte=Get TeamIndex() -> (TeamIndex:Byte=0)
   ) -> (then) {is=title K2Node_CallFunction}
   n7 = Get ObjectiveState() -> (ObjectiveState=0)
   Branch(execute=@n5.else, Condition="Equal (Enum)"(A=Byte to Enum ECaptureState(Byte=@n7), B=Neutralizing))
   > n8 = SetScalarParameterValue(self=Get MID_Backplate(), ParameterName=Get MID_CaptureEffectDirectionParameter() -> (MID_CaptureEffectDirectionParameter=None), Value=1)
  ]}
 UpdateObjectiveType {nodes [
   n1 = UpdateObjectiveType!() -> (then, ObjectiveType:EObjectiveType Enum=DestructionObjective) {is=title K2Node_FunctionEntry}
   > Branch(Condition="Not Equal (Enum)"!(A:EObjectiveType Enum=@n1, B:EObjectiveType Enum=Get Local_ObjectiveType() -> (Local_ObjectiveType=DestructionObjective)) -> (ReturnValue:bool))
   > n2 = Set Local_ObjectiveType(Local_ObjectiveType:EObjectiveType Enum=@n1) -> (Output_Get:EObjectiveType Enum=DestructionObjective)
   > n3 = Switch on EObjectiveType!(execute, Selection:EObjectiveType Enum=@n2) -> (DestructionObjective:Exec, CaptureObjective:Exec) {is=title K2Node_SwitchEnum}
   > n4 = Branch(Condition=Get bFriendly())
   > SetBrushFromMaterial(self=Get Image_BackplateProgress(), Material=/Game/Athena/HUD/LTM/Playground/Materials/M_UI_HUD_Objective_Progress)
   SetBrushFromMaterial(execute=@n4.else, self=Get Image_BackplateProgress(), Material=/Game/Athena/HUD/LTM/Playground/Materials/M_UI_HUD_Objective_Progress_Enemy)
   n5 = Branch(execute=@n3.CaptureObjective, Condition=Get bFriendly())
   > SetBrushFromMaterial(self=Get Image_BackplateProgress(), Material=/Game/Athena/HUD/LTM/Playground/Materials/M_UI_HUD_CaptureObjective_Progress_Friendly)
   SetBrushFromMaterial(
    execute=@n5.else
    self=Get Image_BackplateProgress()
    Material=/Game/Athena/HUD/LTM/Playground/Materials/M_UI_HUD_CaptureObjective_Progress_Enemy
   )
  ]}
 UpdatedCompletedState {nodes [
   n1 = UpdatedCompletedState!() -> (then, bWantsCompleted:bool=~) {is=title K2Node_FunctionEntry}
   > Branch(Condition="Not Equal (Boolean)"!(A:bool=@n1, B:bool=Get bCompletedState() -> (bCompletedState:bool)) -> (ReturnValue:bool) {is=title K2Node_CallFunction})
   > n2 = Set bCompletedState(bCompletedState:bool=Get bWantsCompleted() -> (bWantsCompleted:bool)) -> (Output_Get:bool)
   > n3 = Branch(Condition=@n2)
   > PlayAnimation(InAnimation=Get Completed() -> (Completed:Widget Animation Object Reference))
   PlayAnimation(execute=@n3.else, InAnimation=Get Completed(), StartAtTime=GetEndTime(self=Get Completed()), PlayMode=Reverse)
  ]}
]
variables [
 BackgroundColor_Enemy {type=Color, default=<rgba 0.4667 0 0 1>}
 BackgroundColor_Friendly {type=Color, default=<rgba 0.051 0.1647 0.502 1>}
 bCompletedState {type=boolean}
 bFriendly {type=boolean, default=true}
 FillColor_Enemy {type=Color, default=<rgba 0.9451 0.3608 0.1569 1>}
 FillColor_Friendly {type=Color, default=<rgba 0 0.502 1 1>}
 IndicatorPresetIndex {type=integer}
 Local_ObjectiveType {type=EObjectiveType, default=DestructionObjective}
 MID_BackgroundColor {type=string, default=BackingColor}
 MID_BlockedParameter {type=string, default=BlockedState}
 MID_CaptureEffectDirectionParameter {type=string, default=InvertCapturePulseDirection}
 MID_CaptureEffectParameter {type=string, default=CapturePulseOpacity}
 MID_FillColor {type=string, default=FillColor}
 NewPercentage {type=number}
 ObjectiveIndicatorIcons {
  type="array<Texture2D>"
  default [
   /Game/Athena/UI/InGame/Creative/PropertyEditor/Icon_Picker/Textures/T_UI_IconPicker_NONE
   /Game/Athena/UI/InGame/Creative/PropertyEditor/Icon_Picker/Textures/64x64/T_UI_IconLibrary_A_64
   /Game/Athena/UI/InGame/Creative/PropertyEditor/Icon_Picker/Textures/64x64/T_UI_IconLibrary_B_64
   /Game/Athena/UI/InGame/Creative/PropertyEditor/Icon_Picker/Textures/64x64/T_UI_IconLibrary_C_64
   /Game/Athena/UI/InGame/Creative/PropertyEditor/Icon_Picker/Textures/64x64/T_UI_IconLibrary_D_64
   /Game/Athena/UI/InGame/Creative/PropertyEditor/Icon_Picker/Textures/64x64/T_UI_IconLibrary_W_64
   /Game/Athena/UI/InGame/Creative/PropertyEditor/Icon_Picker/Textures/64x64/T_UI_IconLibrary_X_64
   /Game/Athena/UI/InGame/Creative/PropertyEditor/Icon_Picker/Textures/64x64/T_UI_IconLibrary_Y_64
   /Game/Athena/UI/InGame/Creative/PropertyEditor/Icon_Picker/Textures/64x64/T_UI_IconLibrary_Z_64
   /Game/Athena/UI/InGame/Creative/PropertyEditor/Icon_Picker/Textures_V2/64x64/T_UI_IconLibrary_AMarker_64
   /Game/Athena/UI/InGame/Creative/PropertyEditor/Icon_Picker/Textures_V2/64x64/T_UI_IconLibrary_BMarker_64
   /Game/Athena/UI/InGame/Creative/PropertyEditor/Icon_Picker/Textures_V2/64x64/T_UI_IconLibrary_CMarker_64
   /Game/Athena/UI/InGame/Creative/PropertyEditor/Icon_Picker/Textures_V2/64x64/T_UI_IconLibrary_DMarker_64
   /Game/Athena/UI/InGame/Creative/PropertyEditor/Icon_Picker/Textures_V2/64x64/T_UI_IconLibrary_WMarker_64
   /Game/Athena/UI/InGame/Creative/PropertyEditor/Icon_Picker/Textures_V2/64x64/T_UI_IconLibrary_XMarker_64
   /Game/Athena/UI/InGame/Creative/PropertyEditor/Icon_Picker/Textures_V2/64x64/T_UI_IconLibrary_YMarker_64
   /Game/Athena/UI/InGame/Creative/PropertyEditor/Icon_Picker/Textures_V2/64x64/T_UI_IconLibrary_ZMarker_64
   /Game/Athena/UI/InGame/Creative/PropertyEditor/Icon_Picker/Textures/64x64/T_UI_IconLibrary_0_64
   /Game/Athena/UI/InGame/Creative/PropertyEditor/Icon_Picker/Textures/64x64/T_UI_IconLibrary_1_64
   /Game/Athena/UI/InGame/Creative/PropertyEditor/Icon_Picker/Textures/64x64/T_UI_IconLibrary_2_64
   /Game/Athena/UI/InGame/Creative/PropertyEditor/Icon_Picker/Textures/64x64/T_UI_IconLibrary_3_64
   /Game/Athena/UI/InGame/Creative/PropertyEditor/Icon_Picker/Textures/64x64/T_UI_IconLibrary_4_64
   /Game/Athena/UI/InGame/Creative/PropertyEditor/Icon_Picker/Textures/64x64/T_UI_IconLibrary_5_64
   /Game/Athena/UI/InGame/Creative/PropertyEditor/Icon_Picker/Textures/64x64/T_UI_IconLibrary_6_64
   /Game/Athena/UI/InGame/Creative/PropertyEditor/Icon_Picker/Textures/64x64/T_UI_IconLibrary_7_64
   /Game/Athena/UI/InGame/Creative/PropertyEditor/Icon_Picker/Textures/64x64/T_UI_IconLibrary_8_64
   /Game/Athena/UI/InGame/Creative/PropertyEditor/Icon_Picker/Textures/64x64/T_UI_IconLibrary_9_64
   /Game/Athena/UI/InGame/Creative/PropertyEditor/Icon_Picker/Textures/64x64/T_UI_IconLibrary_droplet_64
   /Game/Athena/UI/InGame/Creative/PropertyEditor/Icon_Picker/Textures/64x64/T_UI_IconLibrary_fishing_64
   /Game/Athena/UI/InGame/Creative/PropertyEditor/Icon_Picker/Textures/64x64/T_UI_IconLibrary_trophy_64
   /Game/Athena/UI/InGame/Creative/PropertyEditor/Icon_Picker/MilitaryTextures/T_64x64/T_UI_IconLibrary_pistol_64
   /Game/Athena/UI/InGame/Creative/PropertyEditor/Icon_Picker/MilitaryTextures/T_64x64/T_UI_IconLibrary_assaultrifle_64
   /Game/Athena/UI/InGame/Creative/PropertyEditor/Icon_Picker/MilitaryTextures/T_64x64/T_UI_IconLibrary_sniperrifle_64
   /Game/Athena/UI/InGame/Creative/PropertyEditor/Icon_Picker/MilitaryTextures/T_64x64/T_UI_IconLibrary_rocketlauncher_64
   /Game/Athena/UI/InGame/Creative/PropertyEditor/Icon_Picker/MilitaryTextures/T_64x64/T_UI_IconLibrary_grenade_64
   /Game/Athena/UI/InGame/Creative/PropertyEditor/Icon_Picker/Textures/64x64/T_UI_IconLibrary_wolf_64
   /Game/Athena/UI/InGame/Creative/PropertyEditor/Icon_Picker/Textures/64x64/T_UI_IconLibrary_raptor_64
   /Game/Athena/UI/InGame/Creative/PropertyEditor/Icon_Picker/Textures/64x64/T_UI_IconLibrary_chicken_64
   /Game/Athena/UI/InGame/Creative/PropertyEditor/Icon_Picker/Textures/64x64/T_UI_IconLibrary_frog_64
   /Game/Athena/UI/InGame/Creative/PropertyEditor/Icon_Picker/Textures/64x64/T_UI_IconLibrary_boar_64
   /Game/Athena/UI/InGame/Creative/PropertyEditor/Icon_Picker/Textures/64x64/T_UI_IconLibrary_fish_64
   /Game/Athena/UI/InGame/Creative/PropertyEditor/Icon_Picker/Textures/64x64/T_UI_IconLibrary_exclamation_64
   /Game/Athena/UI/InGame/Creative/PropertyEditor/Icon_Picker/Textures/64x64/T_UI_IconLibrary_checkmark_64
   /Game/Athena/UI/InGame/Creative/PropertyEditor/Icon_Picker/Textures/64x64/T_UI_IconLibrary_questionmark_64
   /Game/Athena/UI/InGame/Creative/PropertyEditor/Icon_Picker/Textures/64x64/T_UI_IconLibrary_coins_64
   /Game/Athena/UI/InGame/Creative/PropertyEditor/Icon_Picker/Textures/64x64/T_UI_IconLibrary_moneybag_64
   /Game/Athena/UI/InGame/Creative/PropertyEditor/Icon_Picker/Textures/64x64/T_UI_IconLibrary_heart_64
   /Game/Athena/UI/InGame/Creative/PropertyEditor/Icon_Picker/Textures/64x64/T_UI_IconLibrary_hand_64
   /Game/Athena/UI/InGame/Creative/PropertyEditor/Icon_Picker/Textures/64x64/T_UI_IconLibrary_fist_64
   /Game/Athena/UI/InGame/Creative/PropertyEditor/Icon_Picker/Textures/64x64/T_UI_IconLibrary_flame_64
   /Game/Athena/UI/InGame/Creative/PropertyEditor/Icon_Picker/Textures/64x64/T_UI_IconLibrary_radiation_64
   /Game/Athena/UI/InGame/Creative/PropertyEditor/Icon_Picker/Textures/64x64/T_UI_IconLibrary_sword_64
   /Game/Athena/UI/InGame/Creative/PropertyEditor/Icon_Picker/Textures/64x64/T_UI_IconLibrary_shield_64
   /Game/Athena/UI/InGame/Creative/PropertyEditor/Icon_Picker/Textures/64x64/T_UI_IconLibrary_shoppingcart_64
   /Game/Athena/UI/InGame/Creative/PropertyEditor/Icon_Picker/Textures/64x64/T_UI_IconLibrary_car_64
   /Game/Athena/UI/InGame/Creative/PropertyEditor/Icon_Picker/Textures/64x64/T_UI_IconLibrary_truck_64
   /Game/Athena/UI/InGame/Creative/PropertyEditor/Icon_Picker/Textures/64x64/T_UI_IconLibrary_helicopter_64
   /Game/Athena/UI/InGame/Creative/PropertyEditor/Icon_Picker/Textures/64x64/T_UI_IconLibrary_biplane_64
   /Game/Athena/UI/InGame/Creative/PropertyEditor/Icon_Picker/Textures/64x64/T_UI_IconLibrary_boat_64
   /Game/Athena/UI/InGame/Creative/PropertyEditor/Icon_Picker/Textures/64x64/T_UI_IconLibrary_package_64
   /Game/Athena/UI/InGame/Creative/PropertyEditor/Icon_Picker/MilitaryTextures/T_64x64/T_UI_IconLibrary_magazine_64
   /Game/Athena/UI/InGame/Creative/PropertyEditor/Icon_Picker/Textures/64x64/T_UI_IconLibrary_medkit_64
   /Game/Athena/UI/InGame/Creative/PropertyEditor/Icon_Picker/Textures/64x64/T_UI_IconLibrary_bed_64
   /Game/Athena/UI/InGame/Creative/PropertyEditor/Icon_Picker/Textures/64x64/T_UI_IconLibrary_garage_64
   /Game/Athena/UI/InGame/Creative/PropertyEditor/Icon_Picker/Textures/64x64/T_UI_IconLibrary_biohazard_64
   /Game/Athena/UI/InGame/Creative/PropertyEditor/Icon_Picker/Textures/64x64/T_UI_IconLibrary_recycle_64
   /Game/Athena/UI/InGame/Creative/PropertyEditor/Icon_Picker/Textures/64x64/T_UI_IconLibrary_signal_64
   /Game/Athena/UI/InGame/Creative/PropertyEditor/Icon_Picker/Textures/64x64/T_UI_IconLibrary_signpost_64
   /Game/Athena/UI/InGame/Creative/PropertyEditor/Icon_Picker/Textures/64x64/T_UI_IconLibrary_happy_64
   /Game/Athena/UI/InGame/Creative/PropertyEditor/Icon_Picker/Textures/64x64/T_UI_IconLibrary_sad_64
   /Game/Athena/UI/InGame/Creative/PropertyEditor/Icon_Picker/Textures/64x64/T_UI_IconLibrary_cool_64
   /Game/Athena/UI/InGame/Creative/PropertyEditor/Icon_Picker/Textures/64x64/T_UI_IconLibrary_surprised_64
   /Game/Athena/UI/InGame/Creative/PropertyEditor/Icon_Picker/Textures/64x64/T_UI_IconLibrary_tongueout_64
   /Game/Athena/UI/InGame/Creative/PropertyEditor/Icon_Picker/Textures/64x64/T_UI_IconLibrary_wink_64
   /Game/Athena/UI/InGame/Creative/PropertyEditor/Icon_Picker/Textures/64x64/T_UI_IconLibrary_crying_64
   /Game/Athena/UI/InGame/Creative/PropertyEditor/Icon_Picker/Textures/64x64/T_UI_IconLibrary_running_64
   /Game/Athena/UI/InGame/Creative/PropertyEditor/Icon_Picker/Textures/64x64/T_UI_IconLibrary_sprint_64
   /Game/Athena/UI/InGame/Creative/PropertyEditor/Icon_Picker/Textures/64x64/T_UI_IconLibrary_tree_64
   /Game/Athena/UI/InGame/Creative/PropertyEditor/Icon_Picker/Textures/64x64/T_UI_IconLibrary_forest_64
   /Game/Athena/UI/InGame/Creative/PropertyEditor/Icon_Picker/Textures/64x64/T_UI_IconLibrary_anchor_64
   /Game/Athena/UI/InGame/Creative/PropertyEditor/Icon_Picker/Textures/64x64/T_UI_IconLibrary_wheel_64
   /Game/Athena/UI/InGame/Creative/PropertyEditor/Icon_Picker/Textures/64x64/T_UI_IconLibrary_steeringwheel_64
   /Game/Athena/UI/InGame/Creative/PropertyEditor/Icon_Picker/Textures/64x64/T_UI_IconLibrary_speedometer_64
   /Game/Athena/UI/InGame/Creative/PropertyEditor/Icon_Picker/Textures/64x64/T_UI_IconLibrary_campfire_64
   /Game/Athena/UI/InGame/Creative/PropertyEditor/Icon_Picker/Textures/64x64/T_UI_IconLibrary_cave_64
   /Game/Athena/UI/InGame/Creative/PropertyEditor/Icon_Picker/Textures/64x64/T_UI_IconLibrary_house_64
   /Game/Athena/UI/InGame/Creative/PropertyEditor/Icon_Picker/Textures/64x64/T_UI_IconLibrary_information_64
   /Game/Athena/UI/InGame/Creative/PropertyEditor/Icon_Picker/Textures/64x64/T_UI_IconLibrary_medical_64
   /Game/Athena/UI/InGame/Creative/PropertyEditor/Icon_Picker/Textures/64x64/T_UI_IconLibrary_star_64
   /Game/Athena/UI/InGame/Creative/PropertyEditor/Icon_Picker/Textures/64x64/T_UI_IconLibrary_door_64
   /Game/Athena/UI/InGame/Creative/PropertyEditor/Icon_Picker/Textures/64x64/T_UI_IconLibrary_flag_64
   /Game/Athena/UI/InGame/Creative/PropertyEditor/Icon_Picker/Textures/64x64/T_UI_IconLibrary_chequeredflag_64
   /Game/Athena/UI/InGame/Creative/PropertyEditor/Icon_Picker/Textures/64x64/T_UI_IconLibrary_waypoint_64
   /Game/Athena/UI/InGame/Creative/PropertyEditor/Icon_Picker/Textures/64x64/T_UI_IconLibrary_marker_64
   /Game/Athena/UI/InGame/Creative/PropertyEditor/Icon_Picker/Textures/64x64/T_UI_IconLibrary_padlock_64
   /Game/Athena/UI/InGame/Creative/PropertyEditor/Icon_Picker/Textures/64x64/T_UI_IconLibrary_unlocked_64
   /Game/Athena/UI/InGame/Creative/PropertyEditor/Icon_Picker/Textures/64x64/T_UI_IconLibrary_key_64
   /Game/Athena/UI/InGame/Creative/PropertyEditor/Icon_Picker/Textures/64x64/T_UI_IconLibrary_gaspump_64
   /Game/Athena/UI/InGame/Creative/PropertyEditor/Icon_Picker/Textures/64x64/T_UI_IconLibrary_danger_64
   /Game/Athena/UI/InGame/Creative/PropertyEditor/Icon_Picker/Textures/64x64/T_UI_IconLibrary_enemy_64
   /Game/Athena/UI/InGame/Creative/PropertyEditor/Icon_Picker/Textures_V2/64x64/T_UI_IconLibrary_Pawprint_64
   /Game/Athena/UI/InGame/Creative/PropertyEditor/Icon_Picker/Textures_V2/64x64/T_UI_IconLibrary_PetFoodBowl_64
   /Game/Athena/UI/InGame/Creative/PropertyEditor/Icon_Picker/Textures_V2/64x64/T_UI_IconLibrary_LeftArrow_64
   /Game/Athena/UI/InGame/Creative/PropertyEditor/Icon_Picker/Textures_V2/64x64/T_UI_IconLibrary_RightArrow_64
   /Game/Athena/UI/InGame/Creative/PropertyEditor/Icon_Picker/Textures_V2/64x64/T_UI_IconLibrary_UpArrow_64
   /Game/Athena/UI/InGame/Creative/PropertyEditor/Icon_Picker/Textures_V2/64x64/T_UI_IconLibrary_DownArrow_64
   /Game/Athena/UI/InGame/Creative/PropertyEditor/Icon_Picker/Textures_V2/64x64/T_UI_IconLibrary_SpeechBubble_64
   /Game/Athena/UI/InGame/Creative/PropertyEditor/Icon_Picker/Textures_V2/64x64/T_UI_IconLibrary_List_64
   /Game/Athena/UI/InGame/Creative/PropertyEditor/Icon_Picker/Textures_V2/64x64/T_UI_IconLibrary_OfficeBuilding_64
   /Game/Athena/UI/InGame/Creative/PropertyEditor/Icon_Picker/Textures_V2/64x64/T_UI_IconLibrary_ApartmentBuilding_64
   /Game/Athena/UI/InGame/Creative/PropertyEditor/Icon_Picker/Textures_V2/64x64/T_UI_IconLibrary_Education_64
   /Game/Athena/UI/InGame/Creative/PropertyEditor/Icon_Picker/Textures_V2/64x64/T_UI_IconLibrary_Museum_64
   /Game/Athena/UI/InGame/Creative/PropertyEditor/Icon_Picker/Textures_V2/64x64/T_UI_IconLibrary_Briefcase_64
   /Game/Athena/UI/InGame/Creative/PropertyEditor/Icon_Picker/Textures_V2/64x64/T_UI_IconLibrary_Store_64
   /Game/Athena/UI/InGame/Creative/PropertyEditor/Icon_Picker/Textures_V2/64x64/T_UI_IconLibrary_Compass_64
   /Game/Athena/UI/InGame/Creative/PropertyEditor/Icon_Picker/Textures_V2/64x64/T_UI_IconLibrary_Alien_64
   /Game/Athena/UI/InGame/Creative/PropertyEditor/Icon_Picker/Textures_V2/64x64/T_UI_IconLibrary_MusicNote_64
   /Game/Athena/UI/InGame/Creative/PropertyEditor/Icon_Picker/Textures_V2/64x64/T_UI_IconLibrary_Binoculars_64
   /Game/Athena/UI/InGame/Creative/PropertyEditor/Icon_Picker/Textures_V2/64x64/T_UI_IconLibrary_UFO_64
   /Game/Athena/UI/InGame/Creative/PropertyEditor/Icon_Picker/Textures_V2/64x64/T_UI_IconLibrary_Football_64
   /Game/Athena/UI/InGame/Creative/PropertyEditor/Icon_Picker/Textures_V2/64x64/T_UI_IconLibrary_Arena_64
   /Game/Athena/UI/InGame/Creative/PropertyEditor/Icon_Picker/Textures_V2/64x64/T_UI_IconLibrary_Bridge_64
   /Game/Athena/UI/InGame/Creative/PropertyEditor/Icon_Picker/Textures_V2/64x64/T_UI_IconLibrary_Gravestone_64
   /Game/Athena/UI/InGame/Creative/PropertyEditor/Icon_Picker/Textures_V2/64x64/T_UI_IconLibrary_Hospital_64
   /Game/Athena/UI/InGame/Creative/PropertyEditor/Icon_Picker/Textures_V2/64x64/T_UI_IconLibrary_PoliceBadge_64
   /Game/Athena/UI/InGame/Creative/PropertyEditor/Icon_Picker/Textures_V2/64x64/T_UI_IconLibrary_Cutlery_64
   /Game/Athena/UI/InGame/Creative/PropertyEditor/Icon_Picker/Textures_V2/64x64/T_UI_IconLibrary_SheriffBadge_64
   /Game/Athena/UI/InGame/Creative/PropertyEditor/Icon_Picker/Textures_V2/64x64/T_UI_IconLibrary_Laptop_64
   /Game/Athena/UI/InGame/Creative/PropertyEditor/Icon_Picker/Textures_V2/64x64/T_UI_IconLibrary_Power_64
   /Game/Athena/UI/InGame/Creative/PropertyEditor/Icon_Picker/Textures_V2/64x64/T_UI_IconLibrary_Hammer_64
   /Game/Athena/UI/InGame/Creative/PropertyEditor/Icon_Picker/Textures_V2/64x64/T_UI_IconLibrary_Paintbrush_64
   /Game/Athena/UI/InGame/Creative/PropertyEditor/Icon_Picker/Textures_V2/64x64/T_UI_IconLibrary_Shovel_64
   /Game/Athena/UI/InGame/Creative/PropertyEditor/Icon_Picker/Textures_V2/64x64/T_UI_IconLibrary_Restaurant_64
   /Game/Athena/UI/InGame/Creative/PropertyEditor/Icon_Picker/Textures_V2/64x64/T_UI_IconLibrary_Gift_64
   /Game/Athena/UI/InGame/Creative/PropertyEditor/Icon_Picker/Textures_V2/64x64/T_UI_IconLibrary_Handshake_64
   /Game/Athena/UI/InGame/Creative/PropertyEditor/Icon_Picker/Textures_V2/64x64/T_UI_IconLibrary_Pin_64
   /Game/Athena/UI/InGame/Creative/PropertyEditor/Icon_Picker/Textures_V2/64x64/T_UI_IconLibrary_Camera_64
   /Game/Athena/UI/InGame/Creative/PropertyEditor/Icon_Picker/Textures_V2/64x64/T_UI_IconLibrary_PiggyBank_64
   /Game/Athena/UI/InGame/Creative/PropertyEditor/Icon_Picker/Textures_V2/64x64/T_UI_IconLibrary_BankNotes_64
   /Game/Athena/UI/InGame/Creative/PropertyEditor/Icon_Picker/Textures_V2/64x64/T_UI_IconLibrary_Burger_64
   /Game/Athena/UI/InGame/Creative/PropertyEditor/Icon_Picker/Textures_V2/64x64/T_UI_IconLibrary_Mushroom_64
   /Game/Athena/UI/InGame/Creative/PropertyEditor/Icon_Picker/Textures_V2/64x64/T_UI_IconLibrary_PizzaSlice_64
   /Game/Athena/UI/InGame/Creative/PropertyEditor/Icon_Picker/Textures_V2/64x64/T_UI_IconLibrary_GasMask_64
   /Game/Athena/UI/InGame/Creative/PropertyEditor/Icon_Picker/Textures_V2/64x64/T_UI_IconLibrary_Heartbeat_64.T_UI_IconLibrary_HeartBeat_64
   /Game/Athena/UI/InGame/Creative/PropertyEditor/Icon_Picker/Textures_V2/64x64/T_UI_IconLibrary_IceSkate_64
   /Game/Athena/UI/InGame/Creative/PropertyEditor/Icon_Picker/Textures_V2/64x64/T_UI_IconLibrary_SnowFlake_64.T_UI_IconLibrary_Snowflake_64
   /Game/Athena/UI/InGame/Creative/PropertyEditor/Icon_Picker/Textures_V2/64x64/T_UI_IconLibrary_Ghost_64
   /Game/Athena/UI/InGame/Creative/PropertyEditor/Icon_Picker/Textures_V2/64x64/T_UI_IconLibrary_Electrical_64
   /Game/Athena/UI/InGame/Creative/PropertyEditor/Icon_Picker/Textures_V2/64x64/T_UI_IconLibrary_CookingPot_64
   /Game/Athena/UI/InGame/Creative/PropertyEditor/Icon_Picker/Textures_V2/64x64/T_UI_IconLibrary_Cabin_64
   /Game/Athena/UI/InGame/Creative/PropertyEditor/Icon_Picker/Textures_V2/64x64/T_UI_IconLibrary_OfficePhone_64
   /Game/Athena/UI/InGame/Creative/PropertyEditor/Icon_Picker/Textures_V2/64x64/T_UI_IconLibrary_TV_64
   /Game/Athena/UI/InGame/Creative/PropertyEditor/Icon_Picker/Textures_V2/64x64/T_UI_IconLibrary_Gear_64
   /Game/Athena/UI/InGame/Creative/PropertyEditor/Icon_Picker/Textures_V2/64x64/T_UI_IconLibrary_Helix_64
   /Game/Athena/UI/InGame/Creative/PropertyEditor/Icon_Picker/Textures_V2/64x64/T_UI_IconLibrary_Zap_64
   /Game/Athena/UI/InGame/Creative/PropertyEditor/Icon_Picker/Textures_V2/64x64/T_UI_IconLibrary_Factory_64
   /Game/Athena/UI/InGame/Creative/PropertyEditor/Icon_Picker/Textures_V2/64x64/T_UI_IconLibrary_RadioTower_64
   /Game/Athena/UI/InGame/Creative/PropertyEditor/Icon_Picker/Textures_V2/64x64/T_UI_IconLibrary_Microscope_64
   /Game/Athena/UI/InGame/Creative/PropertyEditor/Icon_Picker/Textures_V2/64x64/T_UI_IconLibrary_SatelliteDish_64
   /Game/Athena/UI/InGame/Creative/PropertyEditor/Icon_Picker/Textures_V2/64x64/T_UI_IconLibrary_Map_64
   /Game/Athena/UI/InGame/Creative/PropertyEditor/Icon_Picker/Textures_V2/64x64/T_UI_IconLibrary_Play_64
   /Game/Athena/UI/InGame/Creative/PropertyEditor/Icon_Picker/Textures_V2/64x64/T_UI_IconLibrary_Stop_64
   /Game/Athena/UI/InGame/Creative/PropertyEditor/Icon_Picker/Textures_V2/64x64/T_UI_IconLibrary_Pause_64
   /Game/Athena/UI/InGame/Creative/PropertyEditor/Icon_Picker/Textures_V2/64x64/T_UI_IconLibrary_FastForward_64
   /Game/Athena/UI/InGame/Creative/PropertyEditor/Icon_Picker/Textures_V2/64x64/T_UI_IconLibrary_Rewind_64
   /Game/Athena/UI/InGame/Creative/PropertyEditor/Icon_Picker/Textures_V2/64x64/T_UI_IconLibrary_Volume_64
   /Game/Athena/UI/InGame/Creative/PropertyEditor/Icon_Picker/Textures_V2/64x64/T_UI_IconLibrary_Repeat_64
   /Game/Athena/UI/InGame/Creative/PropertyEditor/Icon_Picker/Textures_V2/64x64/T_UI_IconLibrary_Shuffle_64
   /Game/Athena/UI/InGame/Creative/PropertyEditor/Icon_Picker/Textures_V2/64x64/T_UI_IconLibrary_Chat_64
   /Game/Athena/UI/InGame/Creative/PropertyEditor/Icon_Picker/Textures_V2/64x64/T_UI_IconLibrary_Crown_64
   /Game/Athena/UI/InGame/Creative/PropertyEditor/Icon_Picker/Textures_V2/64x64/T_UI_IconLibrary_Eye_64
   /Game/Athena/UI/InGame/Creative/PropertyEditor/Icon_Picker/Textures_V2/64x64/T_UI_IconLibrary_Message_64
   /Game/Athena/UI/InGame/Creative/PropertyEditor/Icon_Picker/Textures_V2/64x64/T_UI_IconLibrary_LightBulb_64
   /Game/Athena/UI/InGame/Creative/PropertyEditor/Icon_Picker/Textures_V2/64x64/T_UI_IconLibrary_Bomb_64
   /Game/Athena/UI/InGame/Creative/PropertyEditor/Icon_Picker/Textures_V2/64x64/T_UI_IconLibrary_Cannon_64
   /Game/Athena/UI/InGame/Creative/PropertyEditor/Icon_Picker/MilitaryTextures/T_64x64/T_UI_IconLibrary_JetFighter_64
   /Game/Athena/UI/InGame/Creative/PropertyEditor/Icon_Picker/Textures_V2/64x64/T_UI_IconLibrary_Tent_64
   /Game/Athena/UI/InGame/Creative/PropertyEditor/Icon_Picker/Textures_V2/64x64/T_UI_IconLibrary_CCTV_64
   /Game/Athena/UI/InGame/Creative/PropertyEditor/Icon_Picker/Textures_V2/64x64/T_UI_IconLibrary_Radio_64
   /Game/Athena/UI/InGame/Creative/PropertyEditor/Icon_Picker/Textures_V2/64x64/T_UI_IconLibrary_Wifi_64
   /Game/Athena/UI/InGame/Creative/PropertyEditor/Icon_Picker/Textures_V2/64x64/T_UI_IconLibrary_Grass_64
   /Game/Athena/UI/InGame/Creative/PropertyEditor/Icon_Picker/Textures_V2/64x64/T_UI_IconLibrary_Leaf_64
   /Game/Athena/UI/InGame/Creative/PropertyEditor/Icon_Picker/Textures_V2/64x64/T_UI_IconLibrary_PalmTree_64
   /Game/Athena/UI/InGame/Creative/PropertyEditor/Icon_Picker/Textures_V2/64x64/T_UI_IconLibrary_Book_64
   /Game/Athena/UI/InGame/Creative/PropertyEditor/Icon_Picker/Textures_V2/64x64/T_UI_IconLibrary_Toolbox_64
   /Game/Athena/UI/InGame/Creative/PropertyEditor/Icon_Picker/Textures_V2/64x64/T_UI_IconLibrary_Bug_64
   /Game/Athena/UI/InGame/Creative/PropertyEditor/Icon_Picker/Textures_V2/64x64/T_UI_IconLibrary_FloppyDisc_64
   /Game/Athena/UI/InGame/Creative/PropertyEditor/Icon_Picker/Textures_V2/64x64/T_UI_IconLibrary_Cogs_64
   /Game/Athena/UI/InGame/Creative/PropertyEditor/Icon_Picker/Textures_V2/64x64/T_UI_IconLibrary_Clock_64
   /Game/Athena/UI/InGame/Creative/PropertyEditor/Icon_Picker/Textures_V2/64x64/T_UI_IconLibrary_TestTube_64
   /Game/Athena/UI/InGame/Creative/PropertyEditor/Icon_Picker/Textures_V2/64x64/T_UI_IconLibrary_Stopwatch_64
   /Game/Athena/UI/InGame/Creative/PropertyEditor/Icon_Picker/Textures_V2/64x64/T_UI_IconLibrary_Basketball_64
   /Game/Athena/UI/InGame/Creative/PropertyEditor/Icon_Picker/Textures_V2/64x64/T_UI_IconLibrary_BoxingGlove_64
   /Game/Athena/UI/InGame/Creative/PropertyEditor/Icon_Picker/Textures_V2/64x64/T_UI_IconLibrary_SoccerBall_64
   /Game/Athena/UI/InGame/Creative/PropertyEditor/Icon_Picker/Textures_V2/64x64/T_UI_IconLibrary_Roadblock_64
   /Game/Athena/UI/InGame/Creative/PropertyEditor/Icon_Picker/Textures_V2/64x64/T_UI_IconLibrary_Rocket_64
   /Game/Athena/UI/InGame/Creative/PropertyEditor/Icon_Picker/Textures_V2/64x64/T_UI_IconLibrary_Taxi_64
   /Game/Athena/UI/InGame/Creative/PropertyEditor/Icon_Picker/Textures_V2/64x64/T_UI_IconLibrary_Footprint_64
   /Game/Athena/UI/InGame/Creative/PropertyEditor/Icon_Picker/Textures_V2/64x64/T_UI_IconLibrary_Shoeprint_64
   /Game/Athena/UI/InGame/Creative/PropertyEditor/Icon_Picker/Textures_V2/64x64/T_UI_IconLibrary_Torch_64
   /Game/Athena/UI/InGame/Creative/PropertyEditor/Icon_Picker/Textures_V2/64x64/T_UI_IconLibrary_Axe_64
   /Game/Athena/UI/InGame/Creative/PropertyEditor/Icon_Picker/Textures_V2/64x64/T_UI_IconLibrary_Link_64
   /Game/Athena/UI/InGame/Creative/PropertyEditor/Icon_Picker/Textures_V2/64x64/T_UI_IconLibrary_Jigsaw_64
   /Game/Athena/UI/InGame/Creative/PropertyEditor/Icon_Picker/Textures_V2/64x64/T_UI_IconLibrary_Magnifier_64
   /Game/Athena/UI/InGame/Creative/PropertyEditor/Icon_Picker/Textures_V2/64x64/T_UI_IconLibrary_Wrench_64
   /Game/Athena/UI/InGame/Creative/PropertyEditor/Icon_Picker/Textures_V2/64x64/T_UI_IconLibrary_Input_64
   /Game/Athena/UI/InGame/Creative/PropertyEditor/Icon_Picker/Textures_V2/64x64/T_UI_IconLibrary_Calendar_64
   /Game/Athena/UI/InGame/Creative/PropertyEditor/Icon_Picker/Textures_V2/64x64/T_UI_IconLibrary_Projector_64
   /Game/Athena/UI/InGame/Creative/PropertyEditor/Icon_Picker/Textures_V2/64x64/T_UI_IconLibrary_Drama_64
   /Game/Athena/UI/InGame/Creative/PropertyEditor/Icon_Picker/Textures_V2/64x64/T_UI_IconLibrary_Hourglass_64
   /Game/Athena/UI/InGame/Creative/PropertyEditor/Icon_Picker/Textures_V2/64x64/T_UI_IconLibrary_Ticket_64
   /Game/Athena/UI/InGame/Creative/PropertyEditor/Icon_Picker/Textures_V2/64x64/T_UI_IconLibrary_Cactus_64
   /Game/Athena/UI/InGame/Creative/PropertyEditor/Icon_Picker/Textures_V2/64x64/T_UI_IconLibrary_JetPlane_64
   /Game/Athena/UI/InGame/Creative/PropertyEditor/Icon_Picker/Textures_V2/64x64/T_UI_IconLibrary_Railway_64
   /Game/Athena/UI/InGame/Creative/PropertyEditor/Icon_Picker/Textures_V2/64x64/T_UI_IconLibrary_Handcuffs_64
   /Game/Athena/UI/InGame/Creative/PropertyEditor/Icon_Picker/Textures_V2/64x64/T_UI_IconLibrary_Monument_64
   /Game/Athena/UI/InGame/Creative/PropertyEditor/Icon_Picker/Textures_V2/64x64/T_UI_IconLibrary_Boot_64
   /Game/Athena/UI/InGame/Creative/PropertyEditor/Icon_Picker/Textures_V2/64x64/T_UI_IconLibrary_Shirt_64
   /Game/Athena/UI/InGame/Creative/PropertyEditor/Icon_Picker/Textures_V2/64x64/T_UI_IconLibrary_Controller_64
   /Game/Athena/UI/InGame/Creative/PropertyEditor/Icon_Picker/Textures_V2/64x64/T_UI_IconLibrary_Keyboard_64
   /Game/Athena/UI/InGame/Creative/PropertyEditor/Icon_Picker/Textures_V2/64x64/T_UI_IconLibrary_Tape_64
   /Game/Athena/UI/InGame/Creative/PropertyEditor/Icon_Picker/Textures_V2/64x64/T_UI_IconLibrary_Plant_64
   /Game/Athena/UI/InGame/Creative/PropertyEditor/Icon_Picker/Textures_V2/64x64/T_UI_IconLibrary_WateringCan_64
   /Game/Athena/UI/InGame/Creative/PropertyEditor/Icon_Picker/Textures_V2/64x64/T_UI_IconLibrary_PaperClip_64
   /Game/Athena/UI/InGame/Creative/PropertyEditor/Icon_Picker/Textures_V2/64x64/T_UI_IconLibrary_PaintPot_64
   /Game/Athena/UI/InGame/Creative/PropertyEditor/Icon_Picker/Textures_V2/64x64/T_UI_IconLibrary_Clipboard_64
   /Game/Athena/UI/InGame/Creative/PropertyEditor/Icon_Picker/Textures_V2/64x64/T_UI_IconLibrary_TrashCan_64
   /Game/Athena/UI/InGame/Creative/PropertyEditor/Icon_Picker/Textures_V2/64x64/T_UI_IconLibrary_Scissors_64
   /Game/Athena/UI/InGame/Creative/PropertyEditor/Icon_Picker/Textures_V2/64x64/T_UI_IconLibrary_Pencil_64
   /Game/Athena/UI/InGame/Creative/PropertyEditor/Icon_Picker/Textures_V2/64x64/T_UI_IconLibrary_CoffeeCup_64
   /Game/Athena/UI/InGame/Creative/PropertyEditor/Icon_Picker/Textures_V2/64x64/T_UI_IconLibrary_CoffeeMug_64
   /Game/Athena/UI/InGame/Creative/PropertyEditor/Icon_Picker/Textures_V2/64x64/T_UI_IconLibrary_Cooker_64
   /Game/Athena/UI/InGame/Creative/PropertyEditor/Icon_Picker/Textures_V2/64x64/T_UI_IconLibrary_FrenchFries_64
   /Game/Athena/UI/InGame/Creative/PropertyEditor/Icon_Picker/Textures_V2/64x64/T_UI_IconLibrary_Fridge_64
   /Game/Athena/UI/InGame/Creative/PropertyEditor/Icon_Picker/Textures_V2/64x64/T_UI_IconLibrary_IceCream_64
   /Game/Athena/UI/InGame/Creative/PropertyEditor/Icon_Picker/Textures_V2/64x64/T_UI_IconLibrary_Microwave_64
   /Game/Athena/UI/InGame/Creative/PropertyEditor/Icon_Picker/Textures_V2/64x64/T_UI_IconLibrary_Cards_64
   /Game/Athena/UI/InGame/Creative/PropertyEditor/Icon_Picker/Textures_V2/64x64/T_UI_IconLibrary_Shamrock_64
   /Game/Athena/UI/InGame/Creative/PropertyEditor/Icon_Picker/Textures_V2/64x64/T_UI_IconLibrary_Dice_64
   /Game/Athena/UI/InGame/Creative/PropertyEditor/Icon_Picker/Textures_V2/64x64/T_UI_IconLibrary_ChessPiece_64
   /Game/Athena/UI/InGame/Creative/PropertyEditor/Icon_Picker/Textures_V2/64x64/T_UI_IconLibrary_Bell_64
   /Game/Athena/UI/InGame/Creative/PropertyEditor/Icon_Picker/Textures_V2/64x64/T_UI_IconLibrary_Candle_64
   /Game/Athena/UI/InGame/Creative/PropertyEditor/Icon_Picker/Textures_V2/64x64/T_UI_IconLibrary_Pumpkin_64
   /Game/Athena/UI/InGame/Creative/PropertyEditor/Icon_Picker/Textures_V2/64x64/T_UI_IconLibrary_Bat_64
   /Game/Athena/UI/InGame/Creative/PropertyEditor/Icon_Picker/Textures_V2/64x64/T_UI_IconLibrary_Sofa_64
   /Game/Athena/UI/InGame/Creative/PropertyEditor/Icon_Picker/Textures_V2/64x64/T_UI_IconLibrary_Tap_64
   /Game/Athena/UI/InGame/Creative/PropertyEditor/Icon_Picker/Textures_V2/64x64/T_UI_IconLibrary_Stairs_64
   /Game/Athena/UI/InGame/Creative/PropertyEditor/Icon_Picker/Textures_V2/64x64/T_UI_IconLibrary_Circuit_64
   /Game/Athena/UI/InGame/Creative/PropertyEditor/Icon_Picker/Textures_V2/64x64/T_UI_IconLibrary_Processor_64
   /Game/Athena/UI/InGame/Creative/PropertyEditor/Icon_Picker/Textures_V2/64x64/T_UI_IconLibrary_RobotArm_64
   /Game/Athena/UI/InGame/Creative/PropertyEditor/Icon_Picker/Textures_V2/64x64/T_UI_IconLibrary_Globe_64
   /Game/Athena/UI/InGame/Creative/PropertyEditor/Icon_Picker/Textures_V2/64x64/T_UI_IconLibrary_Satellite_64
   /Game/Athena/UI/InGame/Creative/PropertyEditor/Icon_Picker/Textures_V2/64x64/T_UI_IconLibrary_Atom_64
   /Game/Athena/UI/InGame/Creative/PropertyEditor/Icon_Picker/Textures_V2/64x64/T_UI_IconLibrary_Microphone_64
   /Game/Athena/UI/InGame/Creative/PropertyEditor/Icon_Picker/Textures_V2/64x64/T_UI_IconLibrary_TrebleClef_64
   /Game/Athena/UI/InGame/Creative/PropertyEditor/Icon_Picker/Textures_V2/64x64/T_UI_IconLibrary_Cloud_64
   /Game/Athena/UI/InGame/Creative/PropertyEditor/Icon_Picker/Textures_V2/64x64/T_UI_IconLibrary_Connected_64
   /Game/Athena/UI/InGame/Creative/PropertyEditor/Icon_Picker/Textures_V2/64x64/T_UI_IconLibrary_Disconnected_64
   /Game/Athena/UI/InGame/Creative/PropertyEditor/Icon_Picker/Textures_V2/64x64/T_UI_IconLibrary_Server_64
   /Game/Athena/UI/InGame/Creative/PropertyEditor/Icon_Picker/Textures_V2/64x64/T_UI_IconLibrary_Clapperboard_64
   /Game/Athena/UI/InGame/Creative/PropertyEditor/Icon_Picker/Textures_V2/64x64/T_UI_IconLibrary_Hashtag_64
   /Game/Athena/UI/InGame/Creative/PropertyEditor/Icon_Picker/Textures_V2/64x64/T_UI_IconLibrary_Infinity_64
   /Game/Athena/UI/InGame/Creative/PropertyEditor/Icon_Picker/Textures_V2/64x64/T_UI_IconLibrary_Parachute_64
   /Game/Athena/UI/InGame/Creative/PropertyEditor/Icon_Picker/Textures_V2/64x64/T_UI_IconLibrary_Bus_64
   /Game/Athena/UI/InGame/Creative/PropertyEditor/Icon_Picker/Textures_V2/64x64/T_UI_IconLibrary_BeachBall_64
   /Game/Athena/UI/InGame/Creative/PropertyEditor/Icon_Picker/Textures_V2/64x64/T_UI_IconLibrary_Umbrella_64
   /Game/Athena/UI/InGame/Creative/PropertyEditor/Icon_Picker/Textures_V2/64x64/T_UI_IconLibrary_Suitcase_64
   /Game/Athena/UI/InGame/Creative/PropertyEditor/Icon_Picker/Textures_V2/64x64/T_UI_IconLibrary_AlarmClock_64
   /Game/Athena/UI/InGame/Creative/PropertyEditor/Icon_Picker/Textures_V2/64x64/T_UI_IconLibrary_Calculator_64
   /Game/Athena/UI/InGame/Creative/PropertyEditor/Icon_Picker/Textures_V2/64x64/T_UI_IconLibrary_Raincloud_64
   /Game/Athena/UI/InGame/Creative/PropertyEditor/Icon_Picker/Textures_V2/64x64/T_UI_IconLibrary_SnowCloud_64.T_UI_IconLibrary_Snowcloud_64
   /Game/Athena/UI/InGame/Creative/PropertyEditor/Icon_Picker/Textures_V2/64x64/T_UI_IconLibrary_Stormcloud_64
   /Game/Athena/UI/InGame/Creative/PropertyEditor/Icon_Picker/Textures_V2/64x64/T_UI_IconLibrary_Sun_64
   /Game/Athena/UI/InGame/Creative/PropertyEditor/Icon_Picker/Textures_V2/64x64/T_UI_IconLibrary_Moon_64
   /Game/Athena/UI/InGame/Creative/PropertyEditor/Icon_Picker/Textures_V2/64x64/T_UI_IconLibrary_Tornado_64
   /Game/Athena/UI/InGame/Creative/PropertyEditor/Icon_Picker/Textures_V2/64x64/T_UI_IconLibrary_Flower_64
   /Game/Athena/UI/InGame/Creative/PropertyEditor/Icon_Picker/Textures_V2/64x64/T_UI_IconLibrary_Chest_64
   /Game/Athena/UI/InGame/Creative/PropertyEditor/Icon_Picker/Textures_V2/64x64/T_UI_IconLibrary_Palette_64
   /Game/Athena/UI/InGame/Creative/PropertyEditor/Icon_Picker/Textures_V2/64x64/T_UI_IconLibrary_Cheese_64
   /Game/Athena/UI/InGame/Creative/PropertyEditor/Icon_Picker/Textures_V2/64x64/T_UI_IconLibrary_BowlingPins_64
   /Game/Athena/UI/InGame/Creative/PropertyEditor/Icon_Picker/Textures_V2/64x64/T_UI_IconLibrary_BowlingBall_64
   /Game/Athena/UI/InGame/Creative/PropertyEditor/Icon_Picker/Textures_V2/64x64/T_UI_IconLibrary_Swimming_64
   /Game/Athena/UI/InGame/Creative/PropertyEditor/Icon_Picker/Textures_V2/64x64/T_UI_IconLibrary_BaseballBat_64
   /Game/Athena/UI/InGame/Creative/PropertyEditor/Icon_Picker/Textures_V2/64x64/T_UI_IconLibrary_Baseball_64
   /Game/Athena/UI/InGame/Creative/PropertyEditor/Icon_Picker/Textures_V2/64x64/T_UI_IconLibrary_VolleyBall_64.T_UI_IconLibrary_Volleyball_64
   /Game/Athena/UI/InGame/Creative/PropertyEditor/Icon_Picker/Textures_V2/64x64/T_UI_IconLibrary_Hockey_64
   /Game/Athena/UI/InGame/Creative/PropertyEditor/Icon_Picker/Textures_V2/64x64/T_UI_IconLibrary_Tennis_64
   /Game/Athena/UI/InGame/Creative/PropertyEditor/Icon_Picker/Textures_V2/64x64/T_UI_IconLibrary_Golf_64
   /Game/Athena/UI/InGame/Creative/PropertyEditor/Icon_Picker/Textures_V2/64x64/T_UI_IconLibrary_Archery_64
   /Game/Athena/UI/InGame/Creative/PropertyEditor/Icon_Picker/Textures_V2/64x64/T_UI_IconLibrary_Target_64
   /Game/Athena/UI/InGame/Creative/PropertyEditor/Icon_Picker/Textures_V2/64x64/T_UI_IconLibrary_Crosshair_64
   /Game/Athena/UI/InGame/Creative/PropertyEditor/Icon_Picker/Textures_V2/64x64/T_UI_IconLibrary_Balloon_64
   /Game/Athena/UI/InGame/Creative/PropertyEditor/Icon_Picker/Textures_V2/64x64/T_UI_IconLibrary_Quadcrasher_64.T_UI_IconLibrary_QuadCrasher_64
   /Game/Athena/UI/InGame/Creative/PropertyEditor/Icon_Picker/Textures_V2/64x64/T_UI_IconLibrary_GolfCart_64
   /Game/Athena/UI/InGame/Creative/PropertyEditor/Icon_Picker/Textures_V2/64x64/T_UI_IconLibrary_Driftboard_64.T_UI_IconLibrary_DriftBoard_64
   /Game/Athena/UI/InGame/Creative/PropertyEditor/Icon_Picker/Textures_V2/64x64/T_UI_IconLibrary_HamsterBall_64
   /Game/Athena/UI/InGame/Creative/PropertyEditor/Icon_Picker/Textures_V2/64x64/T_UI_IconLibrary_Motorboat_64
   /Game/Athena/UI/InGame/Creative/PropertyEditor/Icon_Picker/Textures_V2/64x64/T_UI_IconLibrary_Mech_64
   /Game/Athena/UI/InGame/Creative/PropertyEditor/Icon_Picker/Textures_V2/64x64/T_UI_IconLibrary_Pow_64
   /Game/Athena/UI/InGame/Creative/PropertyEditor/Icon_Picker/Textures_V2/64x64/T_UI_IconLibrary_Impact_64
   /Game/Athena/UI/InGame/Creative/PropertyEditor/Icon_Picker/Textures_V2/64x64/T_UI_IconLibrary_Mountains_64
   /Game/Athena/UI/InGame/Creative/PropertyEditor/Icon_Picker/Textures_V2/64x64/T_UI_IconLibrary_Waves_64
   /Game/Athena/UI/InGame/Creative/PropertyEditor/Icon_Picker/Textures_V2/64x64/T_UI_IconLibrary_Water_64
   /Game/Athena/UI/InGame/Creative/PropertyEditor/Icon_Picker/Textures_V2/64x64/T_UI_IconLibrary_Road_64
   /Game/Athena/UI/InGame/Creative/PropertyEditor/Icon_Picker/Textures_V2/64x64/T_UI_IconLibrary_Caution_64
   /Game/Athena/UI/InGame/Creative/PropertyEditor/Icon_Picker/Textures_V2/64x64/T_UI_IconLibrary_StarMarker_64
   /Game/Athena/UI/InGame/Creative/PropertyEditor/Icon_Picker/Textures_V2/64x64/T_UI_IconLibrary_RightLaneMerge_64
   /Game/Athena/UI/InGame/Creative/PropertyEditor/Icon_Picker/Textures_V2/64x64/T_UI_IconLibrary_LeftorRightTurn_64
   /Game/Athena/UI/InGame/Creative/PropertyEditor/Icon_Picker/Textures_V2/64x64/T_UI_IconLibrary_LeftCurve_64
   /Game/Athena/UI/InGame/Creative/PropertyEditor/Icon_Picker/Textures_V2/64x64/T_UI_IconLibrary_RightCurve_64
   /Game/Athena/UI/InGame/Creative/PropertyEditor/Icon_Picker/Textures_V2/64x64/T_UI_IconLibrary_LeftTurn_64
   /Game/Athena/UI/InGame/Creative/PropertyEditor/Icon_Picker/Textures_V2/64x64/T_UI_IconLibrary_RightTurn_64
   /Game/Athena/UI/InGame/Creative/PropertyEditor/Icon_Picker/Textures_V2/64x64/T_UI_IconLibrary_U-Turn_64
   /Game/Athena/UI/InGame/Creative/PropertyEditor/Icon_Picker/Textures_V2/64x64/T_UI_IconLibrary_NarrowLane_64
   /Game/Athena/UI/InGame/Creative/PropertyEditor/Icon_Picker/Textures_V2/64x64/T_UI_IconLibrary_NewLane_64
   /Game/Athena/UI/InGame/Creative/PropertyEditor/Icon_Picker/Textures_V2/64x64/T_UI_IconLibrary_Intersection_64
   /Game/Athena/UI/InGame/Creative/PropertyEditor/Icon_Picker/Textures_V2/64x64/T_UI_IconLibrary_People_64
   /Game/Athena/UI/InGame/Creative/PropertyEditor/Icon_Picker/Textures_V2/64x64/T_UI_IconLibrary_Bicycle_64
   /Game/Athena/UI/InGame/Creative/PropertyEditor/Icon_Picker/Textures_V2/64x64/T_UI_IconLibrary_Wood_64
   /Game/Athena/UI/InGame/Creative/PropertyEditor/Icon_Picker/Textures_V2/64x64/T_UI_IconLibrary_Stone_64
   /Game/Athena/UI/InGame/Creative/PropertyEditor/Icon_Picker/Textures_V2/64x64/T_UI_IconLibrary_Metal_64
   /Game/Athena/UI/InGame/Creative/PropertyEditor/Icon_Picker/Textures_V2/64x64/T_UI_IconLibrary_Design_64
   /Game/Athena/UI/InGame/Creative/PropertyEditor/Icon_Picker/Textures_V2/64x64/T_UI_IconLibrary_CandyBar_64
   /Game/Athena/UI/InGame/Creative/PropertyEditor/Icon_Picker/Textures_V2/64x64/T_UI_IconLibrary_Bookshelf_64
   /Game/Athena/UI/InGame/Creative/PropertyEditor/Icon_Picker/Textures_V2/64x64/T_UI_IconLibrary_Llama_64
   /Game/Athena/UI/InGame/Creative/PropertyEditor/Icon_Picker/Textures_V2/64x64/T_UI_IconLibrary_Saw_64
   /Game/Athena/UI/InGame/Creative/PropertyEditor/Icon_Picker/Textures_V2/64x64/T_UI_IconLibrary_Gnome_64
   /Game/Athena/UI/InGame/Creative/PropertyEditor/Icon_Picker/Textures_V2/64x64/T_UI_IconLibrary_Ham_64
   /Game/Athena/UI/InGame/Creative/PropertyEditor/Icon_Picker/Textures_V2/64x64/T_UI_IconLibrary_Fingerprint_64
   /Game/Athena/UI/InGame/Creative/PropertyEditor/Icon_Picker/Textures_V2/64x64/T_UI_IconLibrary_Hack_64
   /Game/Athena/UI/InGame/Creative/PropertyEditor/Icon_Picker/Textures_V2/64x64/T_UI_IconLibrary_Place_64
   /Game/Athena/UI/InGame/Creative/PropertyEditor/Icon_Picker/Textures_V2/64x64/T_UI_IconLibrary_Cube_64
   /Game/Athena/UI/InGame/Creative/PropertyEditor/Icon_Picker/Textures_V2/64x64/T_UI_IconLibrary_Alarm_64
   /Game/Athena/UI/InGame/Creative/PropertyEditor/Icon_Picker/Textures_V2/64x64/T_UI_IconLibrary_Comet_64
   /Game/Athena/UI/InGame/Creative/PropertyEditor/Icon_Picker/Textures_V2/64x64/T_UI_IconLibrary_Castle_64
   /Game/Athena/UI/InGame/Creative/PropertyEditor/Icon_Picker/Textures_V2/64x64/T_UI_IconLibrary_YinYang_64
   /Game/Athena/UI/InGame/Creative/PropertyEditor/Icon_Picker/Textures_V2/64x64/T_UI_IconLibrary_Helmet_64
   /Game/Athena/UI/InGame/Creative/PropertyEditor/Icon_Picker/Textures_V2/64x64/T_UI_IconLibrary_Jonesy_64
   /Game/Athena/UI/InGame/Creative/PropertyEditor/Icon_Picker/Textures_V2/64x64/T_UI_IconLibrary_Melee_64
   /Game/Athena/UI/InGame/Creative/PropertyEditor/Icon_Picker/Textures_V2/64x64/T_UI_IconLibrary_Medal_64
   /Game/Athena/UI/InGame/Creative/PropertyEditor/Icon_Picker/Textures_V2/64x64/T_UI_IconLibrary_Wing_64
   /Game/Athena/UI/InGame/Creative/PropertyEditor/Icon_Picker/Textures_V2/64x64/T_UI_IconLibrary_Swap_64
   /Game/Athena/UI/InGame/Creative/PropertyEditor/Icon_Picker/Textures_V2/64x64/T_UI_IconLibrary_Reset_64
   /Game/Athena/UI/InGame/Creative/PropertyEditor/Icon_Picker/Textures_V2/64x64/T_UI_IconLibrary_Shark_64
   /Game/Athena/UI/InGame/Creative/PropertyEditor/Icon_Picker/Textures_V2/64x64/T_UI_IconLibrary_FreezerChest_64
   /Game/Athena/UI/InGame/Creative/PropertyEditor/Icon_Picker/Textures_V2/64x64/T_UI_IconLibrary_PaperAirplane_64
   /Game/Athena/UI/InGame/Creative/PropertyEditor/Icon_Picker/Textures_V2/64x64/T_UI_IconLibrary_Crafting_64
   /Game/Athena/UI/InGame/Creative/PropertyEditor/Icon_Picker/Textures_V2/64x64/T_UI_IconLibrary_T-Rex_64
   /Game/Athena/UI/InGame/Creative/PropertyEditor/Icon_Picker/Textures_V2/64x64/T_UI_IconLibrary_Zipline_64.T_UI_IconLibrary_ZipLine_64
   /Game/Athena/UI/InGame/Creative/PropertyEditor/Icon_Picker/Textures_V2/64x64/T_UI_IconLibrary_Tree_64
   /Game/Athena/UI/InGame/Creative/PropertyEditor/Icon_Picker/Textures_V2/64x64/T_UI_IconLibrary_Diamond_64
   /Game/Athena/UI/InGame/Creative/PropertyEditor/Icon_Picker/Textures_V2/64x64/T_UI_IconLibrary_Directions_64
   /Game/Athena/UI/InGame/Creative/PropertyEditor/Icon_Picker/Textures_V2/64x64/T_UI_IconLibrary_Share_64
   /Game/Athena/UI/InGame/Creative/PropertyEditor/Icon_Picker/Textures_V2/64x64/T_UI_IconLibrary_SkiMask_64
   /Game/Athena/UI/InGame/Creative/PropertyEditor/Icon_Picker/Textures_V2/64x64/T_UI_IconLibrary_Folder_64
   /Game/Athena/UI/InGame/Creative/PropertyEditor/Icon_Picker/Textures_V2/64x64/T_UI_IconLibrary_Muscle_64
   /Game/Athena/UI/InGame/Creative/PropertyEditor/Icon_Picker/Textures_V2/64x64/T_UI_IconLibrary_Mask_64
   /Game/Athena/UI/InGame/Creative/PropertyEditor/Icon_Picker/Textures_V2/64x64/T_UI_IconLibrary_Lamp_64
   /Game/Athena/UI/InGame/Creative/PropertyEditor/Icon_Picker/Textures_V2/64x64/T_UI_IconLibrary_Incognito_64
   /Game/Athena/UI/InGame/Creative/PropertyEditor/Icon_Picker/Textures_V2/64x64/T_UI_IconLibrary_Radar_64
   /Game/Athena/UI/InGame/Creative/PropertyEditor/Icon_Picker/Textures_V2/64x64/T_UI_IconLibrary_ClothesHanger_64
   /Game/Athena/UI/InGame/Creative/PropertyEditor/Icon_Picker/Textures_V2/64x64/T_UI_IconLibrary_ShoppingBag_64
   /Game/Athena/UI/InGame/Creative/PropertyEditor/Icon_Picker/Textures_V2/64x64/T_UI_IconLibrary_Scenery_64
   /Game/Athena/UI/InGame/Creative/PropertyEditor/Icon_Picker/Textures_V2/64x64/T_UI_IconLibrary_DogTags_64
   /Game/Athena/UI/InGame/Creative/PropertyEditor/Icon_Picker/Textures_V2/64x64/T_UI_IconLibrary_Locker_64
   /Game/Athena/UI/InGame/Creative/PropertyEditor/Icon_Picker/Textures_V2/64x64/T_UI_IconLibrary_Bracket_64
  ]
 }
 OriginalPercentage {type=number}
 TrackedIndex {type=integer, default=-1}
]
widgets {
 info {inheritedWidgetCount=0, parentClass=&UserWidget, rootWidgetClass=&SizeBox, widgetCount=5}
 widgets [
  SizeBox_0 {
   widgetClassPath=&SizeBox
   widget=SizeBox_0
   properties {bOverride_HeightOverride=true, bOverride_WidthOverride=true, heightOverride=52, widthOverride=52}
  }
  Overlay_7 {widgetClassPath=&Overlay, parent=SizeBox_0, widget=Overlay_7, slot {class=SizeBoxSlot}}
  Image_BackplateProgress {
   widgetClassPath=&Image
   bIsVariable=true
   parent=Overlay_7
   widget=Image_BackplateProgress
   properties {brush {
     bIsDynamicallyLoaded=false
     drawAs=Image
     imageType=NoImage
     mirroring=NoMirror
     resourceName=None
     resourceObject=/Game/Athena/HUD/LTM/Playground/Materials/M_UI_HUD_Objective_Progress_Friendly
     tiling=NoTile
     imageSize=<64 64>
     margin=<ltrb 0.09 0.09 0.09 0.09>
     outlineSettings {
      bUseBrushTransparency=false
      roundingType=HalfHeightRadius
      width=0
      color {colorUseRule=UseColor_Specified, specifiedColor=<rgba 0 0 0 0>}
      cornerRadii=<0 0 0 0>
     }
     tintColor {colorUseRule=UseColor_Specified, specifiedColor=<rgba 1 1 1 1>}
     uVRegion {bIsValid=false, max=<0 0>, min=<0 0>}
    }}
   slot {class=OverlaySlot, properties {horizontalAlignment=HAlign_Fill, verticalAlignment=VAlign_Fill}}
  }
  Indicator_Color {
   widgetClassPath=&Border
   bIsVariable=true
   parent=Overlay_7
   widget=Indicator_Color
   properties {
    horizontalAlignment=HAlign_Center
    verticalAlignment=VAlign_Center
    background {
     bIsDynamicallyLoaded=false
     drawAs=NoDrawType
     imageType=NoImage
     mirroring=NoMirror
     resourceName=None
     resourceObject=null
     tiling=NoTile
     imageSize=<24 24>
     margin=<ltrb 0 0 0 0>
     outlineSettings {
      bUseBrushTransparency=false
      roundingType=HalfHeightRadius
      width=0
      color {colorUseRule=UseColor_Specified, specifiedColor=<rgba 0 0 0 0>}
      cornerRadii=<0 0 0 0>
     }
     tintColor {colorUseRule=UseColor_Specified, specifiedColor=<rgba 1 1 1 1>}
     uVRegion {bIsValid=false, max=<0 0>, min=<0 0>}
    }
    contentColorAndOpacity=<rgba 0.5 1 1 1>
    padding=<ltrb 0 0 0 0>
    renderTransform {angle=0, scale=<1 1>, shear=<0 0>, translation=<0 -3>}
   }
   slot {class=OverlaySlot, properties {horizontalAlignment=HAlign_Center, verticalAlignment=VAlign_Center}}
  }
  IndicatorIcon {
   widgetClassPath=&Image
   bIsVariable=true
   parent=Indicator_Color
   widget=IndicatorIcon
   properties {brush {
     bIsDynamicallyLoaded=false
     drawAs=Image
     imageType=NoImage
     mirroring=NoMirror
     resourceName=None
     resourceObject=/Game/Athena/UI/InGame/Creative/PropertyEditor/Icon_Picker/Textures/T_UI_IconPicker_EMPTY
     tiling=NoTile
     imageSize=<24 24>
     margin=<ltrb 0 0 0 0>
     outlineSettings {
      bUseBrushTransparency=false
      roundingType=HalfHeightRadius
      width=0
      color {colorUseRule=UseColor_Specified, specifiedColor=<rgba 0 0 0 0>}
      cornerRadii=<0 0 0 0>
     }
     tintColor {colorUseRule=UseColor_Specified, specifiedColor=<rgba 1 1 1 1>}
     uVRegion {bIsValid=false, max=<0 0>, min=<0 0>}
    }}
   slot {class=BorderSlot, properties {horizontalAlignment=HAlign_Center, verticalAlignment=VAlign_Center, padding=<ltrb 0 0 0 0>}}
  }
 ]
}
