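# Minigame HUD widget for Playground LTMs that shows friendly and enemy objectives as progress icons and remaining/total counters.
# While minigame UI displays, it binds to the objective tracker component's events and refreshes on team changes.
# Four icons per side are pooled in free arrays and index-keyed maps, moving between sides on ownership changes and reordered by indicator preset index.
# Numeric views drive material fills and reference damage pulse animations.
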WidgetBlueprint /Game/Athena/HUD/LTM/Playground/MinigameHUD_ObjectiveTracker
animations [
 FriendlyObjectiveDamaged {
  binding_count=1
  length_seconds=0.2
  object=&self:FriendlyObjectiveDamaged
  sequence {bindings [FriendlyBacking_NumericView {
     id=#1
     possessed_class=Image
     tracks [
      {
       type=MovieSceneColorTrack
       display_name=Color and Opacity
       sections [{
         type=MovieSceneColorSection
         end_frame=4
         start_frame=0
         channels [
          Color.R {type=MovieSceneScriptingFloatChannel, const [1,1]}
          Color.G {type=MovieSceneScriptingFloatChannel, const [1,1]}
          Color.B {type=MovieSceneScriptingFloatChannel, const [1,1]}
          Color.A {type=MovieSceneScriptingFloatChannel, keys [[0,1],[1,0.3],[3,1]]}
         ]
        }]
      }
      {
       type=MovieScene2DTransformTrack
       sections [{
         type=MovieScene2DTransformSection
         end_frame=4
         start_frame=0
         channels [
          Translation.X {type=MovieSceneScriptingFloatChannel, const [0,1]}
          Translation.Y {type=MovieSceneScriptingFloatChannel, const [0,1]}
          Angle {type=MovieSceneScriptingFloatChannel, const [0,1]}
          Scale.X {type=MovieSceneScriptingFloatChannel, keys [[0,1],[1,1.05],[3,0.95],[4,1],[10,1]]}
          Scale.Y {type=MovieSceneScriptingFloatChannel, keys [[0,1],[1,1.05],[3,0.95],[4,1],[10,1]]}
          Shear.X {type=MovieSceneScriptingFloatChannel, const [0,1]}
          Shear.Y {type=MovieSceneScriptingFloatChannel, const [0,1]}
         ]
        }]
      }
     ]
    }]}
 }
 EnemyObjectiveDamaged {
  binding_count=1
  length_seconds=0.2
  object=&self:EnemyObjectiveDamaged
  sequence {bindings [EnemyBacking_NumericView {
     id=#2
     possessed_class=Image
     tracks [
      {
       type=MovieSceneColorTrack
       display_name=Color and Opacity
       sections [{
         type=MovieSceneColorSection
         end_frame=4
         start_frame=0
         channels [
          Color.R {type=MovieSceneScriptingFloatChannel, const [1,1]}
          Color.G {type=MovieSceneScriptingFloatChannel, const [1,1]}
          Color.B {type=MovieSceneScriptingFloatChannel, const [1,1]}
          Color.A {type=MovieSceneScriptingFloatChannel, keys [[0,1],[1,0.3],[3,1],[4,1]]}
         ]
        }]
      }
      {
       type=MovieScene2DTransformTrack
       sections [{
         type=MovieScene2DTransformSection
         end_frame=4
         start_frame=0
         channels [
          Translation.X {type=MovieSceneScriptingFloatChannel, const [0,1]}
          Translation.Y {type=MovieSceneScriptingFloatChannel, const [0,1]}
          Angle {type=MovieSceneScriptingFloatChannel, const [0,1]}
          Scale.X {type=MovieSceneScriptingFloatChannel, keys [[0,1],[1,1.01],[3,0.98],[4,1]]}
          Scale.Y {type=MovieSceneScriptingFloatChannel, keys [[0,1],[1,1.01],[3,0.98],[4,1]]}
          Shear.X {type=MovieSceneScriptingFloatChannel, const [0,1]}
          Shear.Y {type=MovieSceneScriptingFloatChannel, const [0,1]}
         ]
        }]
      }
     ]
    }]}
 }
]
graphs [
 BindToObjectiveTracker {nodes [
   n1 = BindToObjectiveTracker!() -> (then, ObjectiveTracker:Fort Minigame Objective Tracker Component Object Reference) {is=title K2Node_FunctionEntry}
   > Is Valid(InputObject=@n1)
   > Bind Event to On Tracked Objective Changed(
    self:Fort Minigame Objective Tracker Component Object Reference=Get ObjectiveTracker() -> (ObjectiveTracker:Fort Minigame Objective Tracker Component Object Reference)
    Delegate=Create Event()
   ) {is=title K2Node_AddDelegate}
   > Bind Event to On Tracked Objective State Changed(self:Fort Minigame Objective Tracker Component Object Reference=Get ObjectiveTracker(), Delegate=Create Event()) {is=title K2Node_AddDelegate}
   > Bind Event to On Tracked Objective Team Updated(self:Fort Minigame Objective Tracker Component Object Reference=Get ObjectiveTracker(), Delegate=Create Event()) {is=title K2Node_AddDelegate}
   > Bind Event to On Tracked Objective Removed(self:Fort Minigame Objective Tracker Component Object Reference=Get ObjectiveTracker(), Delegate=Create Event()) {is=title K2Node_AddDelegate}
   > Bind Event to On Tracked Objective Added(self:Fort Minigame Objective Tracker Component Object Reference=Get ObjectiveTracker(), Delegate=Create Event()) {is=title K2Node_AddDelegate}
   > Bind Event to On Tracked Objective Allowed Interaction Changed(self:Fort Minigame Objective Tracker Component Object Reference=Get ObjectiveTracker(), Delegate=Create Event()) {is=title K2Node_AddDelegate}
  ]}
 ClaimIcon {nodes [
   n1 = ClaimIcon!() -> (then, ObjectiveInfo:"Tracked Objective Structure (by ref)") {is=title K2Node_FunctionEntry}
   n2 = Break Tracked Objective!(TrackedObjective:"Tracked Objective Structure (by ref)"=@n1) -> (
    TrackedIndex:int
    TeamIndex:Byte
    OldTeamIndex:Byte
    IndicatorPresetIndex:int
    ObjectiveSortOrder:int
    ProgressPercentage:float=~
    TrackedState:Byte
    TrackedStateOwnerTeam:Byte
    OldTrackedState:Byte
    ObjectiveType:Byte
    bIsInteractionAllowed:bool=~
    bOldIsInteractionAllowed:bool=~
   ) {is=title K2Node_BreakStruct}
   > Set TrackedIndex(TrackedIndex:int=@n2) -> (Output_Get:int=0)
   > Set bIsFriendly(bIsFriendly:bool="Equal (Byte)"!(A:Byte=@n2.TeamIndex, B:Byte=GetOwnerPlayerTeam!() -> (TeamIndex:Byte=0)) -> (ReturnValue:bool)) -> (Output_Get:bool)
   > n3 = Sequence!(execute) -> (then_0, then_1, then_2:Exec, then_3:Exec, then_4:Exec)
   n4 = Is Valid(exec=@n3.then_1, InputObject=Get ClaimedIcon() -> (ClaimedIcon:Minigame HUD Objective Progress Icon Object Reference))
   n5 = Find(
    TargetMap:Map of Integers to Minigame HUD Objective Progress Icon Object References=Get UsedEnemyIcons() -> (UsedEnemyIcons:Map of Integers to Minigame HUD Objective Progress Icon Object References)
    Key:"Integer (by ref)"=Get TrackedIndex() -> (TrackedIndex:int=0)
   ) -> (Value:Minigame HUD Objective Progress Icon Object Reference)
   Branch(execute=@n4.Is Not Valid, Condition=@n5.ReturnValue)
   > Set ClaimedIcon(ClaimedIcon:Minigame HUD Objective Progress Icon Object Reference=@n5) -> (Output_Get:Minigame HUD Objective Progress Icon Object Reference)
   > n6 = Branch(Condition=Get bIsFriendly() -> (bIsFriendly:bool))
   > FreeIcon!(execute, self, TrackedIndex:int=Get TrackedIndex() -> (TrackedIndex=0)) -> (then)
   > Set ClaimedIcon()
   Set bExistingIcon(execute=@n6.else, bExistingIcon:bool=true) -> (Output_Get:bool)
   n7 = Is Valid(exec=@n3.then_2, InputObject=Get ClaimedIcon())
   n8 = Select(
    Option 0:Array of Minigame HUD Objective Progress Icon Object References=Get FreeEnemyIcons() -> (FreeEnemyIcons:Array of Minigame HUD Objective Progress Icon Object References)
    Option 1:Array of Minigame HUD Objective Progress Icon Object References=Get FreeFriendlyIcons() -> (FreeFriendlyIcons:Array of Minigame HUD Objective Progress Icon Object References)
    Index=Get bIsFriendly()
   ) -> (ReturnValue:Array of Minigame HUD Objective Progress Icon Object References)
   Branch(
    execute=@n7.Is Not Valid
    Condition="integer > integer"(A=Length(TargetArray:Array of Minigame HUD Objective Progress Icon Object References=@n8), B=0) {is=title K2Node_CallFunction}
   )
   n9 = Last Index!(TargetArray:Array of Minigame HUD Objective Progress Icon Object References=@n8) -> (ReturnValue=0)
   > Set ClaimedIcon(ClaimedIcon="Get (a copy)"(Array:Array of Minigame HUD Objective Progress Icon Object References=@n8, Dimension 1=@n9) -> (Output:Minigame HUD Objective Progress Icon Object Reference))
   > Remove Index!(execute, TargetArray:Array of Minigame HUD Objective Progress Icon Object References=@n8, IndexToRemove:int=@n9) -> (then)
   > n10 = Branch(Condition=Get bIsFriendly())
   n11 = Get TrackedIndex() -> (TrackedIndex=0)
   n12 = Get ClaimedIcon()
   > Add!(
    execute
    TargetMap:Map of Integers to Minigame HUD Objective Progress Icon Object References=Get UsedFriendlyIcons() -> (UsedFriendlyIcons:Map of Integers to Minigame HUD Objective Progress Icon Object References)
    Key:"Integer (by ref)"=@n11
    Value:"Minigame HUD Objective Progress Icon Object Reference (by ref)"=@n12
   ) -> (then) {is=title K2Node_CallFunction}
   > SetVisibility(self=Get FriendlyObjectiveBox() -> (FriendlyObjectiveBox:Size Box Object Reference), InVisibility=SelfHitTestInvisible)
   Add(execute=@n10.else, TargetMap=Get UsedEnemyIcons(), Key=@n11, Value=@n12)
   > SetVisibility(self=Get EnemyObjectiveBox() -> (EnemyObjectiveBox:Size Box Object Reference), InVisibility=SelfHitTestInvisible)
   Is Valid(exec=@n3.then_3, InputObject=Get ClaimedIcon())
   > n13 = Branch(Condition=Get bExistingIcon() -> (bExistingIcon:bool))
   SetVisibility(execute=[@n14.then, @n13.then], self=Get ClaimedIcon(), InVisibility=SelfHitTestInvisible)
   InitializeIconFromObjective!(
    execute=@n13.else
    self:Minigame HUD Objective Progress Icon Object Reference=Get ClaimedIcon()
    Objective:Tracked Objective Structure=Get ObjectiveInfo() -> (ObjectiveInfo:"Tracked Objective Structure (by ref)")
    NewIsFriendly:bool=Get bIsFriendly()
   ) -> (then)
   > n14 = RefreshIconOrdering!(execute, self, bIsFriendly:bool=Get bIsFriendly()) -> (then)
   Return Node!(execute=@n3.then_4, IconWidget:Minigame HUD Objective Progress Icon Object Reference=Get ClaimedIcon())
   n15 = Find(TargetMap=Get UsedFriendlyIcons(), Key=Get TrackedIndex() -> (TrackedIndex=0))
   Branch(execute=@n3.then_0, Condition=@n15.ReturnValue)
   > Set ClaimedIcon(ClaimedIcon=@n15)
   > n16 = Branch(Condition=NOT Boolean(A=Get bIsFriendly()))
   > FreeIcon(TrackedIndex=Get TrackedIndex() -> (TrackedIndex=0))
   > Set ClaimedIcon()
   Set bExistingIcon(execute=@n16.else, bExistingIcon=true)
  ]}
 EventGraph {nodes [
   OnObjectiveTrackedAdded!() -> (OutputDelegate, then, TrackedIndex:int) {is=title K2Node_CustomEvent}
   > n1 = Branch(Condition="Equal (Enum)"(
     A:ESlateVisibility Enum=GetVisibility!(self:Widget Object Reference) -> (ReturnValue:ESlateVisibility Enum=Visible)
     B:ESlateVisibility Enum=Collapsed
    ))
   RefreshObjectiveDisplay(execute=@n1.else)
   n2 = OnTrackedObjectiveRemoved!() -> (OutputDelegate, then, TrackedIndex:int) {is=title K2Node_CustomEvent}
   > RemoveFromObjectiveDisplay!(execute, self, TrackedIndex:int=@n2.TrackedIndex) -> (then)
   > Delay(Duration=0.1)
   > RefreshObjectiveDisplay()
   n3 = OnTrackedObjectiveChanged!() -> (OutputDelegate, then, TrackedIndex:int) {is=title K2Node_CustomEvent}
   > UpdateIcon!(execute, self, TrackedIndex:int=@n3.TrackedIndex) -> (then)
   > UpdateAllNumericViews()
   n4 = OnObjectiveOwnershipChanged!() -> (OutputDelegate, then, TrackedIndex:int) {is=title K2Node_CustomEvent}
   > ClaimIcon!(
    execute
    self
    ObjectiveInfo:"Tracked Objective Structure (by ref)"=GetTrackedObjective!(
     self:Fort Minigame Objective Tracker Component Object Reference=GetObjectiveTracker!() -> (ObjectiveTracker:Fort Minigame Objective Tracker Component Object Reference)
     TrackedIndex:int=@n4.TrackedIndex
    ) -> (TrackedObjective:Tracked Objective Structure, ReturnValue:bool)
   ) -> (then, IconWidget:Minigame HUD Objective Progress Icon Object Reference) {is=title K2Node_CallFunction}
   > UpdateIcon(TrackedIndex=@n4.TrackedIndex)
   > UpdateAllNumericViews()
   > UpdateViewStyle()
   > RefreshObjectiveBarVisibility()
   n5 = OnTrackedObjectiveStateChanged!() -> (OutputDelegate, then, TrackedIndex:int) {is=title K2Node_CustomEvent}
   n6 = GetTrackedObjective(self=GetObjectiveTracker(), TrackedIndex=@n5.TrackedIndex)
   > Branch(Condition=@n6.ReturnValue)
   n7 = Break Tracked Objective(TrackedObjective=@n6) -> (ProgressPercentage=~, bIsInteractionAllowed=~, bOldIsInteractionAllowed=~)
   > UpdateIcon(TrackedIndex=@n7)
   > n8 = Sequence!(execute) -> (then_0, then_1)
   n9 = Find(TargetMap=Get UsedFriendlyIcons(), Key=@n7)
   > Branch(Condition=@n9.ReturnValue)
   > UpdateObjectiveState!(
    execute
    self:Minigame HUD Objective Progress Icon Object Reference=@n9
    ObjectiveState:Byte=@n7.TrackedState
    OldObjectiveState:Byte=@n7.OldTrackedState
    TeamIndex:Byte=@n7.TeamIndex
    ObjectiveStateOwnerTeam:Byte=@n7.TrackedStateOwnerTeam
   ) -> (then)
   n10 = Find(TargetMap=Get UsedEnemyIcons(), Key=@n7)
   Branch(execute=@n8.then_1, Condition=@n10.ReturnValue)
   > UpdateObjectiveState(
    self=@n10
    ObjectiveState=@n7.TrackedState
    OldObjectiveState=@n7.OldTrackedState
    TeamIndex=@n7.TeamIndex
    ObjectiveStateOwnerTeam=@n7.TrackedStateOwnerTeam
   )
   n11 = OnObjectiveTrackedAllowedInteractionChanged!() -> (OutputDelegate, then, Tracked Index:int) {is=title K2Node_CustomEvent}
   n12 = GetTrackedObjective(self=GetObjectiveTracker(), TrackedIndex=@n11.Tracked Index)
   > Branch(Condition=@n12.ReturnValue)
   > n13 = Sequence()
   n14 = Break Tracked Objective(TrackedObjective=@n12) -> (ProgressPercentage=~, bIsInteractionAllowed=~, bOldIsInteractionAllowed=~)
   n15 = Find(TargetMap=Get UsedFriendlyIcons(), Key=@n14)
   > Branch(Condition=@n15.ReturnValue)
   > SetIsInteractionAllowed!(execute, self:Minigame HUD Objective Progress Icon Object Reference=@n15, bInIsAllowedInteraction:bool=@n14.bIsInteractionAllowed) -> (then)
   n16 = Find(TargetMap=Get UsedEnemyIcons(), Key=@n14)
   Branch(execute=@n13.then_1, Condition=@n16.ReturnValue)
   > SetIsInteractionAllowed(self=@n16, bInIsAllowedInteraction=@n14.bIsInteractionAllowed)
   Event Construct()
   > Parent: Construct()
   > n17 = Sequence()
   > n18 = Set FreeFriendlyIcons(FreeFriendlyIcons:Array of Minigame HUD Objective Progress Icon Object References=Make Array!(
     "[0]":Minigame HUD Objective Progress Icon Object Reference=Get FriendlyIcon1() -> (FriendlyIcon1:Minigame HUD Objective Progress Icon Object Reference)
     "[1]":Minigame HUD Objective Progress Icon Object Reference=Get FriendlyIcon2() -> (FriendlyIcon2:Minigame HUD Objective Progress Icon Object Reference)
     "[2]":Minigame HUD Objective Progress Icon Object Reference=Get FriendlyIcon3() -> (FriendlyIcon3:Minigame HUD Objective Progress Icon Object Reference)
     "[3]":Minigame HUD Objective Progress Icon Object Reference=Get FriendlyIcon4() -> (FriendlyIcon4:Minigame HUD Objective Progress Icon Object Reference)
    ) -> (Array:Array of Minigame HUD Objective Progress Icon Object References)) -> (Output_Get:Array of Minigame HUD Objective Progress Icon Object References)
   > SetFriendlyTeamColors!(execute, self, Array:Array of Minigame HUD Objective Progress Icon Object References=@n18) -> (then)
   > n19 = Set FreeEnemyIcons(FreeEnemyIcons:Array of Minigame HUD Objective Progress Icon Object References=Make Array(
     "[0]"=Get EnemyIcon1() -> (EnemyIcon1:Minigame HUD Objective Progress Icon Object Reference)
     "[1]"=Get EnemyIcon2() -> (EnemyIcon2:Minigame HUD Objective Progress Icon Object Reference)
     "[2]"=Get EnemyIcon3() -> (EnemyIcon3:Minigame HUD Objective Progress Icon Object Reference)
     "[3]"=Get EnemyIcon4() -> (EnemyIcon4:Minigame HUD Objective Progress Icon Object Reference)
    )) -> (Output_Get:Array of Minigame HUD Objective Progress Icon Object References)
   > SetEnemyTeamColors!(execute, self, Array:Array of Minigame HUD Objective Progress Icon Object References=@n19) -> (then)
   > n20 = For Each Loop(Array:Array of Minigame HUD Objective Progress Icon Object References=Get FreeFriendlyIcons()) -> (Array Element:Minigame HUD Objective Progress Icon Object Reference)
   n21 = For Each Loop(Exec=@n20.Completed, Array=Get FreeEnemyIcons())
   > SetVisibility(self=@n21, InVisibility=Collapsed)
   > RemoveFromParent(self=@n21)
   SetVisibility(execute=@n20.LoopBody, self=@n20, InVisibility=Collapsed)
   > RemoveFromParent(self=@n20)
   SetVisibility(execute=@n17.then_1, self=[Get EnemyObjectiveBox(), Get FriendlyObjectiveBox()], InVisibility=Collapsed)
   n22 = Event PreConstruct()
   > Branch(Condition=@n22.IsDesignTime)
   > n23 = Set FreeFriendlyIcons(FreeFriendlyIcons=Make Array("[0]"=Get FriendlyIcon1(), "[1]"=Get FriendlyIcon2(), "[2]"=Get FriendlyIcon3(), "[3]"=Get FriendlyIcon4()))
   > SetFriendlyTeamColors(Array=@n23)
   > n24 = Set FreeEnemyIcons(FreeEnemyIcons=Make Array("[0]"=Get EnemyIcon1(), "[1]"=Get EnemyIcon2(), "[2]"=Get EnemyIcon3(), "[3]"=Get EnemyIcon4()))
   > SetEnemyTeamColors(Array=@n24)
   Event OnLocalPlayerChangedTeams() {is=title K2Node_Event}
   > RefreshObjectiveDisplay()
   n25 = Event OnDisplayMinigameInProgressUI!() -> (OutputDelegate, then, bDisplay:bool) {is=title K2Node_Event}
   > Parent: OnDisplayMinigameInProgressUI!(execute, bDisplay:bool=@n25.bDisplay) -> (then) {is=title K2Node_CallParentFunction}
   > SetVisibility(InVisibility=Select(Option 0:ESlateVisibility Enum=Collapsed, Option 1:ESlateVisibility Enum=SelfHitTestInvisible, Index=@n25.bDisplay) -> (ReturnValue:ESlateVisibility Enum))
   > n26 = Branch(Condition=@n25.bDisplay)
   n27 = GetObjectiveTracker()
   > BindToObjectiveTracker!(execute, self, ObjectiveTracker:Fort Minigame Objective Tracker Component Object Reference=@n27) -> (then) {is=title K2Node_CallFunction}
   > RefreshObjectiveDisplay()
   UnbindFromObjectiveTracker!(execute=@n26.else, self, ObjectiveTracker:Fort Minigame Objective Tracker Component Object Reference=@n27) -> (then)
   > Is Valid(InputObject=@n27)
   > n28 = For Each Loop(Array:Array of Tracked Objective Structures=GetTrackedObjectivesByQuery!(
     self:Fort Minigame Objective Tracker Component Object Reference=@n27
     TeamIndex:Byte=GetOwnerPlayerTeam() -> (TeamIndex=0)
     QueryType:ETrackedObjectiveQuery Enum=All
    ) -> (TrackedObjectives:Array of Tracked Objective Structures, ReturnValue:bool)) -> (Array Element:Tracked Objective Structure)
   > RemoveFromObjectiveDisplay(TrackedIndex=Break Tracked Objective(TrackedObjective=@n28) -> (ProgressPercentage=~, bIsInteractionAllowed=~, bOldIsInteractionAllowed=~))
  ]}
 FreeIcon {nodes [
   FreeIcon!() -> (then, TrackedIndex:int) {is=title K2Node_FunctionEntry}
   n1 = Find(TargetMap=Get UsedFriendlyIcons(), Key=Get TrackedIndex() -> (TrackedIndex=0))
   > n2 = Branch(Condition=@n1.ReturnValue)
   > n3 = Set FoundIcon(FoundIcon:Minigame HUD Objective Progress Icon Object Reference=@n1) -> (Output_Get:Minigame HUD Objective Progress Icon Object Reference)
   > SetVisibility(self=@n3, InVisibility=Collapsed)
   > RemoveFromParent(self=@n3)
   > Add(
    TargetArray:Array of Minigame HUD Objective Progress Icon Object References=Get FreeFriendlyIcons()
    NewItem:"Minigame HUD Objective Progress Icon Object Reference (by ref)"=@n3
   ) {is=title K2Node_CallArrayFunction}
   > Remove!(
    execute
    TargetMap:Map of Integers to Minigame HUD Objective Progress Icon Object References=Get UsedFriendlyIcons()
    Key:"Integer (by ref)"=Get TrackedIndex() -> (TrackedIndex=0)
   ) -> (then, ReturnValue:bool)
   > RefreshIconOrdering(bIsFriendly=true)
   n4 = Get UsedFriendlyIcons()
   > Branch(Condition="Equal (Integer)"(
     A=Length!(TargetMap:Map of Integers to Minigame HUD Objective Progress Icon Object References=@n4) -> (ReturnValue:int=0) {is=title K2Node_CallFunction}
     B=0
    ) {is=title K2Node_CallFunction})
   > SetVisibility(self=Get FriendlyObjectiveBox(), InVisibility=Collapsed)
   n5 = Find(TargetMap=Get UsedEnemyIcons(), Key=Get TrackedIndex() -> (TrackedIndex=0))
   Branch(execute=@n2.else, Condition=@n5.ReturnValue)
   > n6 = Set FoundIcon(FoundIcon=@n5)
   > SetVisibility(self=@n6, InVisibility=Collapsed)
   > RemoveFromParent(self=@n6)
   > Add(TargetArray=Get FreeEnemyIcons(), NewItem=@n6)
   > Remove(TargetMap=Get UsedEnemyIcons(), Key=Get TrackedIndex() -> (TrackedIndex=0))
   > RefreshIconOrdering(bIsFriendly=false)
   n7 = Get UsedEnemyIcons()
   > Branch(Condition="Equal (Integer)"(A=Length(TargetMap=@n7) -> (ReturnValue=0), B=0))
   > SetVisibility(self=Get EnemyObjectiveBox(), InVisibility=Collapsed)
  ]}
 GetIconByTrackedIndex {nodes [
   n1 = GetIconByTrackedIndex!() -> (then, TrackedIndex:int) {is=title K2Node_FunctionEntry}
   n2 = Find(TargetMap=Get UsedFriendlyIcons(), Key=@n1)
   > n3 = Branch(Condition=@n2.ReturnValue)
   > Return Node!(execute, IconFound:Minigame HUD Objective Progress Icon Object Reference=@n2)
   Return Node(execute=@n3.else, IconFound=Find(TargetMap=Get UsedEnemyIcons(), Key=@n1))
  ]}
 GetObjectiveTracker {nodes [
   GetObjectiveTracker() {is=title K2Node_FunctionEntry}
   > n1 = Sequence()
   > n2 = Get!(execute:Exec) -> (then:Exec, else:Exec, CurrentMinigame:Fort Minigame Object Reference)
   > Set CurrentTracker(CurrentTracker:Fort Minigame Objective Tracker Component Object Reference=GetComponentByClass!(self:Actor Object Reference=@n2, ComponentClass:Actor Component Class Reference=/Script/FortniteGame.FortMinigameObjectiveTrackerComponent) -> (ReturnValue:Fort Minigame Objective Tracker Component Object Reference)) -> (Output_Get:Fort Minigame Objective Tracker Component Object Reference)
   Return Node!(
    execute=@n1.then_1
    ObjectiveTracker:Fort Minigame Objective Tracker Component Object Reference=Get CurrentTracker() -> (CurrentTracker:Fort Minigame Objective Tracker Component Object Reference)
   )
  ]}
 GetObjectivesByAffiliation {nodes [
   n1 = GetObjectivesByAffiliation!() -> (then, Affiliation:ETrackedObjectiveQuery Enum=All) {is=title K2Node_FunctionEntry}
   n2 = GetTrackedObjectivesByQuery(self=GetObjectiveTracker() {is=title K2Node_CallFunction}, TeamIndex=GetOwnerPlayerTeam() -> (TeamIndex=0), QueryType=@n1)
   > Return Node!(execute, TrackedObjectivesInfo:Array of Tracked Objective Structures=@n2, bHasObjectives:bool=@n2.ReturnValue)
  ]}
 GetOrderedIcons {nodes [
   n1 = GetOrderedIcons!() -> (then, IconMap:Map of Integers to Minigame HUD Objective Progress Icon Object References) {is=title K2Node_FunctionEntry}
   > n2 = Sequence()
   > n3 = Values!(execute, TargetMap:Map of Integers to Minigame HUD Objective Progress Icon Object References=@n1) -> (then, Values:Array of Minigame HUD Objective Progress Icon Object References)
   > n4 = For Each Loop(Array:Array of Minigame HUD Objective Progress Icon Object References=@n3) -> (Array Element:Minigame HUD Objective Progress Icon Object Reference)
   > Set InsertIndex(InsertIndex:int=0) -> (Output_Get:int=0)
   > n5 = For Each Loop with Break!(
    Exec:Exec
    Array:Array of Minigame HUD Objective Progress Icon Object References=Get SortedWidgets() -> (SortedWidgets:Array of Minigame HUD Objective Progress Icon Object References)
    Break:Exec=@n6.then
   ) -> (LoopBody:Exec, Array Element:Minigame HUD Objective Progress Icon Object Reference, Array Index:int, Completed:Exec)
   Insert!(
    execute=@n5.Completed
    TargetArray:Array of Minigame HUD Objective Progress Icon Object References=Get SortedWidgets()
    NewItem:"Minigame HUD Objective Progress Icon Object Reference (by ref)"=@n4
    Index:int=Get InsertIndex() -> (InsertIndex:int=0)
   ) -> (then) {is=title K2Node_CallArrayFunction}
   n6 = Branch(
    execute=@n5.LoopBody
    Condition="integer < integer"!(
     A:int=Get IndicatorPresetIndex!(self:Minigame HUD Objective Progress Icon Object Reference=@n4) -> (IndicatorPresetIndex:int=0)
     B:int=Get IndicatorPresetIndex(self=@n5) -> (IndicatorPresetIndex=0)
    ) -> (ReturnValue:bool) {is=title K2Node_CallFunction}
   )
   Increment Int(" "=@n6.else, Value=Get InsertIndex() -> (InsertIndex=0))
   Return Node!(execute=@n2.then_1, OrderedIcons:Array of Minigame HUD Objective Progress Icon Object References=Get SortedWidgets())
  ]}
 GetOwnerPlayerTeam {nodes [
   GetOwnerPlayerTeam() {is=title K2Node_FunctionEntry}
   n1 = GetOwningPlayer()
   > n2 = Is Valid(InputObject=@n1)
   > n3 = Cast To FortPlayerController(Object=@n1)
   > Return Node!(execute, TeamIndex:Byte=GetActorTeam!(Actor:Actor Object Reference=@n3) -> (ReturnValue:Byte=0))
   Return Node(execute=[@n3.CastFailed, @n2.Is Not Valid], TeamIndex=255)
  ]}
 OnStateChanged {nodes [
   n1 = OnStateChanged!() -> (then, bIsFriendly:bool=~) {is=title K2Node_FunctionEntry}
   > n2 = Set ObjectiveTracker(ObjectiveTracker:Fort Minigame Objective Tracker Component Object Reference=GetObjectiveTracker()) -> (Output_Get:Fort Minigame Objective Tracker Component Object Reference)
   > Is Valid(InputObject=@n2)
   > Set ObjectiveAnimation(ObjectiveAnimation:Widget Animation Object Reference=Select(
     Option 0:Widget Animation Object Reference=Get EnemyObjectiveDamaged() -> (EnemyObjectiveDamaged:Widget Animation Object Reference)
     Option 1:Widget Animation Object Reference=Get FriendlyObjectiveDamaged() -> (FriendlyObjectiveDamaged:Widget Animation Object Reference)
     Index=@n1
    ) -> (ReturnValue:Widget Animation Object Reference)) -> (Output_Get:Widget Animation Object Reference)
   > Set NumericView(NumericView:Image Object Reference=Select(
     Option 0:Image Object Reference=Get EnemyBacking_NumericView() -> (EnemyBacking_NumericView:Image Object Reference)
     Option 1:Image Object Reference=Get FriendlyBacking_NumericView() -> (FriendlyBacking_NumericView:Image Object Reference)
     Index=@n1
    ) -> (ReturnValue:Image Object Reference)) -> (Output_Get:Image Object Reference)
   > Set Numerator(Numerator:Common Text Object Reference=Select(
     Option 0:Common Text Object Reference=Get EnemyNumerator() -> (EnemyNumerator:Common Text Object Reference)
     Option 1:Common Text Object Reference=Get FriendlyNumerator() -> (FriendlyNumerator:Common Text Object Reference)
     Index=@n1
    ) -> (ReturnValue:Common Text Object Reference)) -> (Output_Get:Common Text Object Reference)
   > Set Demoninator(Demoninator:Common Text Object Reference=Select(
     Option 0=Get EnemyDenominator() -> (EnemyDenominator:Common Text Object Reference)
     Option 1=Get FriendlyDenominator() -> (FriendlyDenominator:Common Text Object Reference)
     Index=@n1
    )) -> (Output_Get:Common Text Object Reference)
   n3 = GetTrackedObjectivesByQuery(
    self=Get ObjectiveTracker()
    TeamIndex=GetOwnerPlayerTeam() -> (TeamIndex=0) {is=title K2Node_CallFunction}
    QueryType=Select(Option 0:ETrackedObjectiveQuery Enum=Hostile, Option 1:ETrackedObjectiveQuery Enum=Friendly, Index=@n1) -> (ReturnValue:ETrackedObjectiveQuery Enum)
   )
   > Set TrackedObjectives(TrackedObjectives:Array of Tracked Objective Structures=@n3) -> (Output_Get:Array of Tracked Objective Structures)
   > n4 = Set NumObjectivesTotal(NumObjectivesTotal:int=Length(TargetArray:Array of Tracked Objective Structures=@n3) {is=title K2Node_CallArrayFunction}) -> (Output_Get:int=0)
   > "SetText (Text)"(self=Get Demoninator() -> (Demoninator:Common Text Object Reference), InText="To Text (Integer)"(Value=@n4))
   > n5 = Sequence()
   > Branch(Condition="integer > integer"(A=Get NumObjectivesTotal() -> (NumObjectivesTotal:int=0), B=0))
   > n6 = For Each Loop(Array:Array of Tracked Objective Structures=Get TrackedObjectives() -> (TrackedObjectives:Array of Tracked Objective Structures)) -> (Array Element:Tracked Objective Structure)
   > Branch(Condition="float > float"(
     A=Break Tracked Objective(TrackedObjective=@n6) -> (ProgressPercentage=~, bIsInteractionAllowed=~, bOldIsInteractionAllowed=~).ProgressPercentage
     B=0
    ) {is=title K2Node_CallFunction})
   > Increment Int(Value=Get NumObjectivesRemaining() -> (NumObjectivesRemaining:int=0))
   n7 = Get NumObjectivesTotal() -> (NumObjectivesTotal=0)
   n8 = Get NumObjectivesRemaining() -> (NumObjectivesRemaining=0)
   Set FillPercentage(
    execute=@n6.Completed
    FillPercentage:double=float / float(A="To Float (Integer)"(InInt=@n8), B="To Float (Integer)"(InInt=@n7)) {is=title K2Node_CallFunction}
   ) -> (Output_Get:double)
   "SetText (Text)"(
    execute=@n5.then_1
    self=Get Numerator() -> (Numerator:Common Text Object Reference)
    InText="To Text (Integer)"(Value=Get NumObjectivesRemaining() -> (NumObjectivesRemaining=0))
   )
   > n9 = GetDynamicMaterial(self=Get NumericView() -> (NumericView:Image Object Reference))
   > Is Valid(InputObject=@n9)
   > SetScalarParameterValue(self=@n9, ParameterName=PercentFill, Value=Get FillPercentage() -> (FillPercentage:double))
   n10 = Get ObjectiveAnimation() -> (ObjectiveAnimation:Widget Animation Object Reference)
   > Branch(Condition=NOT Boolean(A=IsAnimationPlaying(InAnimation=@n10)))
   > PlayAnimation(InAnimation=Get ObjectiveAnimation())
  ]}
 RefreshIconOrdering {nodes [
   n1 = RefreshIconOrdering!() -> (then, bIsFriendly:bool=~) {is=title K2Node_FunctionEntry}
   > n2 = Set IconContainer(IconContainer:Horizontal Box Object Reference=Select(
     Option 0:Horizontal Box Object Reference=Get EnemyIconsBox() -> (EnemyIconsBox:Horizontal Box Object Reference)
     Option 1:Horizontal Box Object Reference=Get FriendlyIconsBox() -> (FriendlyIconsBox:Horizontal Box Object Reference)
     Index=@n1
    ) -> (ReturnValue:Horizontal Box Object Reference)) -> (Output_Get:Horizontal Box Object Reference)
   > ClearChildren(self=@n2)
   > n3 = Sequence()
   > n4 = Branch(Condition=Get bIsFriendly())
   > Set SortedWidgets(SortedWidgets:Array of Minigame HUD Objective Progress Icon Object References=GetOrderedIcons!(IconMap:Map of Integers to Minigame HUD Objective Progress Icon Object References=Get UsedFriendlyIcons()) -> (OrderedIcons:Array of Minigame HUD Objective Progress Icon Object References) {is=title K2Node_CallFunction}) -> (Output_Get:Array of Minigame HUD Objective Progress Icon Object References)
   Set SortedWidgets(execute=@n4.else, SortedWidgets=GetOrderedIcons(IconMap=Get UsedEnemyIcons()))
   n5 = For Each Loop(Exec=@n3.then_1, Array:Array of Minigame HUD Objective Progress Icon Object References=Get SortedWidgets()) -> (Array Element:Minigame HUD Objective Progress Icon Object Reference)
   > n6 = AddChildToHorizontalBox!(
    execute
    self:Horizontal Box Object Reference=Get IconContainer() -> (IconContainer:Horizontal Box Object Reference)
    Content:Widget Object Reference=@n5
   ) -> (then, ReturnValue:Horizontal Box Slot Object Reference)
   > SetPadding!(execute, self:Horizontal Box Slot Object Reference=@n6, InPadding:Margin Structure=Make Margin(Left=0, Top=0, Right=6, Bottom=0)) -> (then)
   > SetSize!(
    execute
    self:Horizontal Box Slot Object Reference=@n6
    InSize:Slate Child Size Structure=Make Slate Child Size!(Value:float=1, SizeRule:ESlateSizeRule Enum=Fill) -> (SlateChildSize:Slate Child Size Structure) {is=title K2Node_MakeStruct}
   ) -> (then)
  ]}
 RefreshObjectiveBarVisibility {nodes [
   RefreshObjectiveBarVisibility() {is=title K2Node_FunctionEntry}
   > SetVisibility(
    self=Get Friendly() -> (Friendly:Overlay Object Reference)
    InVisibility=Select(
     Option 0:ESlateVisibility Enum=Collapsed
     Option 1:ESlateVisibility Enum=SelfHitTestInvisible
     Index=GetObjectivesByAffiliation!(Affiliation:ETrackedObjectiveQuery Enum=Friendly) -> (TrackedObjectivesInfo:Array of Tracked Objective Structures, bHasObjectives:bool) {is=title K2Node_CallFunction}.bHasObjectives
    ) -> (ReturnValue:ESlateVisibility Enum)
   )
   > SetVisibility(
    self=Get Enemy() -> (Enemy:Overlay Object Reference)
    InVisibility=Select(Option 0=Collapsed, Option 1=SelfHitTestInvisible, Index=GetObjectivesByAffiliation(Affiliation=NotFriendly).bHasObjectives)
   )
  ]}
 RefreshObjectiveDisplay {nodes [
   RefreshObjectiveDisplay() {is=title K2Node_FunctionEntry}
   > n1 = Set ObjectiveTracker(ObjectiveTracker=GetObjectiveTracker())
   > Is Valid(InputObject=@n1)
   n2 = GetTrackedObjectivesByQuery(self=@n1, TeamIndex=GetOwnerPlayerTeam() -> (TeamIndex=0), QueryType=All)
   > Set TrackedObjectives(TrackedObjectives=@n2)
   > n3 = Branch(Condition=@n2.ReturnValue)
   > SetVisibility(InVisibility=Collapsed)
   > n4 = For Each Loop(Array:Array of Tracked Objective Structures=Get TrackedObjectives()) -> (Array Element:Tracked Objective Structure)
   n5 = For Each Loop(Exec=@n4.Completed, Array=Get TrackedObjectives())
   SetVisibility(execute=[@n5.Completed, @n3.else], InVisibility=SelfHitTestInvisible)
   > RefreshObjectiveBarVisibility() {is=title K2Node_CallFunction}
   ClaimIcon(execute=@n5.LoopBody, ObjectiveInfo=@n5)
   n6 = Break Tracked Objective(TrackedObjective=@n5) -> (ProgressPercentage=~, bIsInteractionAllowed=~, bOldIsInteractionAllowed=~)
   > UpdateIcon(TrackedIndex=@n6)
   > OnObjectiveTrackedAllowedInteractionChanged!(execute, self, Tracked Index:int=@n6) -> (then) {is=title K2Node_CallFunction}
   RemoveFromObjectiveDisplay(
    execute=@n4.LoopBody
    TrackedIndex=Break Tracked Objective(TrackedObjective=@n4) -> (ProgressPercentage=~, bIsInteractionAllowed=~, bOldIsInteractionAllowed=~)
   )
  ]}
 RemoveFromObjectiveDisplay {nodes [
   n1 = RemoveFromObjectiveDisplay!() -> (then, TrackedIndex:int) {is=title K2Node_FunctionEntry}
   > FreeIcon(TrackedIndex=@n1) {is=title K2Node_CallFunction}
   > UpdateAllNumericViews()
   > UpdateViewStyle()
   > RefreshObjectiveBarVisibility()
  ]}
 SetEnemyTeamColors {nodes [
   n1 = SetEnemyTeamColors!() -> (then, Array:Array of Minigame HUD Objective Progress Icon Object References) {is=title K2Node_FunctionEntry}
   > n2 = For Each Loop(Array:Array of Minigame HUD Objective Progress Icon Object References=@n1) -> (Array Element:Minigame HUD Objective Progress Icon Object Reference)
   > SetBrushFromMaterial(
    self=Get Image_BackplateProgress!(self:Minigame HUD Objective Progress Icon Object Reference=@n2) -> (Image_BackplateProgress:Image Object Reference)
    Material=/Game/Athena/HUD/LTM/Playground/Materials/M_UI_HUD_Objective_Progress_Enemy
   )
   > SetContentColorAndOpacity!(
    execute
    self:Border Object Reference=Get Indicator_Color!(self:Minigame HUD Objective Progress Icon Object Reference=@n2) -> (Indicator_Color:Border Object Reference)
    InContentColorAndOpacity:Linear Color Structure=(R=1,G=0.7913,B=0.5029,A=1)
   ) -> (then)
  ]}
 SetFriendlyTeamColors {nodes [
   n1 = SetFriendlyTeamColors!() -> (then, Array:Array of Minigame HUD Objective Progress Icon Object References) {is=title K2Node_FunctionEntry}
   > n2 = For Each Loop(Array=@n1)
   > SetBrushFromMaterial(self=Get Image_BackplateProgress(self=@n2), Material=/Game/Athena/HUD/LTM/Playground/Materials/M_UI_HUD_Objective_Progress_Friendly)
   > SetContentColorAndOpacity(self=Get Indicator_Color(self=@n2), InContentColorAndOpacity=(R=0.5029,G=1,B=1,A=1))
  ]}
 UnbindFromObjectiveTracker {nodes [
   n1 = UnbindFromObjectiveTracker!() -> (then, ObjectiveTracker:Fort Minigame Objective Tracker Component Object Reference) {is=title K2Node_FunctionEntry}
   > Is Valid(InputObject=@n1)
   > Unbind Event from On Tracked Objective Changed(self:Fort Minigame Objective Tracker Component Object Reference=Get ObjectiveTracker(), Delegate=Create Event()) {is=title K2Node_RemoveDelegate}
   > Unbind Event from On Tracked Objective State Changed(self:Fort Minigame Objective Tracker Component Object Reference=Get ObjectiveTracker(), Delegate=Create Event()) {is=title K2Node_RemoveDelegate}
   > Unbind Event from On Tracked Objective Team Updated(self:Fort Minigame Objective Tracker Component Object Reference=Get ObjectiveTracker(), Delegate=Create Event()) {is=title K2Node_RemoveDelegate}
   > Unbind Event from On Tracked Objective Removed(self:Fort Minigame Objective Tracker Component Object Reference=Get ObjectiveTracker(), Delegate=Create Event()) {is=title K2Node_RemoveDelegate}
   > Unbind Event from On Tracked Objective Added(self:Fort Minigame Objective Tracker Component Object Reference=Get ObjectiveTracker(), Delegate=Create Event()) {is=title K2Node_RemoveDelegate}
   > Unbind Event from On Tracked Objective Allowed Interaction Changed(self:Fort Minigame Objective Tracker Component Object Reference=Get ObjectiveTracker(), Delegate=Create Event()) {is=title K2Node_RemoveDelegate}
  ]}
 UpdateAllNumericViews {nodes [
   UpdateAllNumericViews() {is=title K2Node_FunctionEntry}
   > UpdateNumericView!(execute, self, bIsFriendly:bool=false) -> (then)
   > UpdateNumericView(bIsFriendly=true)
  ]}
 UpdateIcon {nodes [
   n1 = UpdateIcon!() -> (then, TrackedIndex:int) {is=title K2Node_FunctionEntry}
   > n2 = Set Icon(Icon:Minigame HUD Objective Progress Icon Object Reference=GetIconByTrackedIndex!(TrackedIndex:int=@n1) -> (IconFound:Minigame HUD Objective Progress Icon Object Reference) {is=title K2Node_CallFunction}) -> (Output_Get:Minigame HUD Objective Progress Icon Object Reference)
   > Is Valid(InputObject=@n2)
   n3 = GetObjectiveTracker()
   > AnimateNewScore!(
    execute
    self:Minigame HUD Objective Progress Icon Object Reference=@n2
    NewScoreValue:double=Break Tracked Objective(TrackedObjective=GetTrackedObjective(self=@n3, TrackedIndex=@n1)) -> (ProgressPercentage=~, bIsInteractionAllowed=~, bOldIsInteractionAllowed=~).ProgressPercentage
    bSkipAnimation:bool=NOT Boolean(A=IsVisible(self=@n2))
   ) -> (then)
   > SetVisibility(self=@n2, InVisibility=SelfHitTestInvisible)
  ]}
 UpdateNumericView {nodes [
   n1 = UpdateNumericView!() -> (then, bIsFriendly:bool=~) {is=title K2Node_FunctionEntry}
   > n2 = Set ObjectiveTracker(ObjectiveTracker=GetObjectiveTracker())
   > Is Valid(InputObject=@n2)
   > Set ObjectiveAnimation(ObjectiveAnimation=Select(
     Option 0:Widget Animation Object Reference=Get EnemyObjectiveDamaged()
     Option 1:Widget Animation Object Reference=Get FriendlyObjectiveDamaged()
     Index=@n1
    ) -> (ReturnValue:Widget Animation Object Reference))
   > Set NumericView(NumericView=Select(Option 0:Image Object Reference=Get EnemyBacking_NumericView(), Option 1:Image Object Reference=Get FriendlyBacking_NumericView(), Index=@n1) -> (ReturnValue:Image Object Reference))
   > Set Numerator(Numerator=Select(Option 0:Common Text Object Reference=Get EnemyNumerator(), Option 1:Common Text Object Reference=Get FriendlyNumerator(), Index=@n1) -> (ReturnValue:Common Text Object Reference))
   > Set Demoninator(Demoninator=Select(Option 0=Get EnemyDenominator(), Option 1=Get FriendlyDenominator(), Index=@n1))
   n3 = GetTrackedObjectivesByQuery(
    self=Get ObjectiveTracker()
    TeamIndex=GetOwnerPlayerTeam() -> (TeamIndex=0)
    QueryType=Select(Option 0:ETrackedObjectiveQuery Enum=NotFriendly, Option 1:ETrackedObjectiveQuery Enum=Friendly, Index=@n1) -> (ReturnValue:ETrackedObjectiveQuery Enum)
   )
   > Set TrackedObjectives(TrackedObjectives=@n3)
   > n4 = Set NumObjectivesTotal(NumObjectivesTotal=Length(TargetArray=@n3)) -> (Output_Get=0)
   > "SetText (Text)"(self=Get Demoninator(), InText="To Text (Integer)"(Value=@n4))
   > n5 = Sequence()
   > Branch(Condition="integer > integer"(A=Get NumObjectivesTotal() -> (NumObjectivesTotal=0), B=0))
   > n6 = For Each Loop(Array:Array of Tracked Objective Structures=Get TrackedObjectives()) -> (Array Element:Tracked Objective Structure)
   n7 = Break Tracked Objective(TrackedObjective=@n6) -> (ProgressPercentage=~, bIsInteractionAllowed=~, bOldIsInteractionAllowed=~)
   > n8 = Branch(Condition="float > float"(A=@n7.ProgressPercentage, B=0))
   Increment Int(" "=[@n8.then, @n10.then], Value=Get NumObjectivesRemaining() -> (NumObjectivesRemaining=0))
   n9 = Byte to Enum EObjectiveType!(Byte:Byte=@n7.ObjectiveType) -> (ReturnValue:EObjectiveType Enum) {is=title K2Node_CastByteToEnum}
   n10 = Branch(
    execute=@n8.else
    Condition=AND Boolean(A="Equal (Byte)"(A=@n7.TeamIndex, B=0), B="Equal (Enum)"(A:EObjectiveType Enum=@n9, B:EObjectiveType Enum=CaptureObjective))
   )
   n11 = Get NumObjectivesTotal() -> (NumObjectivesTotal=0)
   n12 = Get NumObjectivesRemaining() -> (NumObjectivesRemaining=0)
   Set FillPercentage(execute=@n6.Completed, FillPercentage=float / float(A="To Float (Integer)"(InInt=@n12), B="To Float (Integer)"(InInt=@n11)))
   n13 = "To String (Text)"(InText="GetText (Text)"!(self:Text Object Reference=Get Numerator()) -> (ReturnValue:Text))
   Set NumRemainingChanged(
    execute=@n5.then_1
    NumRemainingChanged:bool="Not Equal (Integer)"!(A:int=String To Integer(InString=@n13), B:int=Get NumObjectivesRemaining() -> (NumObjectivesRemaining=0)) -> (ReturnValue:bool) {is=title K2Node_CallFunction}
   ) -> (Output_Get:bool)
   > "SetText (Text)"(self=Get Numerator(), InText="To Text (Integer)"(Value=Get NumObjectivesRemaining() -> (NumObjectivesRemaining=0)))
   > n14 = GetDynamicMaterial(self=Get NumericView())
   > Is Valid(InputObject=@n14)
   > SetScalarParameterValue(self=@n14, ParameterName=PercentFill, Value=Get FillPercentage())
   n15 = IsAnimationPlaying(InAnimation=Get ObjectiveAnimation())
   > Branch(Condition=AND Boolean(A=NOT Boolean(A=@n15), B=Get NumRemainingChanged() -> (NumRemainingChanged:bool)))
   > PlayAnimation(InAnimation=Get ObjectiveAnimation())
  ]}
 UpdateViewStyle {nodes [
   UpdateViewStyle() {is=title K2Node_FunctionEntry}
   > n1 = Sequence!(execute) -> (then_0, then_1, then_2:Exec)
   n2 = "integer > integer"(A=Length(TargetArray=GetObjectivesByAffiliation(Affiliation=Friendly)), B=4)
   > SetActiveWidgetIndex(
    self=Get FriendlyBackingViewSwitcher() -> (FriendlyBackingViewSwitcher:Common Widget Switcher Legacy Object Reference)
    Index=Select(Option 0:int=0, Option 1:int=1, Index=@n2) -> (ReturnValue:int)
   )
   > SetActiveWidgetIndex(
    self=Get FriendlyObjectivesView() -> (FriendlyObjectivesView:Common Widget Switcher Legacy Object Reference)
    Index=Select(Option 0=0, Option 1=1, Index=@n2)
   )
   > n3 = Branch(Condition="Not Equal (Boolean)"!(A:bool=Get FriendlyNumbersModeViewActive() -> (FriendlyNumbersModeViewActive:bool), B:bool=@n2) -> (ReturnValue:bool) {is=title K2Node_CallFunction})
   > n4 = Set bIsDirty(bIsDirty:bool=true) -> (Output_Get:bool)
   Set FriendlyNumbersModeViewActive(execute=[@n4.then, @n3.else], FriendlyNumbersModeViewActive:bool=@n2) -> (Output_Get:bool)
   n5 = "integer > integer"(A=Length(TargetArray=GetObjectivesByAffiliation(Affiliation=NotFriendly)), B=4)
   SetActiveWidgetIndex(
    execute=@n1.then_1
    self:Common Visibility Switcher Object Reference=Get EnemyBackingViewSwitcher() -> (EnemyBackingViewSwitcher:Common Widget Stack Legacy Object Reference)
    Index=Select(Option 0=0, Option 1=1, Index=@n5)
   )
   > SetActiveWidgetIndex(
    self:Widget Switcher Object Reference=Get EnemyObjectivesView() -> (EnemyObjectivesView:Common Widget Switcher Legacy Object Reference)
    Index=Select(Option 0=0, Option 1=1, Index=@n5)
   )
   > n6 = Branch(Condition="Not Equal (Boolean)"(A=Get EnemyNumbersModeViewActive() -> (EnemyNumbersModeViewActive:bool), B=@n5))
   > n7 = Set bIsDirty(bIsDirty=true)
   Set EnemyNumbersModeViewActive(execute=[@n7.then, @n6.else], EnemyNumbersModeViewActive:bool=@n5) -> (Output_Get:bool)
   Branch(execute=@n1.then_2, Condition=Get bIsDirty() -> (bIsDirty:bool))
   > RefreshObjectiveDisplay() {is=title K2Node_CallFunction}
  ]}
]
variables [
 EnemyNumbersModeViewActive {type=boolean}
 FreeEnemyIcons {type="array<MinigameHUD_ObjectiveProgressIcon_C>", default []}
 FreeFriendlyIcons {type="array<MinigameHUD_ObjectiveProgressIcon_C>", default []}
 FriendlyNumbersModeViewActive {type=boolean}
 UsedEnemyIcons {default {}}
 UsedFriendlyIcons {default {}}
]
widgets {
 info {inheritedWidgetCount=1, parentClass=/Script/FortniteUI.MinigameObjectiveTrackerWidget, rootWidgetClass=&SizeBox, widgetCount=56}
 widgets [
  SizeBox_27 {
   widgetClassPath=&SizeBox
   widget=SizeBox_27
   properties {bOverride_WidthOverride=true, maxDesiredHeight=200, maxDesiredWidth=300, minDesiredWidth=300, widthOverride=300}
  }
  Overlay_0 {widgetClassPath=&Overlay, parent=SizeBox_27, widget=Overlay_0, slot {class=SizeBoxSlot, properties {padding=<ltrb 0 6 0 0>}}}
  OverallBacking {
   widgetClassPath=&Border
   parent=Overlay_0
   widget=OverallBacking
   properties {brushColor=<rgba 0 0.0021 0.0176 0.25>, padding=<ltrb 0 0 0 0>}
   slot {class=OverlaySlot, properties {horizontalAlignment=HAlign_Fill, verticalAlignment=VAlign_Fill}}
  }
  All {
   widgetClassPath=&VerticalBox
   parent=Overlay_0
   widget=All
   slot {class=OverlaySlot, properties {horizontalAlignment=HAlign_Fill, verticalAlignment=VAlign_Fill}}
  }
  Header {
   widgetClassPath=&SizeBox
   parent=All
   widget=Header
   properties {bOverride_MaxDesiredWidth=true, heightOverride=32, maxDesiredWidth=300}
   slot {class=VerticalBoxSlot, properties {horizontalAlignment=HAlign_Center}}
  }
  CommonTextBlock_8 {
   widgetClassPath=&CommonTextBlock
   parent=Header
   widget=CommonTextBlock_8
   properties {
    justification=Center
    text=Objectives
    textTransformPolicy=ToUpper
    wrapTextAt=300
    font {
     bForceMonospaced=false
     bMaterialIsStencil=false
     fontMaterial=null
     fontObject=&BurbankSmall
     letterSpacing=0
     monospacedWidth=1
     size=15
     skewAmount=0
     typefaceFontName=Black
     outlineSettings {
      bApplyOutlineToDropShadows=false
      bMiteredCorners=false
      bSeparateFillAlpha=false
      outlineMaterial=null
      outlineSize=1
      outlineColor=<rgba 0 0 0 0.5>
     }
    }
    shadowOffset=<0 0>
   }
   slot {class=SizeBoxSlot, properties {verticalAlignment=VAlign_Center, padding=<ltrb 0 2 0 2>}}
  }
  ObjectiveProgress {
   widgetClassPath=&VerticalBox
   parent=All
   widget=ObjectiveProgress
   slot {class=VerticalBoxSlot, properties {padding=<ltrb 6 0 6 0>}}
  }
  Friendly {
   widgetClassPath=&Overlay
   bIsVariable=true
   parent=ObjectiveProgress
   widget=Friendly
   slot {class=VerticalBoxSlot, properties {padding=<ltrb 0 0 0 6>}}
  }
  FriendlyBackingViewSwitcher {
   widgetClassPath=&CommonWidgetSwitcherLegacy
   bIsVariable=true
   parent=Friendly
   widget=FriendlyBackingViewSwitcher
   properties {activeWidgetIndex=1}
   slot {class=OverlaySlot, properties {horizontalAlignment=HAlign_Fill, verticalAlignment=VAlign_Fill}}
  }
  FriendlyBacking_IconView {
   widgetClassPath=&Border
   bIsVariable=true
   parent=FriendlyBackingViewSwitcher
   widget=FriendlyBacking_IconView
   properties {renderOpacity=0.8, visibility=SelfHitTestInvisible, brushColor=<rgba 0 0.2159 1 1>, padding=<ltrb 0 0 0 0>}
   slot {class=WidgetSwitcherSlot}
  }
  SizeBox_144 {
   widgetClassPath=&SizeBox
   parent=FriendlyBackingViewSwitcher
   widget=SizeBox_144
   properties {bOverride_HeightOverride=true, heightOverride=60}
   slot {class=WidgetSwitcherSlot}
  }
  FriendlyBacking_NumericView {
   widgetClassPath=&Image
   bIsVariable=true
   parent=SizeBox_144
   widget=FriendlyBacking_NumericView
   properties {
    renderOpacity=0.75
    visibility=SelfHitTestInvisible
    brush {
     bIsDynamicallyLoaded=false
     drawAs=Image
     imageType=NoImage
     mirroring=NoMirror
     resourceName=None
     resourceObject=/Game/Athena/HUD/LTM/Playground/Materials/M_UI_HUD_ObjectiveNumericView_Backing_Friendly
     tiling=NoTile
     imageSize=<32 32>
     margin=<ltrb 0 0 0 0>
     outlineSettings {
      bUseBrushTransparency=false
      roundingType=HalfHeightRadius
      width=0
      color {colorUseRule=UseColor_Specified, specifiedColor=<rgba 0 0 0 0>}
      cornerRadii=<0 0 0 0>
     }
     tintColor {colorUseRule=UseColor_Specified, specifiedColor=<rgba 1 1 1 1>}
     uVRegion {bIsValid=false, max=<0 0>, min=<0 0>}
    }
   }
   slot {class=SizeBoxSlot}
  }
  FriendlyObjectiveBox {
   widgetClassPath=&SizeBox
   bIsVariable=true
   parent=Friendly
   widget=FriendlyObjectiveBox
   properties {heightOverride=64, minDesiredHeight=100}
   slot {class=OverlaySlot, properties {horizontalAlignment=HAlign_Fill, verticalAlignment=VAlign_Fill}}
  }
  HorizontalBox_171 {widgetClassPath=&HorizontalBox, parent=FriendlyObjectiveBox, widget=HorizontalBox_171, slot {class=SizeBoxSlot}}
  FriendlyObjectivesView {
   widgetClassPath=&CommonWidgetSwitcherLegacy
   bIsVariable=true
   parent=HorizontalBox_171
   widget=FriendlyObjectivesView
   slot {class=HorizontalBoxSlot, properties {verticalAlignment=VAlign_Center, padding=<ltrb 0 4 0 4>, size {sizeRule=Fill, value=1}}}
  }
  FriendlyIconsBox {
   widgetClassPath=&HorizontalBox
   bIsVariable=true
   parent=FriendlyObjectivesView
   widget=FriendlyIconsBox
   slot {class=WidgetSwitcherSlot, properties {horizontalAlignment=HAlign_Center}}
  }
  FriendlyIcon1 {
   widgetClassPath=&/MinigameHUD_ObjectiveProgressIcon
   bIsVariable=true
   parent=FriendlyIconsBox
   widget=FriendlyIcon1
   properties {padding=<ltrb 2 0 2 0>}
   slot {class=HorizontalBoxSlot, properties {horizontalAlignment=HAlign_Center, verticalAlignment=VAlign_Center}}
  }
  FriendlyIcon2 {
   widgetClassPath=&MinigameHUD_ObjectiveProgressIcon
   bIsVariable=true
   parent=FriendlyIconsBox
   widget=FriendlyIcon2
   properties {padding=<ltrb 2 0 2 0>}
   slot {class=HorizontalBoxSlot, properties {horizontalAlignment=HAlign_Center, verticalAlignment=VAlign_Center}}
  }
  FriendlyIcon3 {
   widgetClassPath=&MinigameHUD_ObjectiveProgressIcon
   bIsVariable=true
   parent=FriendlyIconsBox
   widget=FriendlyIcon3
   properties {padding=<ltrb 2 0 2 0>}
   slot {class=HorizontalBoxSlot, properties {horizontalAlignment=HAlign_Center, verticalAlignment=VAlign_Center}}
  }
  FriendlyIcon4 {
   widgetClassPath=&MinigameHUD_ObjectiveProgressIcon
   bIsVariable=true
   parent=FriendlyIconsBox
   widget=FriendlyIcon4
   properties {padding=<ltrb 2 0 2 0>}
   slot {class=HorizontalBoxSlot, properties {horizontalAlignment=HAlign_Center, verticalAlignment=VAlign_Center}}
  }
  FriendlyNumericBox {
   widgetClassPath=&SizeBox
   bIsVariable=true
   parent=FriendlyObjectivesView
   widget=FriendlyNumericBox
   properties {bOverride_HeightOverride=true, heightOverride=40}
   slot {class=WidgetSwitcherSlot, properties {padding=<ltrb 0 0 10 0>}}
  }
  FriendlyNumericView {
   widgetClassPath=&HorizontalBox
   parent=FriendlyNumericBox
   widget=FriendlyNumericView
   slot {class=SizeBoxSlot, properties {horizontalAlignment=HAlign_Right}}
  }
  FriendlyNumerator {
   widgetClassPath=&CommonTextBlock
   bIsVariable=true
   parent=FriendlyNumericView
   widget=FriendlyNumerator
   properties {
    justification=Right
    text="{N}"
    colorAndOpacity {colorUseRule=UseColor_Specified, specifiedColor=<rgba 0.2159 1 1 1>}
    font {
     bForceMonospaced=false
     bMaterialIsStencil=false
     fontMaterial=null
     fontObject=&Burbank
     letterSpacing=0
     monospacedWidth=1
     size=32
     skewAmount=0
     typefaceFontName=Burbank-Black
     outlineSettings {
      bApplyOutlineToDropShadows=false
      bMiteredCorners=false
      bSeparateFillAlpha=false
      outlineMaterial=null
      outlineSize=0
      outlineColor=<rgba 0 0 0 1>
     }
    }
    renderTransform {angle=0, scale=<1 1>, shear=<-10 0>, translation=<0 -4>}
    shadowOffset=<0 0>
   }
   slot {class=HorizontalBoxSlot, properties {horizontalAlignment=HAlign_Right, verticalAlignment=VAlign_Center, padding=<ltrb 0 0 2 0>}}
  }
  CommonTextBlock_514 {
   widgetClassPath=&CommonTextBlock
   parent=FriendlyNumericView
   widget=CommonTextBlock_514
   properties {
    justification=Right
    text=/
    colorAndOpacity {colorUseRule=UseColor_Specified, specifiedColor=<rgba 0.004 0.0232 0.2159 1>}
    font {
     bForceMonospaced=false
     bMaterialIsStencil=false
     fontMaterial=null
     fontObject=&BurbankSmall
     letterSpacing=0
     monospacedWidth=1
     size=18
     skewAmount=0
     typefaceFontName=Bold
     outlineSettings {
      bApplyOutlineToDropShadows=false
      bMiteredCorners=false
      bSeparateFillAlpha=false
      outlineMaterial=null
      outlineSize=0
      outlineColor=<rgba 0 0 0 1>
     }
    }
    renderTransform {angle=0, scale=<1 1>, shear=<-2 0>, translation=<0 2>}
    shadowOffset=<0 0>
   }
   slot {class=HorizontalBoxSlot, properties {horizontalAlignment=HAlign_Right, verticalAlignment=VAlign_Bottom}}
  }
  FriendlyDenominator {
   widgetClassPath=&CommonTextBlock
   bIsVariable=true
   parent=FriendlyNumericView
   widget=FriendlyDenominator
   properties {
    justification=Right
    text="{n}"
    colorAndOpacity {colorUseRule=UseColor_Specified, specifiedColor=<rgba 0.004 0.0232 0.2159 1>}
    font {
     bForceMonospaced=false
     bMaterialIsStencil=false
     fontMaterial=null
     fontObject=&BurbankSmall
     letterSpacing=0
     monospacedWidth=1
     size=18
     skewAmount=0
     typefaceFontName=Black
     outlineSettings {
      bApplyOutlineToDropShadows=false
      bMiteredCorners=false
      bSeparateFillAlpha=false
      outlineMaterial=null
      outlineSize=0
      outlineColor=<rgba 0 0 0 1>
     }
    }
    renderTransform {angle=0, scale=<1 1>, shear=<-8 0>, translation=<0 2>}
    shadowOffset=<0 0>
   }
   slot {class=HorizontalBoxSlot, properties {horizontalAlignment=HAlign_Right, verticalAlignment=VAlign_Bottom}}
  }
  OverlayDefendObjectivemarker {
   widgetClassPath=&Overlay
   parent=Friendly
   widget=OverlayDefendObjectivemarker
   slot {class=OverlaySlot, properties {verticalAlignment=VAlign_Center}}
  }
  Overlay_103 {
   widgetClassPath=&Overlay
   parent=OverlayDefendObjectivemarker
   widget=Overlay_103
   properties {renderTransform {angle=0, scale=<1 1>, shear=<0 0>, translation=<-26 0>}}
   slot {class=OverlaySlot, properties {verticalAlignment=VAlign_Center}}
  }
  SizeBox_1 {
   widgetClassPath=&SizeBox
   parent=Overlay_103
   widget=SizeBox_1
   properties {bOverride_HeightOverride=true, bOverride_WidthOverride=true, heightOverride=40, widthOverride=40}
   slot {class=OverlaySlot}
  }
  DefendObjectiveBacking {
   widgetClassPath=&Image
   bIsVariable=true
   parent=SizeBox_1
   widget=DefendObjectiveBacking
   properties {brush {
     bIsDynamicallyLoaded=false
     drawAs=Image
     imageType=NoImage
     mirroring=NoMirror
     resourceName=None
     resourceObject=/Game/Athena/HUD/LTM/Playground/Materials/M_UI_CreativeObjectivePositioning_Backing_Defend
     tiling=NoTile
     imageSize=<32 32>
     margin=<ltrb 0 0 0 0>
     outlineSettings {
      bUseBrushTransparency=false
      roundingType=HalfHeightRadius
      width=0
      color {colorUseRule=UseColor_Specified, specifiedColor=<rgba 0 0 0 0>}
      cornerRadii=<0 0 0 0>
     }
     tintColor {colorUseRule=UseColor_Specified, specifiedColor=<rgba 1 1 1 1>}
     uVRegion {bIsValid=false, max=<0 0>, min=<0 0>}
    }}
   slot {class=SizeBoxSlot}
  }
  LocalPlayer_TeamMarker {
   widgetClassPath=&Image
   bIsVariable=true
   parent=Overlay_103
   widget=LocalPlayer_TeamMarker
   properties {
    visibility=SelfHitTestInvisible
    brush {
     bIsDynamicallyLoaded=false
     drawAs=Image
     imageType=NoImage
     mirroring=NoMirror
     resourceName=None
     resourceObject=/Game/Athena/HUD/LTM/Playground/Materials/M_UI_CreativeObjective_Positioning_Defend
     tiling=NoTile
     imageSize=<22 24>
     margin=<ltrb 0 0 0 0>
     outlineSettings {
      bUseBrushTransparency=false
      roundingType=HalfHeightRadius
      width=0
      color {colorUseRule=UseColor_Specified, specifiedColor=<rgba 0 0 0 0>}
      cornerRadii=<0 0 0 0>
     }
     tintColor {colorUseRule=UseColor_Specified, specifiedColor=<rgba 1 1 1 1>}
     uVRegion {bIsValid=false, max=<0 0>, min=<0 0>}
    }
   }
   slot {class=OverlaySlot, properties {horizontalAlignment=HAlign_Center, verticalAlignment=VAlign_Center}}
  }
  PlayerName {
   widgetClassPath=&TextBlock
   bIsVariable=true
   parent=OverlayDefendObjectivemarker
   widget=PlayerName
   properties {
    autoWrapText=true
    lineHeightPercentage=0.75
    text=YOUR TEAM
    visibility=Collapsed
    colorAndOpacity {colorUseRule=UseColor_Specified, specifiedColor=<rgba 0 0 0 1>}
    font {
     bForceMonospaced=false
     bMaterialIsStencil=false
     fontMaterial=null
     fontObject=&Burbank
     letterSpacing=0
     monospacedWidth=1
     size=18
     skewAmount=0
     typefaceFontName=Burbank-PlayerPower
     outlineSettings {
      bApplyOutlineToDropShadows=false
      bMiteredCorners=false
      bSeparateFillAlpha=false
      outlineMaterial=null
      outlineSize=1
      outlineColor=<rgba 1 1 1 0.8>
     }
    }
   }
   slot {class=OverlaySlot}
  }
  Enemy {
   widgetClassPath=&Overlay
   bIsVariable=true
   parent=ObjectiveProgress
   widget=Enemy
   slot {class=VerticalBoxSlot, properties {padding=<ltrb 0 0 0 6>}}
  }
  EnemyBackingViewSwitcher {
   widgetClassPath=/Script/CommonUILegacy.CommonWidgetStackLegacy
   bIsVariable=true
   parent=Enemy
   widget=EnemyBackingViewSwitcher
   properties {activeWidgetIndex=1}
   slot {class=OverlaySlot, properties {horizontalAlignment=HAlign_Fill, verticalAlignment=VAlign_Fill}}
  }
  EnemyBacking_IconView {
   widgetClassPath=&Border
   bIsVariable=true
   parent=EnemyBackingViewSwitcher
   widget=EnemyBacking_IconView
   properties {renderOpacity=0.8, visibility=SelfHitTestInvisible, brushColor=<rgba 0.8308 0.0168 0.0103 1>, padding=<ltrb 0 0 0 0>}
   slot {class=WidgetSwitcherSlot}
  }
  SizeBox_143 {
   widgetClassPath=&SizeBox
   parent=EnemyBackingViewSwitcher
   widget=SizeBox_143
   properties {bOverride_HeightOverride=true, heightOverride=60}
   slot {class=WidgetSwitcherSlot}
  }
  EnemyBacking_NumericView {
   widgetClassPath=&Image
   bIsVariable=true
   parent=SizeBox_143
   widget=EnemyBacking_NumericView
   properties {
    renderOpacity=0.75
    visibility=SelfHitTestInvisible
    brush {
     bIsDynamicallyLoaded=false
     drawAs=Image
     imageType=NoImage
     mirroring=NoMirror
     resourceName=None
     resourceObject=/Game/Athena/HUD/LTM/Playground/Materials/M_UI_HUD_ObjectiveNumericView_Backing_Enemy
     tiling=NoTile
     imageSize=<32 32>
     margin=<ltrb 0 0 0 0>
     outlineSettings {
      bUseBrushTransparency=false
      roundingType=HalfHeightRadius
      width=0
      color {colorUseRule=UseColor_Specified, specifiedColor=<rgba 0 0 0 0>}
      cornerRadii=<0 0 0 0>
     }
     tintColor {colorUseRule=UseColor_Specified, specifiedColor=<rgba 1 1 1 1>}
     uVRegion {bIsValid=false, max=<0 0>, min=<0 0>}
    }
   }
   slot {class=SizeBoxSlot}
  }
  EnemyObjectiveBox {
   widgetClassPath=&SizeBox
   bIsVariable=true
   parent=Enemy
   widget=EnemyObjectiveBox
   properties {minDesiredHeight=100}
   slot {class=OverlaySlot, properties {horizontalAlignment=HAlign_Fill, verticalAlignment=VAlign_Fill}}
  }
  HorizontalBox_92 {widgetClassPath=&HorizontalBox, parent=EnemyObjectiveBox, widget=HorizontalBox_92, slot {class=SizeBoxSlot}}
  EnemyObjectivesView {
   widgetClassPath=&CommonWidgetSwitcherLegacy
   bIsVariable=true
   parent=HorizontalBox_92
   widget=EnemyObjectivesView
   slot {class=HorizontalBoxSlot, properties {verticalAlignment=VAlign_Center, padding=<ltrb 0 4 0 4>, size {sizeRule=Fill, value=1}}}
  }
  EnemyIconsBox {
   widgetClassPath=&HorizontalBox
   bIsVariable=true
   parent=EnemyObjectivesView
   widget=EnemyIconsBox
   slot {class=WidgetSwitcherSlot, properties {horizontalAlignment=HAlign_Center}}
  }
  EnemyIcon1 {
   widgetClassPath=&MinigameHUD_ObjectiveProgressIcon
   bIsVariable=true
   parent=EnemyIconsBox
   widget=EnemyIcon1
   properties {padding=<ltrb 2 0 2 0>}
   slot {class=HorizontalBoxSlot, properties {horizontalAlignment=HAlign_Center}}
  }
  EnemyIcon2 {
   widgetClassPath=&MinigameHUD_ObjectiveProgressIcon
   bIsVariable=true
   parent=EnemyIconsBox
   widget=EnemyIcon2
   properties {padding=<ltrb 2 0 2 0>}
   slot {class=HorizontalBoxSlot, properties {horizontalAlignment=HAlign_Center, verticalAlignment=VAlign_Center}}
  }
  EnemyIcon3 {
   widgetClassPath=&MinigameHUD_ObjectiveProgressIcon
   bIsVariable=true
   parent=EnemyIconsBox
   widget=EnemyIcon3
   properties {padding=<ltrb 2 0 2 0>}
   slot {class=HorizontalBoxSlot, properties {horizontalAlignment=HAlign_Center, verticalAlignment=VAlign_Center}}
  }
  EnemyIcon4 {
   widgetClassPath=&MinigameHUD_ObjectiveProgressIcon
   bIsVariable=true
   parent=EnemyIconsBox
   widget=EnemyIcon4
   properties {padding=<ltrb 2 0 2 0>}
   slot {class=HorizontalBoxSlot, properties {horizontalAlignment=HAlign_Center, verticalAlignment=VAlign_Center}}
  }
  EnemyNumericBox {
   widgetClassPath=&SizeBox
   bIsVariable=true
   parent=EnemyObjectivesView
   widget=EnemyNumericBox
   properties {bOverride_HeightOverride=true, heightOverride=40}
   slot {class=WidgetSwitcherSlot, properties {padding=<ltrb 0 0 10 0>}}
  }
  EnemyNumericView {
   widgetClassPath=&HorizontalBox
   parent=EnemyNumericBox
   widget=EnemyNumericView
   slot {class=SizeBoxSlot, properties {horizontalAlignment=HAlign_Right}}
  }
  EnemyNumerator {
   widgetClassPath=&CommonTextBlock
   bIsVariable=true
   parent=EnemyNumericView
   widget=EnemyNumerator
   properties {
    justification=Right
    text="{N}"
    colorAndOpacity {colorUseRule=UseColor_Specified, specifiedColor=<rgba 1 0.6654 0.2159 1>}
    font {
     bForceMonospaced=false
     bMaterialIsStencil=false
     fontMaterial=null
     fontObject=&Burbank
     letterSpacing=0
     monospacedWidth=1
     size=32
     skewAmount=0
     typefaceFontName=Burbank-Black
     outlineSettings {
      bApplyOutlineToDropShadows=false
      bMiteredCorners=false
      bSeparateFillAlpha=false
      outlineMaterial=null
      outlineSize=0
      outlineColor=<rgba 0 0 0 1>
     }
    }
    renderTransform {angle=0, scale=<1 1>, shear=<-10 0>, translation=<0 -4>}
    shadowOffset=<0 0>
   }
   slot {class=HorizontalBoxSlot, properties {horizontalAlignment=HAlign_Right, verticalAlignment=VAlign_Center, padding=<ltrb 0 0 2 0>}}
  }
  CommonTextBlock_517 {
   widgetClassPath=&CommonTextBlock
   parent=EnemyNumericView
   widget=CommonTextBlock_517
   properties {
    justification=Right
    text=/
    colorAndOpacity {colorUseRule=UseColor_Specified, specifiedColor=<rgba 0.0999 0.0027 0.0027 1>}
    font {
     bForceMonospaced=false
     bMaterialIsStencil=false
     fontMaterial=null
     fontObject=&BurbankSmall
     letterSpacing=0
     monospacedWidth=1
     size=18
     skewAmount=0
     typefaceFontName=Bold
     outlineSettings {
      bApplyOutlineToDropShadows=false
      bMiteredCorners=false
      bSeparateFillAlpha=false
      outlineMaterial=null
      outlineSize=0
      outlineColor=<rgba 0 0 0 1>
     }
    }
    renderTransform {angle=0, scale=<1 1>, shear=<-2 0>, translation=<0 2>}
    shadowOffset=<0 0>
   }
   slot {class=HorizontalBoxSlot, properties {horizontalAlignment=HAlign_Right, verticalAlignment=VAlign_Bottom}}
  }
  EnemyDenominator {
   widgetClassPath=&CommonTextBlock
   bIsVariable=true
   parent=EnemyNumericView
   widget=EnemyDenominator
   properties {
    justification=Right
    text="{n}"
    colorAndOpacity {colorUseRule=UseColor_Specified, specifiedColor=<rgba 0.0999 0.0027 0.0027 1>}
    font {
     bForceMonospaced=false
     bMaterialIsStencil=false
     fontMaterial=null
     fontObject=&BurbankSmall
     letterSpacing=0
     monospacedWidth=1
     size=18
     skewAmount=0
     typefaceFontName=Black
     outlineSettings {
      bApplyOutlineToDropShadows=false
      bMiteredCorners=false
      bSeparateFillAlpha=false
      outlineMaterial=null
      outlineSize=0
      outlineColor=<rgba 0 0 0 1>
     }
    }
    renderTransform {angle=0, scale=<1 1>, shear=<-8 0>, translation=<0 2>}
    shadowOffset=<0 0>
   }
   slot {class=HorizontalBoxSlot, properties {horizontalAlignment=HAlign_Right, verticalAlignment=VAlign_Bottom}}
  }
  Overlay_105 {widgetClassPath=&Overlay, parent=Enemy, widget=Overlay_105, slot {class=OverlaySlot, properties {verticalAlignment=VAlign_Fill}}}
  Overlay_3 {
   widgetClassPath=&Overlay
   parent=Overlay_105
   widget=Overlay_3
   properties {renderTransform {angle=0, scale=<1 1>, shear=<0 0>, translation=<-26 0>}}
   slot {class=OverlaySlot, properties {verticalAlignment=VAlign_Center}}
  }
  SizeBox_141 {
   widgetClassPath=&SizeBox
   parent=Overlay_3
   widget=SizeBox_141
   properties {bOverride_HeightOverride=true, bOverride_WidthOverride=true, heightOverride=40, widthOverride=40}
   slot {class=OverlaySlot, properties {verticalAlignment=VAlign_Center}}
  }
  AttackObjectiveBacking {
   widgetClassPath=&Image
   bIsVariable=true
   parent=SizeBox_141
   widget=AttackObjectiveBacking
   properties {brush {
     bIsDynamicallyLoaded=false
     drawAs=Image
     imageType=NoImage
     mirroring=NoMirror
     resourceName=None
     resourceObject=/Game/Athena/HUD/LTM/Playground/Materials/M_UI_CreativeObjectivePositioning_Backing_Attack
     tiling=NoTile
     imageSize=<32 32>
     margin=<ltrb 0 0 0 0>
     outlineSettings {
      bUseBrushTransparency=false
      roundingType=HalfHeightRadius
      width=0
      color {colorUseRule=UseColor_Specified, specifiedColor=<rgba 0 0 0 0>}
      cornerRadii=<0 0 0 0>
     }
     tintColor {colorUseRule=UseColor_Specified, specifiedColor=<rgba 1 1 1 1>}
     uVRegion {bIsValid=false, max=<0 0>, min=<0 0>}
    }}
   slot {class=SizeBoxSlot}
  }
  Image_351 {
   widgetClassPath=&Image
   bIsVariable=true
   parent=Overlay_3
   widget=Image_351
   properties {
    brush {
     bIsDynamicallyLoaded=false
     drawAs=Image
     imageType=NoImage
     mirroring=NoMirror
     resourceName=None
     resourceObject=/Game/Athena/HUD/LTM/Playground/Materials/M_UI_CreativeObjective_Positioning_Attack
     tiling=NoTile
     imageSize=<28 28>
     margin=<ltrb 0 0 0 0>
     outlineSettings {
      bUseBrushTransparency=false
      roundingType=HalfHeightRadius
      width=0
      color {colorUseRule=UseColor_Specified, specifiedColor=<rgba 0 0 0 0>}
      cornerRadii=<0 0 0 0>
     }
     tintColor {colorUseRule=UseColor_Specified, specifiedColor=<rgba 1 1 1 1>}
     uVRegion {bIsValid=false, max=<0 0>, min=<0 0>}
    }
    renderTransform {angle=-10, scale=<1 1>, shear=<0 0>, translation=<0 0>}
   }
   slot {class=OverlaySlot, properties {horizontalAlignment=HAlign_Center, verticalAlignment=VAlign_Center}}
  }
  TextBlock_0 {
   widgetClassPath=&TextBlock
   bIsVariable=true
   parent=Overlay_105
   widget=TextBlock_0
   properties {
    autoWrapText=true
    lineHeightPercentage=0.75
    text=ENEMY TEAMS
    visibility=Collapsed
    colorAndOpacity {colorUseRule=UseColor_Specified, specifiedColor=<rgba 0 0 0 1>}
    font {
     bForceMonospaced=false
     bMaterialIsStencil=false
     fontMaterial=null
     fontObject=&Burbank
     letterSpacing=0
     monospacedWidth=1
     size=18
     skewAmount=0
     typefaceFontName=Burbank-PlayerPower
     outlineSettings {
      bApplyOutlineToDropShadows=false
      bMiteredCorners=false
      bSeparateFillAlpha=false
      outlineMaterial=null
      outlineSize=1
      outlineColor=<rgba 1 1 1 0.8>
     }
    }
   }
   slot {class=OverlaySlot}
  }
  TutorialHighlight {widgetClassPath=&TutorialHighlight}
 ]
}
