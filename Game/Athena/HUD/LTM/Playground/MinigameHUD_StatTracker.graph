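# Minigame HUD widget base that shows one minigame stat, score by default, in a numeric text block with animated count-up or count-down.
# Stat updates fetch the value from the minigame's stats component and animate toward it; team config changes and in-progress display changes snap to it.
# An event-repeater animation, speed-scaled by StatIncrementTime, serves as the timer for TickScore, which steps by whole numbers toward the target and stops the animation on arrival.
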WidgetBlueprint /Game/Athena/HUD/LTM/Playground/MinigameHUD_StatTracker
animations [ScoreTickAnimation {
  binding_count=0
  length_seconds=1
  object=&self:ScoreTickAnimation
  sequence {bindings [], tracks [{type=MovieSceneEventTrack, sections [{type=MovieSceneEventRepeaterSection, end_frame=20, start_frame=0}]}]}
 }]
graphs [
 AnimateNewScore {nodes [
   n1 = AnimateNewScore!() -> (then, NewScoreValue:double=~, bSkipAnimation:bool=~) {is=title K2Node_FunctionEntry}
   > n2 = Set NewScoreValue(NewScoreValue:double=@n1) -> (Output_Get:double)
   n3 = Get StatIncrementTime() -> (StatIncrementTime:double)
   n4 = Get DisplayedScore() -> (DisplayedScore:double)
   > n5 = Branch(Condition=OR Boolean!(
     A="Nearly Equal (Float)"(A=@n2, B=@n4, ErrorTolerance=1)
     B=@n1.bSkipAnimation
     C:bool="float <= float"!(A:double=@n3, B:double=0) -> (ReturnValue:bool) {is=title K2Node_CallFunction}
    ) -> (ReturnValue))
   > UpdateScoreUI!(execute, self, InUpdatedDisplayScore:double=@n2) -> (then)
   Set ScoreTickAnimationStopTriggered(execute=@n5.else, ScoreTickAnimationStopTriggered:bool=false) -> (Output_Get:bool)
   > PlayAnimation(
    InAnimation=Get ScoreTickAnimation() -> (ScoreTickAnimation:Widget Animation Object Reference)
    PlaybackSpeed=float / float(A=1, B=@n3) {is=title K2Node_CallFunction}
   )
  ]}
 EventGraph {nodes [
   Event MinigameTeamConfigUpdated() {is=title K2Node_CustomEvent}
   > UpdateNewScore!(execute, self, bSetOldScoreToNewScore:bool=true) -> (then)
   n1 = Event OnDisplayMinigameStartUI!() -> (OutputDelegate, then, bDisplay:bool) {is=title K2Node_Event}
   > Branch(Condition=@n1.bDisplay)
   n2 = Create Event()
   n3 = Get CurrentMinigame() -> (CurrentMinigame:Fort Minigame Object Reference)
   > Bind Event to Minigame Team Config Updated(self:Fort Minigame Object Reference=@n3, Delegate=@n2) {is=title K2Node_AddDelegate}
   n4 = Event OnDisplayMinigameInProgressUI!() -> (OutputDelegate, then, bDisplay:bool) {is=title K2Node_Event}
   > SetVisibility(InVisibility=Select(Option 0=Collapsed, Option 1=SelfHitTestInvisible, Index=@n4.bDisplay))
   > UpdateNewScore(bSetOldScoreToNewScore=true)
   n5 = Event OnDisplayMinigameEndUI!() -> (OutputDelegate, then, bDisplay:bool) {is=title K2Node_Event}
   > Branch(Condition=@n5.bDisplay)
   > Unbind Event from Minigame Team Config Updated(self:Fort Minigame Object Reference=@n3, Delegate=@n2) {is=title K2Node_RemoveDelegate}
   Event OnPlayerStatUpdated() {is=title K2Node_Event}
   > UpdateNewScore(bSetOldScoreToNewScore=false)
   Event PreConstruct()
   Event Construct()
  ]}
 GetIntegralMin {nodes [
   n1 = Inputs!() -> (A:double=~, A2:double=~)
   n2 = "Min (Float)"(A:double="Absolute (Float)"(A=@n1), B:double="Absolute (Float)"(A=@n1.A2)) -> (ReturnValue:double) {is=title K2Node_CommutativeAssociativeBinaryOperator}
   Outputs!(ReturnValue:double=int * float!(A:int=Ceil(A=@n2) {is=title K2Node_CallFunction}, B:double="Sign (Float)"!(A:double=@n1) -> (ReturnValue:double)) -> (ReturnValue:double))
  ]}
 GetScoreTickDuration {nodes [
   Inputs!()
   n1 = Get ScoreTickAnimation()
   Outputs(ReturnValue=float - float(A=GetEndTime(self=@n1), B=GetStartTime!(self:Widget Animation Object Reference=@n1) -> (ReturnValue:float)) {is=title K2Node_CallFunction})
  ]}
 TickScore {nodes [
   TickScore() {is=title K2Node_FunctionEntry}
   > n1 = Set DeltaScore(DeltaScore:double=float - float(A=Get NewScoreValue() -> (NewScoreValue:double), B=Get DisplayedScore())) -> (Output_Get:double)
   n2 = float / float(
    A=GetAnimationCurrentTime!(self:User Widget Object Reference, InAnimation:Widget Animation Object Reference=Get ScoreTickAnimation()) -> (ReturnValue:float)
    B=Get Score Tick Duration!() -> (ReturnValue:double=~) {is=title K2Node_MacroInstance}
   )
   > n3 = Set DeltaScoreT(DeltaScoreT:double=float * float(A=@n1, B="Clamp (Float)"(Value=@n2, Min=0, Max=1))) -> (Output_Get:double)
   > n4 = Branch(Condition="Nearly Equal (Float)"(A=@n1, B=0, ErrorTolerance=1))
   > UpdateScoreUI(InUpdatedDisplayScore=Get NewScoreValue())
   > n5 = Branch(Condition=Get ScoreTickAnimationStopTriggered() -> (ScoreTickAnimationStopTriggered:bool))
   StopAnimation(execute=@n5.else, InAnimation=Get ScoreTickAnimation())
   > Set ScoreTickAnimationStopTriggered(ScoreTickAnimationStopTriggered=true)
   UpdateScoreUI(
    execute=@n4.else
    InUpdatedDisplayScore=float + float(A=Get DisplayedScore(), B=Get Integral Min!(A:double=@n1, A2:double=@n3) -> (ReturnValue:double=~) {is=title K2Node_MacroInstance}) {is=title K2Node_CommutativeAssociativeBinaryOperator}
   )
  ]}
 UpdateNewScore {nodes [
   n1 = UpdateNewScore!() -> (then, bSetOldScoreToNewScore:bool=~) {is=title K2Node_FunctionEntry}
   > n2 = Get!(execute:Exec) -> (then:Exec, else:Exec, CurrentMinigame:Fort Minigame Object Reference)
   n3 = GetLoadedStatFromClass!(
    self:Fort Minigame Stats Component Object Reference=GetMinigameStatsComponent!(ContextActor:Actor Object Reference=Get Volume!(self:Fort Minigame Object Reference=@n2) -> (Volume:Actor Object Reference)) -> (ReturnValue:Fort Minigame Stats Component Object Reference)
    InFilterClass:Fort Minigame Stat Filter Class Reference=Get DisplayedStat() -> (DisplayedStat:Fort Minigame Stat Filter Class Reference)
   ) -> (ReturnValue:Fort Minigame Stat Filter Object Reference)
   > AnimateNewScore!(
    execute
    self
    NewScoreValue:double="To Float (Integer)"(InInt=GetStatValue!(self:Minigame Widget Base Object Reference, InFilter:Fort Minigame Stat Filter Object Reference=@n3, bGameTotal:bool=false) -> (ReturnValue:int=0))
    bSkipAnimation:bool=@n1
   ) -> (then) {is=title K2Node_CallFunction}
  ]}
 UpdateScoreUI {nodes [
   n1 = UpdateScoreUI!() -> (then, InUpdatedDisplayScore:double=~) {is=title K2Node_FunctionEntry}
   > n2 = Set DisplayedScore(DisplayedScore:double=@n1) -> (Output_Get:double)
   > SetCurrentValue!(
    execute
    self:Common Numeric Text Block Object Reference=Get StatDisplay() -> (StatDisplay:Common Numeric Text Block Object Reference)
    NewValue:float=@n2
   ) -> (then)
  ]}
]
variables [
 DisplayedScore {type=number}
 DisplayedStat {type=Class, default=/Game/Athena/Playset/Minigames/StatFilters/MinigameStat_Score.MinigameStat_Score_C}
 NewScoreValue {type=number}
 ScoreTickAnimationStopTriggered {type=boolean}
 StatDisplay {type=CommonNumericTextBlock, default=null}
 StatIncrementTime {type=number, default=2}
]
widgets {info {inheritedWidgetCount=0, parentClass=/Script/FortniteUI.MinigameWidgetBase, rootWidgetClass=null, widgetCount=0}, widgets []}
