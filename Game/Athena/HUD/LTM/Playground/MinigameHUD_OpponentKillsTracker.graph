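# Playground minigame HUD widget for the opponent kills tracker, showing a kill icon beside a numeric count.
# It subclasses MinigameHUD_StatTracker and points the inherited StatDisplay at its KillCount text block in PreConstruct, leaving Construct empty.
# Two helper functions, not called here, return the ScoreTickAnimation length and the rounded-up smaller magnitude of two values, signed like the first.
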WidgetBlueprint /Game/Athena/HUD/LTM/Playground/MinigameHUD_OpponentKillsTracker
graphs [
 EventGraph {nodes [
   Event PreConstruct()
   > Set StatDisplay(StatDisplay:Common Numeric Text Block Object Reference=Get KillCount() -> (KillCount:Common Numeric Text Block Object Reference)) -> (Output_Get:Common Numeric Text Block Object Reference)
   Event Construct()
  ]}
 GetIntegralMin {nodes [
   n1 = Inputs!() -> (A:double=~, A2:double=~)
   n2 = "Min (Float)"(A:double="Absolute (Float)"(A=@n1), B:double="Absolute (Float)"(A=@n1.A2)) -> (ReturnValue:double) {is=title K2Node_CommutativeAssociativeBinaryOperator}
   Outputs!(ReturnValue:double=int * float!(A:int=Ceil(A=@n2) {is=title K2Node_CallFunction}, B:double="Sign (Float)"!(A:double=@n1) -> (ReturnValue:double)) -> (ReturnValue:double))
  ]}
 GetScoreTickDuration {nodes [
   Inputs!()
   n1 = Get ScoreTickAnimation() -> (ScoreTickAnimation:Widget Animation Object Reference)
   Outputs(ReturnValue=float - float(A=GetEndTime(self=@n1), B=GetStartTime!(self:Widget Animation Object Reference=@n1) -> (ReturnValue:float)) {is=title K2Node_CallFunction})
  ]}
]
variables []
widgets {
 info {
  inheritedWidgetCount=1
  parentClass=&/MinigameHUD_StatTracker.MinigameHUD_StatTracker_C
  rootWidgetClass=&Overlay
  widgetCount=5
 }
 widgets [
  Overlay_0 {widgetClassPath=&Overlay, widget=Overlay_0}
  HorizontalBox_0 {
   widgetClassPath=&HorizontalBox
   parent=Overlay_0
   widget=HorizontalBox_0
   slot {class=OverlaySlot, properties {horizontalAlignment=HAlign_Center, verticalAlignment=VAlign_Center}}
  }
  KillsImage {
   widgetClassPath=&Image
   parent=HorizontalBox_0
   widget=KillsImage
   properties {brush {
     bIsDynamicallyLoaded=false
     drawAs=Image
     imageType=NoImage
     mirroring=NoMirror
     resourceName=None
     resourceObject=/Game/Athena/HUD/Art/T-Deimos-KillIco
     tiling=NoTile
     imageSize=<36 36>
     margin=<ltrb 0 0 0 0>
     outlineSettings {
      bUseBrushTransparency=false
      roundingType=HalfHeightRadius
      width=0
      color {colorUseRule=UseColor_Specified, specifiedColor=<rgba 0 0 0 0>}
      cornerRadii=<0 0 0 0>
     }
     tintColor {colorUseRule=UseColor_Specified, specifiedColor=<rgba 1 1 1 1>}
     uVRegion {bIsValid=false, max=<0 0>, min=<0 0>}
    }}
   slot {class=HorizontalBoxSlot, properties {horizontalAlignment=HAlign_Center, verticalAlignment=VAlign_Center}}
  }
  KillCount {
   widgetClassPath=/Script/CommonUI.CommonNumericTextBlock
   bIsVariable=true
   parent=HorizontalBox_0
   widget=KillCount
   properties {
    currentNumericValue=99
    justification=Right
    text=0
    font {
     bForceMonospaced=false
     bMaterialIsStencil=false
     fontMaterial=null
     fontObject=&NotoSans
     letterSpacing=0
     monospacedWidth=1
     size=20
     skewAmount=0
     typefaceFontName=NotoSans-Bold
     outlineSettings {
      bApplyOutlineToDropShadows=true
      bMiteredCorners=false
      bSeparateFillAlpha=false
      outlineMaterial=null
      outlineSize=1
      outlineColor=<rgba 0 0 0 0.7>
     }
    }
    shadowColorAndOpacity=<rgba 0.0156 0.0082 0.0063 0.8>
   }
   slot {class=HorizontalBoxSlot, properties {padding=<ltrb 2 0 0 0>}}
  }
  TutorialHighlight {widgetClassPath=&TutorialHighlight}
 ]
}
