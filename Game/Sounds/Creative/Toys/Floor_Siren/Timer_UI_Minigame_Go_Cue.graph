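# Sound cue for the "Go" signal of a minigame timer UI, stored with the Creative floor siren toy sounds.
# A single wave player plays a race start beep borrowed from the gameplay mission sounds at reduced volume.
# It uses the critical in-game UI sound class, and a concurrency override allows one instance at a time, blocking new ones while it plays.
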SoundCue /Game/Sounds/Creative/Toys/Floor_Siren/Timer_UI_Minigame_Go_Cue
root=WavePlayer_1
nodes [{id=WavePlayer_1, properties {soundWaveAssetPtr=/Game/Sounds/Fort_GamePlay_Sounds/Missions/Race_Beep_Start_01}}]
properties {
 bOverrideConcurrency=true
 maxDistance=2097152
 soundClassObject=&Fort_UI_InGame_Critical
 volumeMultiplier=0.6
 concurrencyOverrides {maxCount=1, resolutionRule=PreventNew}
}
