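# Default_Start_Cue is a SoundCue in the Fort default sounds folder, routed to the in-game critical UI sound class.
# It plays a single wave, the front-end vbuck_hover_select_01 sound, through one Wave Player node with no randomization, mixing or modulation.
# Its max distance is set very large, at about two million units.
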SoundCue /Game/Sounds/Fort_Default_Sounds/Default_Start_Cue
root=WavePlayer_0
nodes [{id=WavePlayer_0, properties {soundWaveAssetPtr=/Game/Sounds/Fort_UI_Sounds/FrontEnd/vbuck_hover_select_01}}]
properties {maxDistance=2097152, soundClassObject=&Fort_UI_InGame_Critical}
