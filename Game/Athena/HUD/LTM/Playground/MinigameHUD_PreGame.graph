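# Minigame pre-game HUD widget showing a letterboxed "game starts" or "round begins" countdown that ends on GO.
# OnDisplayMinigameStartUI hides loading-tagged HUD elements, seeds a one-second looping timer from warmup time, and pushes a sound mix.
# Each tick updates text and sound, animating the last ten seconds before OnGo plays.
# Dismissal restores HUD, reverses the bars, then collapses the overlay; the minigame's FNEUIComponent can hide the bars.
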WidgetBlueprint /Game/Athena/HUD/LTM/Playground/MinigameHUD_PreGame
animations [
 Tickown {
  binding_count=4
  length_seconds=0.95
  object=&self:Tickown
  sequence {bindings [
    Glow {
     id=#1
     possessed_class=Image
     tracks [
      {
       type=MovieSceneFloatTrack
       display_name=Render Opacity
       sections [{
         type=MovieSceneFloatSection
         end_frame=19
         start_frame=0
         channels [None {type=MovieSceneScriptingFloatChannel, keys [[0,0],[8,0],[9,0.772],[19,0]]}]
        }]
      }
      {
       type=MovieSceneByteTrack
       display_name=Visibility
       sections [{type=MovieSceneByteSection, end_frame=19, start_frame=0, channels [None {type=MovieSceneScriptingByteChannel, keys [[0,4],[19,2]]}]}]
      }
      {
       type=MovieScene2DTransformTrack
       sections [{
         type=MovieScene2DTransformSection
         end_frame=19
         start_frame=0
         channels [
          Translation.X {type=MovieSceneScriptingFloatChannel}
          Translation.Y {type=MovieSceneScriptingFloatChannel}
          Angle {type=MovieSceneScriptingFloatChannel}
          Scale.X {type=MovieSceneScriptingFloatChannel}
          Scale.Y {type=MovieSceneScriptingFloatChannel}
          Shear.X {type=MovieSceneScriptingFloatChannel}
          Shear.Y {type=MovieSceneScriptingFloatChannel}
         ]
        }]
      }
     ]
    }
    CountdownTextHighlight {
     id=#2
     possessed_class=CommonTextBlock
     tracks [
      {
       type=MovieScene2DTransformTrack
       sections [{
         type=MovieScene2DTransformSection
         end_frame=19
         start_frame=8
         channels [
          Translation.X {type=MovieSceneScriptingFloatChannel}
          Translation.Y {type=MovieSceneScriptingFloatChannel}
          Angle {type=MovieSceneScriptingFloatChannel}
          Scale.X {type=MovieSceneScriptingFloatChannel, keys [[8,1],[19,1.6638]]}
          Scale.Y {type=MovieSceneScriptingFloatChannel, keys [[8,1],[19,1.3622]]}
          Shear.X {type=MovieSceneScriptingFloatChannel}
          Shear.Y {type=MovieSceneScriptingFloatChannel}
         ]
        }]
      }
      {
       type=MovieSceneFloatTrack
       display_name=Render Opacity
       sections [{
         type=MovieSceneFloatSection
         end_frame=19
         start_frame=8
         channels [None {type=MovieSceneScriptingFloatChannel, keys [[8,0],[9,0.5722],[19,0]]}]
        }]
      }
      {
       type=MovieSceneByteTrack
       display_name=Visibility
       sections [{
         type=MovieSceneByteSection
         end_frame=19
         start_frame=0
         channels [None {type=MovieSceneScriptingByteChannel, keys [[0,2],[7.732,4],[19,2]]}]
        }]
      }
     ]
    }
    StartGameText {
     id=#3
     possessed_class=CommonRichTextBlock
     tracks [{
       type=MovieSceneWidgetMaterialTrack
       display_name=Default Text Style Override.Font.Outline Settings.OutlineMaterial
       sections [{
         type=MovieSceneParameterSection
         end_frame=19
         start_frame=0
         channels [YValue {type=MovieSceneScriptingFloatChannel, keys [[0,1.1],[7,1.1],[19,-0.1]]}]
        }]
      }]
    }
    CountdownText {
     id=#4
     possessed_class=CommonTextBlock
     tracks [
      {
       type=MovieScene2DTransformTrack
       sections [{
         type=MovieScene2DTransformSection
         end_frame=19
         start_frame=0
         channels [
          Translation.X {type=MovieSceneScriptingFloatChannel}
          Translation.Y {type=MovieSceneScriptingFloatChannel}
          Angle {type=MovieSceneScriptingFloatChannel}
          Scale.X {type=MovieSceneScriptingFloatChannel, keys [[0,1.3],[5,1.3],[8,1],[19,1]]}
          Scale.Y {type=MovieSceneScriptingFloatChannel, keys [[0,1.3],[5,1.3],[8,1],[19,1]]}
          Shear.X {type=MovieSceneScriptingFloatChannel}
          Shear.Y {type=MovieSceneScriptingFloatChannel}
         ]
        }]
      }
      {
       type=MovieSceneFloatTrack
       display_name=Render Opacity
       sections [{type=MovieSceneFloatSection, end_frame=19, start_frame=0, channels [None {type=MovieSceneScriptingFloatChannel, keys [[0,1],[12,1],[19,0]]}]}]
      }
      {
       type=MovieSceneColorTrack
       display_name=Color and Opacity
       sections [{
         type=MovieSceneColorSection
         end_frame=10
         start_frame=0
         channels [
          Color.R {type=MovieSceneScriptingFloatChannel, keys [[0,1],[8.083,1],[9,0.2843]]}
          Color.G {type=MovieSceneScriptingFloatChannel, keys [[0,1],[8.083,1],[8.3823,0.917],[9,0.9768]]}
          Color.B {type=MovieSceneScriptingFloatChannel, const [1,4]}
          Color.A {type=MovieSceneScriptingFloatChannel, const [1,4]}
         ]
        }]
      }
     ]
    }
   ]}
 }
 LetterBoxIn {
  binding_count=2
  length_seconds=0.25
  object=&self:LetterBoxIn
  sequence {bindings [
    Image_LetterBoxTop {
     id=#5
     possessed_class=Image
     tracks [{
       type=MovieScene2DTransformTrack
       sections [{
         type=MovieScene2DTransformSection
         end_frame=5
         start_frame=0
         channels [
          Translation.X {type=MovieSceneScriptingFloatChannel, const [0,2]}
          Translation.Y {type=MovieSceneScriptingFloatChannel, keys [[0,-169],[5,-35]]}
          Angle {type=MovieSceneScriptingFloatChannel}
          Scale.X {type=MovieSceneScriptingFloatChannel}
          Scale.Y {type=MovieSceneScriptingFloatChannel}
          Shear.X {type=MovieSceneScriptingFloatChannel}
          Shear.Y {type=MovieSceneScriptingFloatChannel}
         ]
        }]
      }]
    }
    Image_LetterboxBottom {
     id=#6
     possessed_class=Image
     tracks [{
       type=MovieScene2DTransformTrack
       sections [{
         type=MovieScene2DTransformSection
         end_frame=5
         start_frame=0
         channels [
          Translation.X {type=MovieSceneScriptingFloatChannel, const [0,2]}
          Translation.Y {type=MovieSceneScriptingFloatChannel, keys [[0,170],[5,35]]}
          Angle {type=MovieSceneScriptingFloatChannel}
          Scale.X {type=MovieSceneScriptingFloatChannel}
          Scale.Y {type=MovieSceneScriptingFloatChannel}
          Shear.X {type=MovieSceneScriptingFloatChannel}
          Shear.Y {type=MovieSceneScriptingFloatChannel}
         ]
        }]
      }]
    }
   ]}
 }
 OnGo {
  binding_count=7
  length_seconds=1.15
  object=&self:OnGo
  sequence {bindings [
    Flare {
     id=#7
     possessed_class=Image
     tracks [
      {
       type=MovieSceneByteTrack
       display_name=Visibility
       sections [{type=MovieSceneByteSection, end_frame=23, start_frame=0, channels [None {type=MovieSceneScriptingByteChannel, keys [[0,4],[23,2]]}]}]
      }
      {
       type=MovieSceneFloatTrack
       display_name=Render Opacity
       sections [{
         type=MovieSceneFloatSection
         end_frame=23
         start_frame=0
         channels [None {type=MovieSceneScriptingFloatChannel, keys [[0,0],[4,0.3135],[10.8353,0.3135],[15.6943,0],[23,0]]}]
        }]
      }
      {
       type=MovieScene2DTransformTrack
       sections [{
         type=MovieScene2DTransformSection
         end_frame=17
         start_frame=0
         channels [
          Translation.X {type=MovieSceneScriptingFloatChannel}
          Translation.Y {type=MovieSceneScriptingFloatChannel}
          Angle {type=MovieSceneScriptingFloatChannel}
          Scale.X {type=MovieSceneScriptingFloatChannel, keys [[0,1.2935],[9.137,1.4055],[15.6037,6.4673]]}
          Scale.Y {type=MovieSceneScriptingFloatChannel, const [0.55,2]}
          Shear.X {type=MovieSceneScriptingFloatChannel}
          Shear.Y {type=MovieSceneScriptingFloatChannel}
         ]
        }]
      }
     ]
    }
    CommonTextBlock_2 {
     id=#8
     possessed_class=CommonTextBlock
     tracks [
      {
       type=MovieSceneByteTrack
       display_name=Visibility
       sections [{type=MovieSceneByteSection, end_frame=19, start_frame=0, channels [None {type=MovieSceneScriptingByteChannel, keys [[0,2],[9,4],[19,2]]}]}]
      }
      {
       type=MovieScene2DTransformTrack
       sections [{
         type=MovieScene2DTransformSection
         end_frame=19
         start_frame=9
         channels [
          Translation.X {type=MovieSceneScriptingFloatChannel, keys [[9,0],[15,70],[19,145]]}
          Translation.Y {type=MovieSceneScriptingFloatChannel, const [0,3]}
          Angle {type=MovieSceneScriptingFloatChannel}
          Scale.X {type=MovieSceneScriptingFloatChannel}
          Scale.Y {type=MovieSceneScriptingFloatChannel}
          Shear.X {type=MovieSceneScriptingFloatChannel, keys [[10,0],[19,-12]]}
          Shear.Y {type=MovieSceneScriptingFloatChannel, const [0,2]}
         ]
        }]
      }
      {
       type=MovieSceneFloatTrack
       display_name=Render Opacity
       sections [{
         type=MovieSceneFloatSection
         end_frame=19
         start_frame=8
         channels [None {type=MovieSceneScriptingFloatChannel, keys [[8,0],[9,0.5],[19,0]]}]
        }]
      }
     ]
    }
    ThinGlow {
     id=#9
     possessed_class=Image
     tracks [
      {
       type=MovieScene2DTransformTrack
       sections [{
         type=MovieScene2DTransformSection
         end_frame=23
         start_frame=0
         channels [
          Translation.X {type=MovieSceneScriptingFloatChannel}
          Translation.Y {type=MovieSceneScriptingFloatChannel}
          Angle {type=MovieSceneScriptingFloatChannel}
          Scale.X {type=MovieSceneScriptingFloatChannel, keys [[0,0],[9.2867,0],[12.019,1.55],[16.768,4.5911],[23,5.0311]]}
          Scale.Y {type=MovieSceneScriptingFloatChannel, const [0.55,5]}
          Shear.X {type=MovieSceneScriptingFloatChannel}
          Shear.Y {type=MovieSceneScriptingFloatChannel}
         ]
        }]
      }
      {
       type=MovieSceneFloatTrack
       display_name=Render Opacity
       sections [{type=MovieSceneFloatSection, end_frame=23, start_frame=10, channels [None {type=MovieSceneScriptingFloatChannel, keys [[10,0.407],[23,0]]}]}]
      }
      {
       type=MovieSceneByteTrack
       display_name=Visibility
       sections [{type=MovieSceneByteSection, end_frame=23, start_frame=0, channels [None {type=MovieSceneScriptingByteChannel, keys [[0,4],[23,2]]}]}]
      }
     ]
    }
    StartGameText {
     id=#3
     possessed_class=CommonRichTextBlock
     tracks [{
       type=MovieSceneWidgetMaterialTrack
       display_name=Default Text Style Override.Font.Outline Settings.OutlineMaterial
       sections [{
         type=MovieSceneParameterSection
         end_frame=19
         start_frame=0
         channels [YValue {type=MovieSceneScriptingFloatChannel, keys [[0,1.1],[7,1.1],[19,-0.1]]}]
        }]
      }]
    }
    CommonTextBlock_0 {
     id=#10
     possessed_class=CommonTextBlock
     tracks [
      {
       type=MovieSceneByteTrack
       display_name=Visibility
       sections [{type=MovieSceneByteSection, end_frame=19, start_frame=0, channels [None {type=MovieSceneScriptingByteChannel, keys [[0,2],[9,4],[19,2]]}]}]
      }
      {
       type=MovieScene2DTransformTrack
       sections [{
         type=MovieScene2DTransformSection
         end_frame=19
         start_frame=9
         channels [
          Translation.X {type=MovieSceneScriptingFloatChannel, keys [[9,0],[15,-70],[19,-145]]}
          Translation.Y {type=MovieSceneScriptingFloatChannel, const [0,3]}
          Angle {type=MovieSceneScriptingFloatChannel}
          Scale.X {type=MovieSceneScriptingFloatChannel}
          Scale.Y {type=MovieSceneScriptingFloatChannel}
          Shear.X {type=MovieSceneScriptingFloatChannel, keys [[9,0],[19,12]]}
          Shear.Y {type=MovieSceneScriptingFloatChannel, const [0,2]}
         ]
        }]
      }
      {
       type=MovieSceneFloatTrack
       display_name=Render Opacity
       sections [{
         type=MovieSceneFloatSection
         end_frame=19
         start_frame=8
         channels [None {type=MovieSceneScriptingFloatChannel, keys [[8,0],[9,0.5],[19,0]]}]
        }]
      }
     ]
    }
    CountdownText {
     id=#4
     possessed_class=CommonTextBlock
     tracks [
      {
       type=MovieScene2DTransformTrack
       sections [{
         type=MovieScene2DTransformSection
         end_frame=19
         start_frame=0
         channels [
          Translation.X {type=MovieSceneScriptingFloatChannel}
          Translation.Y {type=MovieSceneScriptingFloatChannel}
          Angle {type=MovieSceneScriptingFloatChannel}
          Scale.X {type=MovieSceneScriptingFloatChannel, keys [[0,0.7584],[2,1],[9,1],[17,1.4013]]}
          Scale.Y {type=MovieSceneScriptingFloatChannel, keys [[0,0.7584],[2,1],[9,1],[17,1.4013]]}
          Shear.X {type=MovieSceneScriptingFloatChannel}
          Shear.Y {type=MovieSceneScriptingFloatChannel}
         ]
        }]
      }
      {
       type=MovieSceneFloatTrack
       display_name=Render Opacity
       sections [{type=MovieSceneFloatSection, end_frame=19, start_frame=0, channels [None {type=MovieSceneScriptingFloatChannel, keys [[0,1],[12,1],[17,0]]}]}]
      }
      {
       type=MovieSceneColorTrack
       display_name=Color and Opacity
       sections [{
         type=MovieSceneColorSection
         end_frame=10
         start_frame=0
         channels [
          Color.R {type=MovieSceneScriptingFloatChannel, keys [[0,1],[8.083,1],[9,0.2843]]}
          Color.G {type=MovieSceneScriptingFloatChannel, keys [[0,1],[8.083,1],[8.3823,0.917],[9,0.9768]]}
          Color.B {type=MovieSceneScriptingFloatChannel, const [1,4]}
          Color.A {type=MovieSceneScriptingFloatChannel, const [1,4]}
         ]
        }]
      }
     ]
    }
    Glow {
     id=#11
     possessed_class=Image
     tracks [
      {
       type=MovieSceneByteTrack
       display_name=Visibility
       sections [{type=MovieSceneByteSection, end_frame=23, start_frame=0, channels [None {type=MovieSceneScriptingByteChannel, keys [[0,4],[23,2]]}]}]
      }
      {
       type=MovieSceneFloatTrack
       display_name=Render Opacity
       sections [{
         type=MovieSceneFloatSection
         end_frame=19
         start_frame=0
         channels [None {type=MovieSceneScriptingFloatChannel, keys [[0,0],[2,0.4829],[8,0.7532],[12,0.4829],[19,0]]}]
        }]
      }
      {
       type=MovieScene2DTransformTrack
       sections [{
         type=MovieScene2DTransformSection
         end_frame=2
         start_frame=2
         channels [
          Translation.X {type=MovieSceneScriptingFloatChannel}
          Translation.Y {type=MovieSceneScriptingFloatChannel}
          Angle {type=MovieSceneScriptingFloatChannel}
          Scale.X {type=MovieSceneScriptingFloatChannel, const [2.45,1]}
          Scale.Y {type=MovieSceneScriptingFloatChannel, const [1.4,1]}
          Shear.X {type=MovieSceneScriptingFloatChannel}
          Shear.Y {type=MovieSceneScriptingFloatChannel}
         ]
        }]
      }
     ]
    }
   ]}
 }
]
graphs [
 EventGraph {nodes [
   n1 = Update Message() {is=title K2Node_CustomEvent}
   > PlaySound2D(Sound=/Game/Sounds/Creative/Toys/Floor_Siren/Timer_UI_Minigame_Ready_Cue)
   > n2 = Decrement Int(Value=Get TimeToDisplay() -> (TimeToDisplay:int=0))
   > n3 = Branch(Condition=NOT Boolean(A="integer < integer"!(A:int=@n2, B:int=1) -> (ReturnValue:bool) {is=title K2Node_CallFunction}))
   > "SetText (Text)"(
    self=[
     Get CountdownTextHighlight() -> (CountdownTextHighlight:Common Text Object Reference)
     Get CountdownText() -> (CountdownText:Common Text Object Reference)
    ]
    InText=Format Text!(Format="LOCTEXT(\"{seconds}\")", seconds:int=@n2) -> (Result)
   )
   > PlayAnimation(InAnimation=Select(
     Option 0:Widget Animation Object Reference
     Option 1:Widget Animation Object Reference=Get Tickown() -> (Tickown:Widget Animation Object Reference)
     Index="integer <= integer"!(A:int=@n2, B:int=10) -> (ReturnValue:bool) {is=title K2Node_CallFunction}
    ) -> (ReturnValue:Widget Animation Object Reference))
   "SetText (Text)"(execute=@n3.else, self=[Get CountdownText(), Get CountdownTextHighlight()], InText="LOCTEXT(\"GO\")")
   > PlaySound2D(Sound=/Game/Sounds/Creative/Toys/Floor_Siren/Timer_UI_Tick_Minigame_End_Cue)
   > SetVisibility(self=Get StartGameText() -> (StartGameText:Common Rich Text Block Object Reference), InVisibility=Hidden)
   > PlayAnimation(InAnimation=Get OnGo() -> (OnGo:Widget Animation Object Reference))
   > Clear and Invalidate Timer by Handle(Handle=Get TimerHandle() -> (TimerHandle:Timer Handle Structure))
   DisableCounter() {is=title K2Node_CustomEvent}
   > Clear and Invalidate Timer by Handle(Handle=Get TimerHandle())
   > SetVisibility(self=Get CountDownOverlay() -> (CountDownOverlay:Overlay Object Reference), InVisibility=Collapsed)
   > Set bIsDisplayed(bIsDisplayed:bool=false) -> (Output_Get:bool)
   n4 = UpdateHUD!() -> (OutputDelegate, then, bHideHUD:bool=~) {is=title K2Node_CustomEvent}
   > n5 = Cast To FortPlayerControllerAthena(Object=GetOwningPlayer())
   > SetHUDElementVisibilityWithContext!(
    execute
    self:Fort Player Controller Object Reference=@n5
    HUDElementTags:"Gameplay Tag Container Structure (by ref)"=MakeGameplayTagContainerFromTag!(SingleTag:Gameplay Tag Structure=(TagName="HUD.Hide.WhileLoading")) -> (ReturnValue:Gameplay Tag Container Structure)
    bHideElements:bool=@n4.bHideHUD
    ReasonName:Name=None
    LogContext:String
   ) -> (then)
   n6 = Event PreConstruct()
   > Branch(Condition=@n6.IsDesignTime)
   > "SetText (Text)"(self=Get CountdownTextHighlight(), InText="GetText (Text)"!(self:Text Object Reference=Get CountdownText()) -> (ReturnValue:Text))
   n7 = Event OnDisplayMinigameStartUI!() -> (OutputDelegate, then, bDisplay:bool) {is=title K2Node_Event}
   > n8 = Branch(Condition=@n7.bDisplay)
   > Set bIsDisplayed(bIsDisplayed=@n7.bDisplay)
   > UpdateHUD!(execute, self, bHideHUD:bool=true) -> (then) {is=title K2Node_CallFunction}
   > PlayAnimation(InAnimation=Get LetterBoxIn() -> (LetterBoxIn:Widget Animation Object Reference))
   > Clear and Invalidate Timer by Handle(Handle=Get ClearTimer() -> (ClearTimer:Timer Handle Structure))
   > UpdateLetterBoxVisibility()
   > SetVisibility(self=[Get CountDownOverlay(), Get StartGameText()], InVisibility=SelfHitTestInvisible)
   n9 = Get CurrentMinigame() -> (CurrentMinigame:Fort Minigame Object Reference)
   n10 = Get TotalRounds!(self:Fort Minigame Object Reference=@n9) -> (TotalRounds:int=0)
   n11 = Get CurrentRound!(self:Fort Minigame Object Reference=@n9) -> (CurrentRound:int=0)
   > n12 = Set DisplayedText(DisplayedText:Text=Select(
     Option 0:Text=Get Game Text() -> (Game Text:Text)
     Option 1:Text=Format Text!(Format="LOCTEXT(\"ROUND <Number>{num}</> BEGINS IN\")", num:int=@n11) -> (Result)
     Index="integer > integer"(A=@n10, B=1) {is=title K2Node_CallFunction}
    ) -> (ReturnValue:Text)) -> (Output_Get:Text)
   > SetText(self=Get StartGameText(), InText=@n12)
   n13 = GetWarmupTimeRemaining!(self:Fort Minigame Warmup Component Object Reference=Get WarmupComponent!(self:Fort Minigame Object Reference=@n9) -> (WarmupComponent:Fort Minigame Warmup Component Object Reference)) -> (ReturnValue:float)
   > n14 = Set TimeToDisplay(TimeToDisplay:int=Round(A=@n13) {is=title K2Node_CallFunction}) -> (Output_Get:int=0)
   > n15 = Branch(Condition="integer >= integer"!(A:int=@n14, B:int=1) -> (ReturnValue:bool) {is=title K2Node_CallFunction})
   > PlayAnimation(InAnimation=Select(
     Option 0:Widget Animation Object Reference
     Option 1:Widget Animation Object Reference=Get Tickown()
     Index="integer <= integer"(A=@n14, B=10)
    ) -> (ReturnValue:Widget Animation Object Reference))
   > "SetText (Text)"(self=[Get CountdownText(), Get CountdownTextHighlight()], InText=Format Text!(Format="LOCTEXT(\"{seconds}\")", seconds:int=@n14) -> (Result))
   > n16 = Set Timer by Event(Delegate=@n1, Time=1, bLooping=true)
   > Set TimerHandle(TimerHandle:Timer Handle Structure=@n16) -> (Output_Get:Timer Handle Structure)
   > PushSoundMixModifier!(execute, InSoundMixModifier:Sound Mix Object Reference=/Game/Sounds/Creative/MixModifiers/Creative_InGameStart_MixMod) -> (then)
   > PlaySound2D(Sound=&Timer_UI_Minigame_Ready_Cue)
   Clear and Invalidate Timer by Handle(execute=@n15.else, Handle=Get TimerHandle())
   UpdateHUD(execute=@n8.else, bHideHUD=false)
   > n17 = Branch(Condition=Get bIsDisplayed() -> (bIsDisplayed:bool))
   n18 = Get LetterBoxIn()
   > PlayAnimation(InAnimation=@n18, PlayMode=Reverse, PlaybackSpeed=1.5)
   > PlaySound2D(Sound=/Game/Sounds/Creative/Toys/Floor_Siren/Timer_UI_Minigame_Go_Cue)
   n19 = Get OnGo()
   > Delay(Duration="Max (Float)"(A=GetEndTime(self=@n18), B=GetEndTime(self=@n19)))
   > DisableCounter() {is=title K2Node_CallFunction}
   DisableCounter(execute=@n17.else)
   MakeMessageData!(
    MsgType:EAthenaGameMsgType Enum=CornerMessage
    MsgText:Text=@n12
    MsgSound:Sound Cue Object Reference=/Game/Sounds/Fort_Default_Sounds/Default_Start_Cue
    DisplayTime:float=@n13
   ) -> (ReturnValue:Athena Game Message Data Structure)
   Format Text for Timer!(Time:int=@n14) -> (Text:Text) {is=title K2Node_MacroInstance}
   Format Text for Timer(Time=@n2)
   Get CountdownTextHighlight()
  ]}
 FormatTextForTimer {nodes [
   n1 = Get HighlightTag() -> (HighlightTag:String)
   Outputs!(Text:Text=Format Text!(
     Format="LOCTEXT(\"{Data}{HighlightTag}{time}</>\")"
     Data:Text=Get DisplayedText() -> (DisplayedText:Text)
     HighlightTag:Text="To Text (String)"(InString=@n1)
     time:int=Inputs!() -> (Time:int)
    ) -> (Result))
   Get Game Text()
  ]}
 UpdateLetterBoxVisibility {nodes [
   UpdateLetterBoxVisibility() {is=title K2Node_FunctionEntry}
   n1 = Get CurrentMinigame()
   > n2 = Is Valid(InputObject=@n1)
   n3 = GetFNEUIComponent!(self:Fort Minigame Object Reference=@n1) -> (ReturnValue:FNEUIComponent Object Reference)
   > n4 = Is Valid(InputObject=@n3)
   n5 = Get Image_LetterboxBottom() -> (Image_LetterboxBottom:Image Object Reference)
   n6 = Get Image_LetterBoxTop() -> (Image_LetterBoxTop:Image Object Reference)
   SetVisibility(execute=[@n4.Is Not Valid, @n2.Is Not Valid], self=[@n5, @n6], InVisibility=Visible)
   SetVisibility(
    execute=@n4.Is Valid
    self=[@n5, @n6]
    InVisibility=Select(
     Option 0:ESlateVisibility Enum=Hidden
     Option 1:ESlateVisibility Enum=Visible
     Index=ShouldShowPreGameCountdownBars!(self:FNEUIComponent Object Reference=@n3) -> (ReturnValue:bool)
    ) -> (ReturnValue:ESlateVisibility Enum)
   )
  ]}
]
variables [
 bIsDisplayed {type=boolean}
 ClearTimer {type=TimerHandle}
 DisplayedText {type=string, default=""}
 Game Text {type=string, default=GAME STARTS IN}
 HighlightTag {type=string, default="<enemy>"}
 TimerHandle {type=TimerHandle}
 TimeToDisplay {type=integer}
]
widgets {
 info {inheritedWidgetCount=1, parentClass=/Script/FortniteUI.MinigameWidgetBase, rootWidgetClass=&Overlay, widgetCount=16}
 widgets [
  CountDownOverlay {widgetClassPath=&Overlay, bIsVariable=true, widget=CountDownOverlay}
  CountdownBox {
   widgetClassPath=&VerticalBox
   bIsVariable=true
   parent=CountDownOverlay
   widget=CountdownBox
   slot {class=OverlaySlot, properties {horizontalAlignment=HAlign_Center, verticalAlignment=VAlign_Center, padding=<ltrb 0 0 0 200>}}
  }
  StartGameText {
   widgetClassPath=&CommonRichTextBlock
   bIsVariable=true
   parent=CountdownBox
   widget=StartGameText
   properties {
    bOverrideDefaultStyle=true
    justification=Center
    text=GAME STARTS IN
    textStyleSet=&/CreativeTextStyles
    textTransformPolicy=ToUpper
    defaultTextStyleOverride {
     overflowPolicy=Clip
     transformPolicy=None
     colorAndOpacity {colorUseRule=UseColor_Specified, specifiedColor=<rgba 1 1 1 1>}
     font {
      bForceMonospaced=false
      bMaterialIsStencil=false
      fontMaterial=null
      fontObject=&BurbankSmall
      letterSpacing=0
      monospacedWidth=1
      size=36
      skewAmount=0
      typefaceFontName=Black
      outlineSettings {
       bApplyOutlineToDropShadows=false
       bMiteredCorners=false
       bSeparateFillAlpha=false
       outlineMaterial=/Game/Athena/HUD/LTM/Playground/Materials/M_UI_OutlineSheen_GameCountdown
       outlineSize=1
       outlineColor=<rgba 1 1 1 1>
      }
     }
     highlightColor {colorUseRule=UseColor_Specified, specifiedColor=<rgba 0 0 0 1>}
     highlightShape {
      bIsDynamicallyLoaded=false
      drawAs=Image
      imageType=NoImage
      mirroring=NoMirror
      resourceName=None
      resourceObject=null
      tiling=NoTile
      imageSize=<32 32>
      margin=<ltrb 0 0 0 0>
      outlineSettings {
       bUseBrushTransparency=false
       roundingType=HalfHeightRadius
       width=0
       color {colorUseRule=UseColor_Specified, specifiedColor=<rgba 0 0 0 0>}
       cornerRadii=<0 0 0 0>
      }
      tintColor {colorUseRule=UseColor_Specified, specifiedColor=<rgba 1 1 1 1>}
      uVRegion {bIsValid=false, max=<0 0>, min=<0 0>}
     }
     selectedBackgroundColor {colorUseRule=UseColor_Foreground, specifiedColor=<rgba 1 0 1 1>}
     shadowColorAndOpacity=<rgba 0.0011 0 0.0307 0.8>
     shadowOffset=<3 3>
     strikeBrush {
      bIsDynamicallyLoaded=false
      drawAs=Image
      imageType=NoImage
      mirroring=NoMirror
      resourceName=None
      resourceObject=null
      tiling=NoTile
      imageSize=<32 32>
      margin=<ltrb 0 0 0 0>
      outlineSettings {
       bUseBrushTransparency=false
       roundingType=HalfHeightRadius
       width=0
       color {colorUseRule=UseColor_Specified, specifiedColor=<rgba 0 0 0 0>}
       cornerRadii=<0 0 0 0>
      }
      tintColor {colorUseRule=UseColor_Specified, specifiedColor=<rgba 1 1 1 1>}
      uVRegion {bIsValid=false, max=<0 0>, min=<0 0>}
     }
     underlineBrush {
      bIsDynamicallyLoaded=false
      drawAs=Image
      imageType=NoImage
      mirroring=NoMirror
      resourceName=None
      resourceObject=null
      tiling=NoTile
      imageSize=<32 32>
      margin=<ltrb 0 0 0 0>
      outlineSettings {
       bUseBrushTransparency=false
       roundingType=HalfHeightRadius
       width=0
       color {colorUseRule=UseColor_Specified, specifiedColor=<rgba 0 0 0 0>}
       cornerRadii=<0 0 0 0>
      }
      tintColor {colorUseRule=UseColor_Specified, specifiedColor=<rgba 1 1 1 1>}
      uVRegion {bIsValid=false, max=<0 0>, min=<0 0>}
     }
    }
   }
   slot {class=VerticalBoxSlot}
  }
  Overlay_0 {
   widgetClassPath=&Overlay
   parent=CountdownBox
   widget=Overlay_0
   slot {class=VerticalBoxSlot, properties {horizontalAlignment=HAlign_Center, verticalAlignment=VAlign_Top, padding=<ltrb 0 -20 0 0>}}
  }
  Glow {
   widgetClassPath=&Image
   bIsVariable=true
   parent=Overlay_0
   widget=Glow
   properties {
    renderOpacity=0
    visibility=Hidden
    brush {
     bIsDynamicallyLoaded=false
     drawAs=Image
     imageType=NoImage
     mirroring=NoMirror
     resourceName=None
     resourceObject=/Game/Items/Traps/Blueprints/Toys/UI_Art_Assets/M_UI_RoundBegin_Inst
     tiling=NoTile
     imageSize=<192 32>
     margin=<ltrb 0 0 0 0>
     outlineSettings {
      bUseBrushTransparency=false
      roundingType=HalfHeightRadius
      width=0
      color {colorUseRule=UseColor_Specified, specifiedColor=<rgba 0 0 0 0>}
      cornerRadii=<0 0 0 0>
     }
     tintColor {colorUseRule=UseColor_Specified, specifiedColor=<rgba 1 1 1 1>}
     uVRegion {bIsValid=false, max=<0 0>, min=<0 0>}
    }
    colorAndOpacity=<rgba 0.4528 0.9154 0.9184 1>
    renderTransform {angle=0, scale=<1.95 1.4>, shear=<0 0>, translation=<0 0>}
   }
   slot {class=OverlaySlot, properties {horizontalAlignment=HAlign_Fill, verticalAlignment=VAlign_Fill}}
  }
  Flare {
   widgetClassPath=&Image
   bIsVariable=true
   parent=Overlay_0
   widget=Flare
   properties {
    renderOpacity=0
    visibility=Hidden
    brush {
     bIsDynamicallyLoaded=false
     drawAs=Image
     imageType=NoImage
     mirroring=NoMirror
     resourceName=None
     resourceObject=/Game/UI/Frontend/HeroLoadout/Art/Materials/FX/M_UI_FlareNoise_Test_Inst
     tiling=NoTile
     imageSize=<32 32>
     margin=<ltrb 0 0 0 0>
     outlineSettings {
      bUseBrushTransparency=false
      roundingType=HalfHeightRadius
      width=0
      color {colorUseRule=UseColor_Specified, specifiedColor=<rgba 0 0 0 0>}
      cornerRadii=<0 0 0 0>
     }
     tintColor {colorUseRule=UseColor_Specified, specifiedColor=<rgba 1 1 1 1>}
     uVRegion {bIsValid=false, max=<0 0>, min=<0 0>}
    }
    renderTransform {angle=0, scale=<2 0.55>, shear=<0 0>, translation=<0 0>}
   }
   slot {class=OverlaySlot, properties {horizontalAlignment=HAlign_Fill, verticalAlignment=VAlign_Fill}}
  }
  ThinGlow {
   widgetClassPath=&Image
   bIsVariable=true
   parent=Overlay_0
   widget=ThinGlow
   properties {
    renderOpacity=0
    visibility=Hidden
    brush {
     bIsDynamicallyLoaded=false
     drawAs=Image
     imageType=NoImage
     mirroring=NoMirror
     resourceName=None
     resourceObject=&M_UI_RoundBegin_Inst
     tiling=NoTile
     imageSize=<192 32>
     margin=<ltrb 0 0 0 0>
     outlineSettings {
      bUseBrushTransparency=false
      roundingType=HalfHeightRadius
      width=0
      color {colorUseRule=UseColor_Specified, specifiedColor=<rgba 0 0 0 0>}
      cornerRadii=<0 0 0 0>
     }
     tintColor {colorUseRule=UseColor_Specified, specifiedColor=<rgba 1 1 1 1>}
     uVRegion {bIsValid=false, max=<0 0>, min=<0 0>}
    }
    colorAndOpacity=<rgba 0.054 0.3257 1 1>
    renderTransform {angle=0, scale=<3.85 0.55>, shear=<0 0>, translation=<0 0>}
   }
   slot {class=OverlaySlot, properties {horizontalAlignment=HAlign_Fill, verticalAlignment=VAlign_Fill}}
  }
  CommonTextBlock_0 {
   widgetClassPath=&CommonTextBlock
   bIsVariable=true
   parent=Overlay_0
   widget=CommonTextBlock_0
   properties {
    renderOpacity=0
    text=GO
    visibility=Hidden
    colorAndOpacity {colorUseRule=UseColor_Specified, specifiedColor=<rgba 0.122 0.9303 1 0.7>}
    font {
     bForceMonospaced=false
     bMaterialIsStencil=false
     fontMaterial=null
     fontObject=&BurbankSmall
     letterSpacing=0
     monospacedWidth=1
     size=80
     skewAmount=0
     typefaceFontName=Black
     outlineSettings {
      bApplyOutlineToDropShadows=false
      bMiteredCorners=false
      bSeparateFillAlpha=true
      outlineMaterial=null
      outlineSize=3
      outlineColor=<rgba 0 0.7189 1 0>
     }
    }
    shadowColorAndOpacity=<rgba 0.0011 0 0.0307 0>
    shadowOffset=<3 3>
   }
   slot {class=OverlaySlot, properties {horizontalAlignment=HAlign_Center, verticalAlignment=VAlign_Center}}
  }
  CommonTextBlock_2 {
   widgetClassPath=&CommonTextBlock
   bIsVariable=true
   parent=Overlay_0
   widget=CommonTextBlock_2
   properties {
    renderOpacity=0
    text=GO
    visibility=Hidden
    colorAndOpacity {colorUseRule=UseColor_Specified, specifiedColor=<rgba 0.122 0.9303 1 0.7>}
    font {
     bForceMonospaced=false
     bMaterialIsStencil=false
     fontMaterial=null
     fontObject=&BurbankSmall
     letterSpacing=0
     monospacedWidth=1
     size=80
     skewAmount=0
     typefaceFontName=Black
     outlineSettings {
      bApplyOutlineToDropShadows=false
      bMiteredCorners=false
      bSeparateFillAlpha=true
      outlineMaterial=null
      outlineSize=3
      outlineColor=<rgba 0 0.7189 1 0>
     }
    }
    shadowColorAndOpacity=<rgba 0.0011 0 0.0307 0>
    shadowOffset=<3 3>
   }
   slot {class=OverlaySlot, properties {horizontalAlignment=HAlign_Center, verticalAlignment=VAlign_Center}}
  }
  CountdownText {
   widgetClassPath=&CommonTextBlock
   bIsVariable=true
   parent=Overlay_0
   widget=CountdownText
   properties {
    justification=Center
    text=GO
    colorAndOpacity {colorUseRule=UseColor_Specified, specifiedColor=<rgba 0.2843 0.9768 1 1>}
    font {
     bForceMonospaced=false
     bMaterialIsStencil=false
     fontMaterial=null
     fontObject=&BurbankSmall
     letterSpacing=0
     monospacedWidth=1
     size=80
     skewAmount=0
     typefaceFontName=Black
     outlineSettings {
      bApplyOutlineToDropShadows=false
      bMiteredCorners=false
      bSeparateFillAlpha=false
      outlineMaterial=null
      outlineSize=1
      outlineColor=<rgba 0 0.6489 1 0.44>
     }
    }
    shadowColorAndOpacity=<rgba 0.0011 0 0.0307 0.8>
    shadowOffset=<3 3>
   }
   slot {class=OverlaySlot, properties {horizontalAlignment=HAlign_Center}}
  }
  CountdownTextHighlight {
   widgetClassPath=&CommonTextBlock
   bIsVariable=true
   parent=Overlay_0
   widget=CountdownTextHighlight
   properties {
    text=GO
    visibility=Hidden
    colorAndOpacity {colorUseRule=UseColor_Specified, specifiedColor=<rgba 1 1 1 0.1>}
    font {
     bForceMonospaced=false
     bMaterialIsStencil=false
     fontMaterial=null
     fontObject=&BurbankSmall
     letterSpacing=0
     monospacedWidth=1
     size=80
     skewAmount=0
     typefaceFontName=Black
     outlineSettings {
      bApplyOutlineToDropShadows=false
      bMiteredCorners=false
      bSeparateFillAlpha=true
      outlineMaterial=null
      outlineSize=3
      outlineColor=<rgba 0 0.7189 1 1>
     }
    }
    shadowColorAndOpacity=<rgba 0.0011 0 0.0307 0>
    shadowOffset=<3 3>
   }
   slot {class=OverlaySlot, properties {horizontalAlignment=HAlign_Center}}
  }
  BorderLetterBoxBottom {
   widgetClassPath=&Border
   parent=CountDownOverlay
   widget=BorderLetterBoxBottom
   properties {
    background {
     bIsDynamicallyLoaded=false
     drawAs=NoDrawType
     imageType=NoImage
     mirroring=NoMirror
     resourceName=None
     resourceObject=null
     tiling=NoTile
     imageSize=<32 32>
     margin=<ltrb 0 0 0 0>
     outlineSettings {
      bUseBrushTransparency=false
      roundingType=HalfHeightRadius
      width=0
      color {colorUseRule=UseColor_Specified, specifiedColor=<rgba 0 0 0 0>}
      cornerRadii=<0 0 0 0>
     }
     tintColor {colorUseRule=UseColor_Specified, specifiedColor=<rgba 1 1 1 1>}
     uVRegion {bIsValid=false, max=<0 0>, min=<0 0>}
    }
    padding=<ltrb 0 0 0 0>
    renderTransformPivot=<0.5 1>
   }
   slot {class=OverlaySlot, properties {horizontalAlignment=HAlign_Fill, verticalAlignment=VAlign_Bottom, padding=<ltrb -64 0 -64 -10>}}
  }
  Image_LetterboxBottom {
   widgetClassPath=&Image
   bIsVariable=true
   parent=BorderLetterBoxBottom
   widget=Image_LetterboxBottom
   properties {
    brush {
     bIsDynamicallyLoaded=false
     drawAs=Image
     imageType=NoImage
     mirroring=NoMirror
     resourceName=None
     resourceObject=/Game/UI/Foundation/Banner/Banner-SquareLobbyFrame
     tiling=NoTile
     imageSize=<128 128>
     margin=<ltrb 0 0 0 0>
     outlineSettings {
      bUseBrushTransparency=false
      roundingType=HalfHeightRadius
      width=0
      color {colorUseRule=UseColor_Specified, specifiedColor=<rgba 0 0 0 0>}
      cornerRadii=<0 0 0 0>
     }
     tintColor {colorUseRule=UseColor_Specified, specifiedColor=<rgba 1 1 1 1>}
     uVRegion {bIsValid=false, max=<0 0>, min=<0 0>}
    }
    colorAndOpacity=<rgba 0 0 0 1>
    renderTransform {angle=0, scale=<-1.5 1.5>, shear=<0 0>, translation=<0 35>}
   }
   slot {class=BorderSlot, properties {padding=<ltrb 0 0 0 0>}}
  }
  Border_0 {
   widgetClassPath=&Border
   parent=CountDownOverlay
   widget=Border_0
   properties {
    background {
     bIsDynamicallyLoaded=false
     drawAs=NoDrawType
     imageType=NoImage
     mirroring=NoMirror
     resourceName=None
     resourceObject=null
     tiling=NoTile
     imageSize=<32 32>
     margin=<ltrb 0 0 0 0>
     outlineSettings {
      bUseBrushTransparency=false
      roundingType=HalfHeightRadius
      width=0
      color {colorUseRule=UseColor_Specified, specifiedColor=<rgba 0 0 0 0>}
      cornerRadii=<0 0 0 0>
     }
     tintColor {colorUseRule=UseColor_Specified, specifiedColor=<rgba 1 1 1 1>}
     uVRegion {bIsValid=false, max=<0 0>, min=<0 0>}
    }
    padding=<ltrb 0 0 0 0>
    renderTransformPivot=<0.5 1>
   }
   slot {class=OverlaySlot, properties {horizontalAlignment=HAlign_Fill, padding=<ltrb -64 -10 -64 0>}}
  }
  Image_LetterBoxTop {
   widgetClassPath=&Image
   bIsVariable=true
   parent=Border_0
   widget=Image_LetterBoxTop
   properties {
    brush {
     bIsDynamicallyLoaded=false
     drawAs=Image
     imageType=NoImage
     mirroring=NoMirror
     resourceName=None
     resourceObject=&Banner-SquareLobbyFrame
     tiling=NoTile
     imageSize=<128 128>
     margin=<ltrb 0 0 0 0>
     outlineSettings {
      bUseBrushTransparency=false
      roundingType=HalfHeightRadius
      width=0
      color {colorUseRule=UseColor_Specified, specifiedColor=<rgba 0 0 0 0>}
      cornerRadii=<0 0 0 0>
     }
     tintColor {colorUseRule=UseColor_Specified, specifiedColor=<rgba 1 1 1 1>}
     uVRegion {bIsValid=false, max=<0 0>, min=<0 0>}
    }
    colorAndOpacity=<rgba 0 0 0 1>
    renderTransform {angle=0, scale=<-1.5 1.5>, shear=<0 0>, translation=<0 -35>}
   }
   slot {class=BorderSlot, properties {padding=<ltrb 0 0 0 0>}}
  }
  TutorialHighlight {widgetClassPath=&TutorialHighlight}
 ]
}
