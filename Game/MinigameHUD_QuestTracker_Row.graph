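# Quest tracker row widget for the Creative minigame HUD, showing a quest's name, score/target, progress bar and optional 1st-3rd placement.
# Quest changes apply color, progress mode and visibility, playing intro or checkmark outro animations.
# Progress is parsed back from the text blocks, then a Sequencer repeater event on AnimateBarFill interpolates the bar material's PercentFill.
# SetPlacement re-parents the row into a given horizontal box and restyles its rank.
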WidgetBlueprint /Game/MinigameHUD_QuestTracker_Row
animations [
 OnShow {
  binding_count=2
  length_seconds=0.5
  object=&self:OnShow
  sequence {
   bindings [
    MinigameHUD_QuestTracker_Row_C_67 {
     id=#1
     possessed_class=MinigameHUD_QuestTracker_Row_C
     tracks [
      {
       type=MovieScene2DTransformTrack
       sections [{
         type=MovieScene2DTransformSection
         channels [
          Translation.X {type=MovieSceneScriptingFloatChannel, keys [[0,-50],[30,0]]}
          Translation.Y {type=MovieSceneScriptingFloatChannel}
          Angle {type=MovieSceneScriptingFloatChannel}
          Scale.X {type=MovieSceneScriptingFloatChannel}
          Scale.Y {type=MovieSceneScriptingFloatChannel}
          Shear.X {type=MovieSceneScriptingFloatChannel}
          Shear.Y {type=MovieSceneScriptingFloatChannel}
         ]
        }]
      }
      {
       type=MovieSceneFloatTrack
       display_name=Render Opacity
       sections [{type=MovieSceneFloatSection, channels [None {type=MovieSceneScriptingFloatChannel, keys [[0,0],[15,1]]}]}]
      }
     ]
    }
    WidgetSwitcher_Complete {
     id=#2
     possessed_class=WidgetSwitcher
     tracks [{
       type=MovieSceneIntegerTrack
       display_name=Active Widget Index
       sections [{type=MovieSceneIntegerSection, end_frame=0, start_frame=0, channels [None {type=MovieSceneScriptingIntegerChannel, const [0,1]}]}]
      }]
    }
   ]
   display_rate {numerator=60}
  }
 }
 CheckMarkOutro {
  binding_count=5
  length_seconds=1.05
  object=&self:CheckMarkOutro
  sequence {bindings [
    Image_Glow {
     id=#3
     possessed_class=Image
     tracks [
      {
       type=MovieSceneFloatTrack
       display_name=Render Opacity
       sections [{type=MovieSceneFloatSection, channels [None {type=MovieSceneScriptingFloatChannel, keys [[6,0],[8,1],[12,0]]}]}]
      }
      {
       type=MovieScene2DTransformTrack
       sections [{
         type=MovieScene2DTransformSection
         channels [
          Translation.X {type=MovieSceneScriptingFloatChannel}
          Translation.Y {type=MovieSceneScriptingFloatChannel}
          Angle {type=MovieSceneScriptingFloatChannel}
          Scale.X {type=MovieSceneScriptingFloatChannel, keys [[6,1],[12,1.3]]}
          Scale.Y {type=MovieSceneScriptingFloatChannel, keys [[6,1],[12,1.3]]}
          Shear.X {type=MovieSceneScriptingFloatChannel}
          Shear.Y {type=MovieSceneScriptingFloatChannel}
         ]
        }]
      }
     ]
    }
    CheckMarkIcon {
     id=#4
     possessed_class=Image
     tracks [
      {
       type=MovieSceneFloatTrack
       display_name=Render Opacity
       sections [{type=MovieSceneFloatSection, channels [None {type=MovieSceneScriptingFloatChannel, keys [[0,0],[7.2397,1]]}]}]
      }
      {
       type=MovieScene2DTransformTrack
       sections [{
         type=MovieScene2DTransformSection
         channels [
          Translation.X {type=MovieSceneScriptingFloatChannel}
          Translation.Y {type=MovieSceneScriptingFloatChannel, const [0,3]}
          Angle {type=MovieSceneScriptingFloatChannel}
          Scale.X {type=MovieSceneScriptingFloatChannel, keys [[0,2.3531],[6,0.45],[8,0.835]]}
          Scale.Y {type=MovieSceneScriptingFloatChannel, keys [[0,2.3531],[6,0.45],[8,0.835]]}
          Shear.X {type=MovieSceneScriptingFloatChannel}
          Shear.Y {type=MovieSceneScriptingFloatChannel}
         ]
        }]
      }
     ]
    }
    WidgetSwitcher_Complete {
     id=#5
     possessed_class=WidgetSwitcher
     tracks [{
       type=MovieSceneIntegerTrack
       display_name=Active Widget Index
       sections [{type=MovieSceneIntegerSection, end_frame=0, start_frame=0, channels [None {type=MovieSceneScriptingIntegerChannel, const [1,1]}]}]
      }]
    }
    MinigameHUD_QuestTracker_Row_C_72 {
     id=#6
     possessed_class=MinigameHUD_QuestTracker_Row_C
     tracks [{
       type=MovieSceneFloatTrack
       display_name=Render Opacity
       sections [{type=MovieSceneFloatSection, channels [None {type=MovieSceneScriptingFloatChannel, keys [[16,1],[21,0]]}]}]
      }]
    }
    Image_Spark {
     id=#7
     possessed_class=Image
     tracks [
      {
       type=MovieSceneFloatTrack
       display_name=Render Opacity
       sections [{type=MovieSceneFloatSection, channels [None {type=MovieSceneScriptingFloatChannel, keys [[6,0],[8,1],[12,0]]}]}]
      }
      {
       type=MovieScene2DTransformTrack
       sections [{
         type=MovieScene2DTransformSection
         channels [
          Translation.X {type=MovieSceneScriptingFloatChannel}
          Translation.Y {type=MovieSceneScriptingFloatChannel}
          Angle {type=MovieSceneScriptingFloatChannel}
          Scale.X {type=MovieSceneScriptingFloatChannel, keys [[6,0.75],[12,1]]}
          Scale.Y {type=MovieSceneScriptingFloatChannel, keys [[6,0.75],[12,1]]}
          Shear.X {type=MovieSceneScriptingFloatChannel}
          Shear.Y {type=MovieSceneScriptingFloatChannel}
         ]
        }]
      }
     ]
    }
   ]}
 }
 AnimateBarFill {
  binding_count=0
  length_seconds=0.5
  object=&self:AnimateBarFill
  sequence {bindings [], tracks [{type=MovieSceneEventTrack, sections [{type=MovieSceneEventRepeaterSection, end_frame=10, start_frame=0}]}]}
 }
]
graphs [
 EventGraph {nodes [
   n1 = SetPlacement!() -> (OutputDelegate, then, Placement:int, NewParent:Horizontal Box Object Reference, PlayAnimation:bool=~) {is=title K2Node_CustomEvent}
   > Set New Placement(New Placement:int=@n1.Placement) -> (Output_Get:int=0)
   > Set New Parent(New Parent:Horizontal Box Object Reference=@n1.NewParent) -> (Output_Get:Horizontal Box Object Reference)
   > n2 = Branch(Condition=@n1.PlayAnimation)
   > n3 = "Play Animation with Finished event (legacy)"!(
    execute:Exec
    Widget:User Widget Object Reference=Self-Reference()
    InAnimation:Widget Animation Object Reference=Get OnShow() -> (OnShow:Widget Animation Object Reference)
    StartAtTime:float=0
    NumLoopsToPlay:int=1
    PlayMode:EUMGSequencePlayMode Enum
    PlaybackSpeed:float=1
   ) -> (then:Exec, Result:UMGSequence Player Object Reference, Finished:Exec) {is=title K2Node_PlayAnimation}
   Set In Visibility(
    execute=[@n3.Finished, @n2.else]
    In Visibility:ESlateVisibility Enum=GetVisibility!(self:Widget Object Reference) -> (ReturnValue:ESlateVisibility Enum=Visible)
   ) -> (Output_Get:ESlateVisibility Enum=Visible)
   n4 = Self-Reference()
   > RemoveFromParent(self=@n4)
   > AddChildToHorizontalBox!(
    execute
    self:Horizontal Box Object Reference=Get New Parent() -> (New Parent:Horizontal Box Object Reference)
    Content:Widget Object Reference=@n4
   ) -> (then, ReturnValue:Horizontal Box Slot Object Reference)
   > SetVisibility(InVisibility=Get In Visibility() -> (In Visibility:ESlateVisibility Enum=Visible))
   > SetRank!(execute, self, Rank:int=Get New Placement() -> (New Placement:int=0)) -> (then)
   > Branch(Condition=@n1.PlayAnimation)
   > PlayAnimationForward(InAnimation=Get OnShow())
   n5 = Event OnQuestChanged!() -> (OutputDelegate, then, CurrentQuestComponent:Creative Quest Component Object Reference, bHasActiveQuest:bool) {is=title K2Node_Event}
   > n6 = Branch(Condition=@n5.bHasActiveQuest)
   > Is Valid(InputObject=@n5.CurrentQuestComponent)
   > SetRowVisibility!(execute, self, Show:bool=true) -> (then)
   > ShowNewQuest!(execute, self, QuestComponent:Creative Quest Component Object Reference=@n5.CurrentQuestComponent) -> (then)
   > SetProgressBarFill()
   > Call On Quest Progress Changed!(execute:Exec, self:Minigame HUD Quest Tracker Row Object Reference) -> (then:Exec) {is=title K2Node_CallDelegate}
   > Set ShowCompleteAnimation(ShowCompleteAnimation:bool=Get bShowCompleteEffects!(self:Creative Quest Component Object Reference=@n5.CurrentQuestComponent) -> (bShowCompleteEffects:bool)) -> (Output_Get:bool)
   n7 = Branch(execute=@n6.else, Condition=Get ShowCompleteAnimation() -> (ShowCompleteAnimation:bool))
   > PlayAnimation(InAnimation=Get CheckMarkOutro() -> (CheckMarkOutro:Widget Animation Object Reference))
   > Call On Quest Progress Changed()
   SetRowVisibility(execute=@n7.else, Show=false)
   > Call On Quest Progress Changed()
   n8 = Event OnProgressUpdated() {is=title K2Node_Event}
   SetProgressBarFill(execute=[@n8.then, @n9.then])
   > Call On Quest Progress Changed()
   n9 = Event OnProgressChanged() {is=title K2Node_Event}
   n10 = Event OnDisplayMinigameInProgressUI!() -> (OutputDelegate, then, bDisplay:bool) {is=title K2Node_Event}
   > Parent: OnDisplayMinigameInProgressUI!(execute, bDisplay:bool=@n10.bDisplay) -> (then) {is=title K2Node_CallParentFunction}
   > SetRowVisibility(Show=false)
   > Set NewPercentage(NewPercentage:double=float / float(
     A=Get Score Value!() -> (Value:double=~) {is=title K2Node_MacroInstance}
     B=Get Target Value!() -> (Value:double=~) {is=title K2Node_MacroInstance}
    ) {is=title K2Node_CallFunction}) -> (Output_Get:double)
   Set ShowCompleteAnimation(ShowCompleteAnimation=false)
  ]}
 GetCurrentScore {nodes [
   GetCurrentScore() {is=title K2Node_FunctionEntry}
   > Return Node!(execute, Value:double=Get Score Value() -> (Value=~))
  ]}
 GetScoreValue {nodes [
   Inputs!()
   n1 = "GetText (Text)"!(self:Text Object Reference=Get Text_CurrentAmount() -> (Text_CurrentAmount:Common Text Object Reference)) -> (ReturnValue:Text)
   Outputs!(Value:double=String To Float(InString="To String (Text)"(InText=@n1)))
  ]}
 GetTargetScore {nodes [
   GetTargetScore() {is=title K2Node_FunctionEntry}
   > Return Node(Value=Get Target Value() -> (Value=~))
  ]}
 GetTargetValue {nodes [
   Inputs()
   n1 = "GetText (Text)"(self=Get Text_TargetAmount() -> (Text_TargetAmount:Common Text Object Reference))
   Outputs(Value=String To Float(InString="To String (Text)"(InText=@n1)))
  ]}
 OnQuestProgressChanged {nodes [OnQuestProgressChanged() {is=title K2Node_FunctionEntry}]}
 Sequencer Events {nodes [
   TickBarFill() {is=title K2Node_CustomEvent}
   n1 = Get NewPercentage() -> (NewPercentage:double)
   > n2 = Branch(Condition="Nearly Equal (Float)"(A=@n1, B=Get OriginalPercentage() -> (OriginalPercentage:double), ErrorTolerance=0.01))
   > SetFillPercentage!(execute, self, Value:double=Get NewPercentage()) -> (then)
   n3 = float - float(A=@n1, B=Get OriginalPercentage()) {is=title K2Node_CallFunction}
   n4 = Get AnimateBarFill() -> (AnimateBarFill:Widget Animation Object Reference)
   n5 = "Clamp (Float)"(
    Value=float / float(A=GetAnimationCurrentTime!(self:User Widget Object Reference, InAnimation:Widget Animation Object Reference=@n4) -> (ReturnValue:float), B=1)
    Min=0
    Max=1
   )
   SetFillPercentage(
    execute=@n2.else
    Value=float + float(A=Get OriginalPercentage(), B=float * float(A=@n3, B=@n5)) {is=title K2Node_CommutativeAssociativeBinaryOperator}
   )
  ]}
 SetFillPercentage {nodes [
   n1 = SetFillPercentage!() -> (then, Value:double=~) {is=title K2Node_FunctionEntry}
   > Set OriginalPercentage(OriginalPercentage:double=@n1) -> (Output_Get:double)
   n2 = Get Image_BackplateProgress_1() -> (Image_BackplateProgress_1:Image Object Reference)
   > Is Valid(InputObject=@n2)
   > n3 = GetDynamicMaterial(self=@n2)
   > Is Valid(InputObject=@n3)
   > SetScalarParameterValue(self=@n3, ParameterName=PercentFill, Value=Get OriginalPercentage())
  ]}
 SetIsRanked {nodes [
   n1 = SetIsRanked!() -> (then, IsRanked:bool=~) {is=title K2Node_FunctionEntry}
   > n2 = Set IsRanked(IsRanked:bool=@n1) -> (Output_Get:bool)
   > n3 = Branch(Condition=@n2)
   > SetVisibility(self=Get PlacementBox() -> (PlacementBox:Size Box Object Reference), InVisibility=SelfHitTestInvisible)
   SetVisibility(execute=@n3.else, self=Get PlacementBox(), InVisibility=Collapsed)
  ]}
 SetProgressBarFill {nodes [
   SetProgressBarFill() {is=title K2Node_FunctionEntry}
   > Set NewPercentage(NewPercentage=0)
   > n1 = Branch(Condition="float > float"(A=Get Target Value() -> (Value=~), B=0) {is=title K2Node_CallFunction})
   > n2 = Branch(Condition=Select!(
     Total:bool=true
     Remaining:bool=true
     Off:bool=false
     Index:EShowProgressMode Enum=Get Show Progress Mode() -> (Show Progress Mode:EShowProgressMode Enum=Total)
    ) -> (ReturnValue:bool))
   n3 = float / float(A=Get Score Value() -> (Value=~), B=Get Target Value() -> (Value=~))
   n4 = Set NewPercentage(execute=[@n2.then, @n5.then], NewPercentage=@n3)
   PlayAnimation(execute=[@n4.then, @n5.else], InAnimation=Get AnimateBarFill())
   n5 = Branch(execute=@n2.else, Condition="Nearly Equal (Float)"(A=@n3, B=1, ErrorTolerance=0.01))
   SetFillPercentage(execute=@n1.else, Value=Get NewPercentage()) {is=title K2Node_CallFunction}
  ]}
 SetRank {nodes [
   n1 = SetRank!() -> (then, Rank:int) {is=title K2Node_FunctionEntry}
   > SetVisibility(self=Get PlacementBox(), InVisibility=SelfHitTestInvisible)
   > n2 = Switch on Int!(execute:Exec, Selection:int=@n1) -> (Default:Exec, 0:Exec, 1:Exec, 2:Exec)
   n3 = Get Placement() -> (Placement:Common Text Object Reference)
   n4 = GetDynamicFontMaterial!(execute=@n2.0, self:Text Object Reference=@n3) -> (then, ReturnValue:Material Instance Dynamic Object Reference)
   > SetVectorParameterValue(self=@n4, ParameterName=Color1, Value=(R=1,G=0.7913,B=0.2159,A=1))
   > SetVectorParameterValue(self=@n4, ParameterName=Color2, Value=(R=1,G=1,B=1,A=1))
   > n5 = GetDynamicOutlineMaterial!(execute, self:Text Object Reference=@n3) -> (then, ReturnValue:Material Instance Dynamic Object Reference)
   > SetVectorParameterValue(self=@n5, ParameterName=Color1, Value=(R=0.3467,G=0.0953,B=0.0185,A=1))
   > SetVectorParameterValue(self=@n5, ParameterName=Color2, Value=(R=0.3467,G=0.0953,B=0.0185,A=1))
   > "SetText (Text)"(self=@n3, InText="LOCTEXT(\"1st\")")
   n6 = GetDynamicFontMaterial(execute=@n2.1, self=@n3)
   > SetVectorParameterValue(self=@n6, ParameterName=Color1, Value=(R=0.6038,G=0.6584,B=0.7454,A=1))
   > SetVectorParameterValue(self=@n6, ParameterName=Color2, Value=(R=1,G=1,B=1,A=1))
   > n7 = GetDynamicOutlineMaterial(self=@n3)
   > SetVectorParameterValue(self=@n7, ParameterName=Color1, Value=(R=0.0627,G=0.0739,B=0.0825,A=1))
   > SetVectorParameterValue(self=@n7, ParameterName=Color2, Value=(R=0.0627,G=0.0739,B=0.0825,A=1))
   > "SetText (Text)"(self=@n3, InText="LOCTEXT(\"2nd\")")
   n8 = GetDynamicFontMaterial(execute=@n2.2, self=@n3)
   > SetVectorParameterValue(self=@n8, ParameterName=Color1, Value=(R=0.4508,G=0.2122,B=0.0999,A=1))
   > SetVectorParameterValue(self=@n8, ParameterName=Color2, Value=(R=1,G=0.8796,B=0.6724,A=1))
   > n9 = GetDynamicOutlineMaterial(self=@n3)
   > SetVectorParameterValue(self=@n9, ParameterName=Color1, Value=(R=0.1119,G=0.0137,B=0.0194,A=1))
   > SetVectorParameterValue(self=@n9, ParameterName=Color2, Value=(R=0.1119,G=0.0137,B=0.0194,A=1))
   > "SetText (Text)"(self=@n3, InText="LOCTEXT(\"3rd\")")
   SetColorAndOpacity(self:Image Object Reference, InColorAndOpacity:Linear Color Structure=(R=0.0125,G=0.0125,B=0.0125,A=0.2))
  ]}
 SetRowVisibility {nodes [
   n1 = SetRowVisibility!() -> (then, Show:bool=~) {is=title K2Node_FunctionEntry}
   > n2 = Branch(Condition=@n1)
   > SetVisibility(InVisibility=HitTestInvisible)
   SetVisibility(execute=@n2.else, InVisibility=Collapsed)
  ]}
 ShowCompleteQuest {nodes [
   ShowCompleteQuest() {is=title K2Node_FunctionEntry}
   > n1 = Branch()
   > PlayAnimation(InAnimation=Get CheckMarkOutro())
   SetRowVisibility(execute=@n1.else, Show=false) {is=title K2Node_CallFunction}
  ]}
 ShowNewQuest {nodes [
   n1 = ShowNewQuest!() -> (then, QuestComponent:Creative Quest Component Object Reference) {is=title K2Node_FunctionEntry}
   > n2 = Switch on EShowOnHudMode!(
    execute
    Selection:EShowOnHudMode Enum=Get ShowQuestOnHUD!(self:Creative Quest Component Object Reference=@n1) -> (ShowQuestOnHUD:EShowOnHudMode Enum=No)
   ) -> (No, Detailed:Exec, List:Exec, Both:Exec) {is=title K2Node_SwitchEnum}
   SetRowVisibility(execute=[@n2.Both, @n2.List], Show=true)
   > StopAnimation(InAnimation=Get CheckMarkOutro())
   > PlayAnimation(InAnimation=Get OnShow())
   > n3 = GetDynamicMaterial(self=Get Image_BackplateProgress_1())
   n4 = Get QuestColor!(self:Creative Quest Component Object Reference=Get QuestComponent() -> (QuestComponent:Creative Quest Component Object Reference)) -> (QuestColor:Linear Color Structure)
   > SetVectorParameterValue(self=@n3, ParameterName=FillColor, Value=@n4)
   > SetColorAndOpacity(
    self:Text Object Reference=[Get Text_TargetAmount(), Get TextProgressDivider() -> (TextProgressDivider:Common Text Object Reference)]
    InColorAndOpacity:Slate Color Structure=Make Slate Color(SpecifiedColor=@n4)
   )
   > n5 = Set Show Progress Mode(Show Progress Mode:EShowProgressMode Enum=Get ShowProgressMode!(self:Creative Quest Component Object Reference=Get QuestComponent()) -> (ShowProgressMode:EShowProgressMode Enum=Total)) -> (Output_Get:EShowProgressMode Enum=Total)
   > n6 = Switch on EShowProgressMode!(execute, Selection:EShowProgressMode Enum=@n5) -> (Total:Exec, Remaining:Exec, Off:Exec) {is=title K2Node_SwitchEnum}
   SetMaxDesiredWidth!(
    execute=[@n6.Remaining, @n6.Total]
    self:Size Box Object Reference=Get SizeBox_QuestName() -> (SizeBox_QuestName:Size Box Object Reference)
    InMaxDesiredWidth:float=Get MinDesiredWidth!(self:Size Box Object Reference=Get SizeBox_QuestName()) -> (MinDesiredWidth:float)
   ) -> (then)
   > SetVisibility(self=Get SizeBox_Progress() -> (SizeBox_Progress:Size Box Object Reference), InVisibility=HitTestInvisible)
   n7 = Get SizeBox_Progress()
   n8 = Get SizeBox_QuestName()
   SetMaxDesiredWidth(
    execute=@n6.Off
    self=Get SizeBox_QuestName()
    InMaxDesiredWidth=float + float(A=Get MinDesiredWidth(self=@n8), B=Get WidthOverride!(self:Size Box Object Reference=@n7) -> (WidthOverride:float))
   )
   > SetVisibility(self=Get SizeBox_Progress(), InVisibility=Collapsed)
   SetRowVisibility(execute=[@n2.Detailed, @n2.No], Show=false)
  ]}
]
variables [
 Display {type=boolean}
 In Visibility {type=ESlateVisibility, default=Visible}
 IsRanked {type=boolean}
 New Parent {type=HorizontalBox, default=null}
 New Placement {type=integer}
 NewPercentage {type=number}
 OnQuestProgressChanged {}
 OriginalPercentage {type=number}
 Show Progress Mode {type=EShowProgressMode, default=Total}
 ShowCompleteAnimation {type=boolean}
]
widgets {
 info {inheritedWidgetCount=5, parentClass=/Script/FortniteUI.FortCreativeQuestHUDWidget, rootWidgetClass=&CanvasPanel, widgetCount=26}
 widgets [
  CanvasPanel_364 {widgetClassPath=&CanvasPanel, widget=CanvasPanel_364}
  SizeBox_142 {
   widgetClassPath=&SizeBox
   parent=CanvasPanel_364
   widget=SizeBox_142
   properties {bOverride_HeightOverride=true, bOverride_WidthOverride=true, heightOverride=52, widthOverride=380}
   slot {class=CanvasPanelSlot, properties {bAutoSize=true, layoutData=~}}
  }
  Overlay_5 {widgetClassPath=&Overlay, parent=SizeBox_142, widget=Overlay_5, slot {class=SizeBoxSlot}}
  QuestBackgroundBorder {
   widgetClassPath=&CommonBorder
   bIsVariable=true
   parent=Overlay_5
   widget=QuestBackgroundBorder
   properties {
    background {
     bIsDynamicallyLoaded=false
     drawAs=Box
     imageType=NoImage
     mirroring=NoMirror
     resourceName=None
     resourceObject=/Game/Athena/UI/InGame/Creative/Art/M_UI_Quest_Container_QuestTrackerRow
     tiling=NoTile
     imageSize=<64 64>
     margin=<ltrb 0.5 0.5 0.5 0.5>
     outlineSettings {
      bUseBrushTransparency=false
      roundingType=HalfHeightRadius
      width=0
      color {colorUseRule=UseColor_Specified, specifiedColor=<rgba 0 0 0 0>}
      cornerRadii=<0 0 0 0>
     }
     tintColor {colorUseRule=UseColor_Specified, specifiedColor=<rgba 1 1 1 1>}
     uVRegion {bIsValid=false, max=<0 0>, min=<0 0>}
    }
    padding=<ltrb 0 0 0 0>
   }
   slot {class=OverlaySlot, properties {horizontalAlignment=HAlign_Fill, verticalAlignment=VAlign_Fill}}
  }
  VerticalBox_0 {
   widgetClassPath=&VerticalBox
   parent=Overlay_5
   widget=VerticalBox_0
   slot {class=OverlaySlot, properties {horizontalAlignment=HAlign_Fill, verticalAlignment=VAlign_Center, padding=<ltrb 12 0 12 1>}}
  }
  SizeBox_1 {
   widgetClassPath=&SizeBox
   parent=VerticalBox_0
   widget=SizeBox_1
   properties {bOverride_MaxDesiredWidth=true, maxDesiredWidth=215}
   slot {class=VerticalBoxSlot}
  }
  HorizontalBox_540 {widgetClassPath=&HorizontalBox, parent=SizeBox_1, widget=HorizontalBox_540, slot {class=SizeBoxSlot}}
  PlacementBox {
   widgetClassPath=&SizeBox
   bIsVariable=true
   parent=HorizontalBox_540
   widget=PlacementBox
   slot {class=HorizontalBoxSlot, properties {horizontalAlignment=HAlign_Left}}
  }
  Placement {
   widgetClassPath=&CommonTextBlock
   bIsVariable=true
   parent=PlacementBox
   widget=Placement
   properties {
    text=1st
    font {
     bForceMonospaced=false
     bMaterialIsStencil=false
     fontMaterial=/Game/Items/Traps/Blueprints/Toys/UI_Art_Assets/M_UI_AccoladeText
     fontObject=&Burbank
     letterSpacing=0
     monospacedWidth=1
     size=14
     skewAmount=0
     typefaceFontName=Burbank-Black
     outlineSettings {
      bApplyOutlineToDropShadows=false
      bMiteredCorners=false
      bSeparateFillAlpha=true
      outlineMaterial=/Game/Items/Traps/Blueprints/Toys/UI_Art_Assets/M_UI_AccoladeText_Outline
      outlineSize=1
      outlineColor=<rgba 1 1 1 1>
     }
    }
    renderTransform {angle=0, scale=<1 1>, shear=<-8 0>, translation=<0 0>}
    shadowOffset=<0 0>
   }
   slot {class=SizeBoxSlot, properties {horizontalAlignment=HAlign_Center, verticalAlignment=VAlign_Center}}
  }
  SizeBox_QuestName {
   widgetClassPath=&SizeBox
   bIsVariable=true
   parent=HorizontalBox_540
   widget=SizeBox_QuestName
   properties {bOverride_WidthOverride=true, maxAspectRatio=10, maxDesiredWidth=215, minDesiredWidth=215, widthOverride=215}
   slot {class=HorizontalBoxSlot, properties {horizontalAlignment=HAlign_Left}}
  }
  Text_QuestName {
   widgetClassPath=&CommonTextBlock
   bInherited=true
   bIsVariable=true
   parent=SizeBox_QuestName
   widget=Text_QuestName
   properties {
    clipping=OnDemand
    lineHeightPercentage=0.75
    scrollStyle=&TextScrollStyle-Base_C
    text=Hunter: Tribe Sent Down The Hole
    wrappingPolicy=AllowPerCharacterWrapping
    colorAndOpacity {colorUseRule=UseColor_Specified, specifiedColor=<rgba 0.2159 0.6584 1 1>}
    font {
     bForceMonospaced=false
     bMaterialIsStencil=false
     fontMaterial=null
     fontObject=&BurbankSmall
     letterSpacing=0
     monospacedWidth=1
     size=15
     skewAmount=0
     typefaceFontName=Bold
     outlineSettings {
      bApplyOutlineToDropShadows=false
      bMiteredCorners=false
      bSeparateFillAlpha=false
      outlineMaterial=null
      outlineSize=0
      outlineColor=<rgba 0 0 0 1>
     }
    }
    shadowOffset=<0 0>
   }
   slot {class=SizeBoxSlot, properties {horizontalAlignment=HAlign_Left}}
  }
  WidgetSwitcher_Complete {
   widgetClassPath=&WidgetSwitcher
   bIsVariable=true
   parent=HorizontalBox_540
   widget=WidgetSwitcher_Complete
   slot {class=HorizontalBoxSlot, properties {horizontalAlignment=HAlign_Right}}
  }
  SizeBox_Progress {
   widgetClassPath=&SizeBox
   bIsVariable=true
   parent=WidgetSwitcher_Complete
   widget=SizeBox_Progress
   properties {bOverride_WidthOverride=true, widthOverride=140}
   slot {class=WidgetSwitcherSlot}
  }
  HorizontalBox_52 {
   widgetClassPath=&HorizontalBox
   parent=SizeBox_Progress
   widget=HorizontalBox_52
   slot {class=SizeBoxSlot, properties {horizontalAlignment=HAlign_Right}}
  }
  Text_CurrentAmount {
   widgetClassPath=&CommonTextBlock
   bInherited=true
   bIsVariable=true
   parent=HorizontalBox_52
   widget=Text_CurrentAmount
   properties {
    text=10000
    colorAndOpacity {colorUseRule=UseColor_Specified, specifiedColor=<rgba 0.6038 0.8879 1 1>}
    font {
     bForceMonospaced=false
     bMaterialIsStencil=false
     fontMaterial=null
     fontObject=&BurbankSmall
     letterSpacing=0
     monospacedWidth=1
     size=15
     skewAmount=0
     typefaceFontName=Bold
     outlineSettings {
      bApplyOutlineToDropShadows=false
      bMiteredCorners=false
      bSeparateFillAlpha=false
      outlineMaterial=null
      outlineSize=0
      outlineColor=<rgba 0 0 0 1>
     }
    }
    shadowColorAndOpacity=<rgba 0 0 0 1>
    shadowOffset=<0 0>
   }
   slot {class=HorizontalBoxSlot, properties {horizontalAlignment=HAlign_Right, verticalAlignment=VAlign_Center}}
  }
  TextProgressDivider {
   widgetClassPath=&CommonTextBlock
   bIsVariable=true
   parent=HorizontalBox_52
   widget=TextProgressDivider
   properties {
    text=/
    colorAndOpacity {colorUseRule=UseColor_Specified, specifiedColor=<rgba 0.2195 0.6584 1 1>}
    font {
     bForceMonospaced=false
     bMaterialIsStencil=false
     fontMaterial=null
     fontObject=&BurbankSmall
     letterSpacing=0
     monospacedWidth=1
     size=15
     skewAmount=0
     typefaceFontName=Bold
     outlineSettings {
      bApplyOutlineToDropShadows=false
      bMiteredCorners=false
      bSeparateFillAlpha=false
      outlineMaterial=null
      outlineSize=0
      outlineColor=<rgba 0 0 0 1>
     }
    }
    shadowColorAndOpacity=<rgba 0 0 0 1>
    shadowOffset=<0 0>
   }
   slot {class=HorizontalBoxSlot, properties {horizontalAlignment=HAlign_Right, verticalAlignment=VAlign_Center}}
  }
  Text_TargetAmount {
   widgetClassPath=&CommonTextBlock
   bInherited=true
   bIsVariable=true
   parent=HorizontalBox_52
   widget=Text_TargetAmount
   properties {
    text=10000
    colorAndOpacity {colorUseRule=UseColor_Specified, specifiedColor=<rgba 0.2195 0.6584 1 1>}
    font {
     bForceMonospaced=false
     bMaterialIsStencil=false
     fontMaterial=null
     fontObject=&BurbankSmall
     letterSpacing=0
     monospacedWidth=1
     size=15
     skewAmount=0
     typefaceFontName=Bold
     outlineSettings {
      bApplyOutlineToDropShadows=false
      bMiteredCorners=false
      bSeparateFillAlpha=false
      outlineMaterial=null
      outlineSize=0
      outlineColor=<rgba 0 0 0 1>
     }
    }
    shadowColorAndOpacity=<rgba 0 0 0 1>
    shadowOffset=<0 0>
   }
   slot {class=HorizontalBoxSlot, properties {horizontalAlignment=HAlign_Right, verticalAlignment=VAlign_Center}}
  }
  SizeBoxCheckmark {
   widgetClassPath=&SizeBox
   parent=WidgetSwitcher_Complete
   widget=SizeBoxCheckmark
   properties {bOverride_HeightOverride=true, bOverride_WidthOverride=true, heightOverride=27, widthOverride=140}
   slot {class=WidgetSwitcherSlot}
  }
  Overlay_0 {
   widgetClassPath=&Overlay
   parent=SizeBoxCheckmark
   widget=Overlay_0
   slot {class=SizeBoxSlot, properties {horizontalAlignment=HAlign_Right, verticalAlignment=VAlign_Center, padding=<ltrb 0 0 4 0>}}
  }
  CheckMarkIcon {
   widgetClassPath=&Image
   bIsVariable=true
   parent=Overlay_0
   widget=CheckMarkIcon
   properties {brush {
     bIsDynamicallyLoaded=false
     drawAs=Image
     imageType=NoImage
     mirroring=NoMirror
     resourceName=None
     resourceObject=/Game/UI/Foundation/Textures/Icons/Checks/T-Icon-Check-64
     tiling=NoTile
     imageSize=<27 27>
     margin=<ltrb 0 0 0 0>
     outlineSettings {
      bUseBrushTransparency=false
      roundingType=HalfHeightRadius
      width=0
      color {colorUseRule=UseColor_Specified, specifiedColor=<rgba 0 0 0 0>}
      cornerRadii=<0 0 0 0>
     }
     tintColor {colorUseRule=UseColor_Specified, specifiedColor=<rgba 0.0343 1 0.227 1>}
     uVRegion {bIsValid=false, max=<0 0>, min=<0 0>}
    }}
   slot {class=OverlaySlot, properties {horizontalAlignment=HAlign_Right, verticalAlignment=VAlign_Center}}
  }
  Image_Glow {
   widgetClassPath=&Image
   bIsVariable=true
   parent=Overlay_0
   widget=Image_Glow
   properties {
    renderOpacity=0
    brush {
     bIsDynamicallyLoaded=false
     drawAs=Image
     imageType=NoImage
     mirroring=NoMirror
     resourceName=None
     resourceObject=/Game/Athena/HUD/LTM/Wax/Materials/M_UI_TextGlow_Add
     tiling=NoTile
     imageSize=<60 60>
     margin=<ltrb 0 0 0 0>
     outlineSettings {
      bUseBrushTransparency=false
      roundingType=HalfHeightRadius
      width=0
      color {colorUseRule=UseColor_Specified, specifiedColor=<rgba 0 0 0 0>}
      cornerRadii=<0 0 0 0>
     }
     tintColor {colorUseRule=UseColor_Specified, specifiedColor=<rgba 0 1 0.1443 1>}
     uVRegion {bIsValid=false, max=<0 0>, min=<0 0>}
    }
   }
   slot {class=OverlaySlot, properties {horizontalAlignment=HAlign_Right, verticalAlignment=VAlign_Center, padding=<ltrb -25 -25 -15 -25>}}
  }
  Image_Spark {
   widgetClassPath=&Image
   bIsVariable=true
   parent=Overlay_0
   widget=Image_Spark
   properties {
    renderOpacity=0
    brush {
     bIsDynamicallyLoaded=false
     drawAs=Image
     imageType=NoImage
     mirroring=NoMirror
     resourceName=None
     resourceObject=/Game/UI/InGame/Quest/Materials/M_UI_SparkStar
     tiling=NoTile
     imageSize=<60 60>
     margin=<ltrb 0 0 0 0>
     outlineSettings {
      bUseBrushTransparency=false
      roundingType=HalfHeightRadius
      width=0
      color {colorUseRule=UseColor_Specified, specifiedColor=<rgba 0 0 0 0>}
      cornerRadii=<0 0 0 0>
     }
     tintColor {colorUseRule=UseColor_Specified, specifiedColor=<rgba 1 1 1 1>}
     uVRegion {bIsValid=false, max=<0 0>, min=<0 0>}
    }
   }
   slot {class=OverlaySlot, properties {horizontalAlignment=HAlign_Right, verticalAlignment=VAlign_Center, padding=<ltrb -25 -25 -15 -25>}}
  }
  Text_QuestDetails {
   widgetClassPath=&CommonTextBlock
   bInherited=true
   bIsVariable=true
   parent=HorizontalBox_540
   widget=Text_QuestDetails
   properties {
    bIsEnabled=false
    visibility=Hidden
    font {
     bForceMonospaced=false
     bMaterialIsStencil=false
     fontMaterial=null
     fontObject=&BurbankSmall
     letterSpacing=0
     monospacedWidth=1
     size=1
     skewAmount=0
     typefaceFontName=Bold
     outlineSettings {
      bApplyOutlineToDropShadows=false
      bMiteredCorners=false
      bSeparateFillAlpha=false
      outlineMaterial=null
      outlineSize=0
      outlineColor=<rgba 0 0 0 1>
     }
    }
    shadowColorAndOpacity=<rgba 0 0 0 1>
    shadowOffset=<1.6 1.6>
   }
   slot {class=HorizontalBoxSlot}
  }
  SizeBoxProgressBar {
   widgetClassPath=&SizeBox
   parent=VerticalBox_0
   widget=SizeBoxProgressBar
   properties {bOverride_HeightOverride=true, bOverride_WidthOverride=true, heightOverride=8, widthOverride=20}
   slot {class=VerticalBoxSlot, properties {padding=<ltrb 0 1 0 0>}}
  }
  Image_BackplateProgress_1 {
   widgetClassPath=&Image
   bIsVariable=true
   parent=SizeBoxProgressBar
   widget=Image_BackplateProgress_1
   properties {brush {
     bIsDynamicallyLoaded=false
     drawAs=Image
     imageType=NoImage
     mirroring=NoMirror
     resourceName=None
     resourceObject=/Game/Athena/UI/InGame/Creative/Art/M_UI_QuestTrackerRow_ProgressBar
     tiling=NoTile
     imageSize=<32 70>
     margin=<ltrb 0.1 0.1 0.9 0.9>
     outlineSettings {
      bUseBrushTransparency=false
      roundingType=HalfHeightRadius
      width=0
      color {colorUseRule=UseColor_Specified, specifiedColor=<rgba 0 0 0 0>}
      cornerRadii=<0 0 0 0>
     }
     tintColor {colorUseRule=UseColor_Specified, specifiedColor=<rgba 1 1 1 1>}
     uVRegion {bIsValid=false, max=<0 0>, min=<0 0>}
    }}
   slot {class=SizeBoxSlot}
  }
  TutorialHighlight {widgetClassPath=&TutorialHighlight}
 ]
}
