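# Playground minigame HUD widget, derived from MinigameHUD_StatTracker, that tracks three creative quest rows beneath a top progress bar.
# Row progress events copy the last updated quest's score, target and place to that bar, whose material fill eases over the AnimateBarFill animation.
# When IsRanked is set, backward and forward adjacent-swap passes sort rows by percentage, pushing hidden rows down, and changed rows animate into new slot boxes.
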WidgetBlueprint /Game/Athena/HUD/LTM/Playground/MinigameHUD_QuestTracker
animations [AnimateBarFill {
  binding_count=0
  length_seconds=0.5
  object=&self:AnimateBarFill
  sequence {bindings [], tracks [{type=MovieSceneEventTrack, sections [{type=MovieSceneEventRepeaterSection, end_frame=10, start_frame=0}]}]}
 }]
graphs [
 EventGraph {nodes [
   n1 = UpdateToRow1() {is=title K2Node_CustomEvent}
   n2 = Get Row1() -> (Row1:Minigame HUD Quest Tracker Row Object Reference)
   > UpdateToLastUpdatedQuest!(execute, self, Row:Minigame HUD Quest Tracker Row Object Reference=@n2) -> (then)
   n3 = UpdateToRow2() {is=title K2Node_CustomEvent}
   n4 = Get Row2() -> (Row2:Minigame HUD Quest Tracker Row Object Reference)
   > UpdateToLastUpdatedQuest(Row=@n4)
   n5 = UpdateToRow3() {is=title K2Node_CustomEvent}
   n6 = Get Row3() -> (Row3:Minigame HUD Quest Tracker Row Object Reference)
   > UpdateToLastUpdatedQuest(Row=@n6)
   Event Construct()
   > SetVisibility(self=Get TopProgress() -> (TopProgress:Overlay Object Reference), InVisibility=Collapsed)
   > Clear(TargetArray:Array of Horizontal Box Object References=Get RowParents() -> (RowParents:Array of Horizontal Box Object References))
   > Clear(TargetArray:Array of Minigame HUD Quest Tracker Row Object References=Get Rows() -> (Rows:Array of Minigame HUD Quest Tracker Row Object References))
   > n7 = Set Rows(Rows:Array of Minigame HUD Quest Tracker Row Object References=Make Array!(
     "[0]":Minigame HUD Quest Tracker Row Object Reference=Get Row1()
     "[1]":Minigame HUD Quest Tracker Row Object Reference=Get Row2()
     "[2]":Minigame HUD Quest Tracker Row Object Reference=Get Row3()
    ) -> (Array:Array of Minigame HUD Quest Tracker Row Object References)) -> (Output_Get:Array of Minigame HUD Quest Tracker Row Object References)
   > n8 = For Each Loop(Array:Array of Minigame HUD Quest Tracker Row Object References=@n7) -> (Array Element:Minigame HUD Quest Tracker Row Object Reference)
   ResetRows!(execute=@n8.Completed, self, RowsToReset:Array of Minigame HUD Quest Tracker Row Object References=Get Rows()) -> (then)
   > n9 = Branch(Condition=Get IsRanked() -> (IsRanked:bool))
   n10 = Get RankBox() -> (RankBox:Size Box Object Reference)
   > n11 = SetVisibility(self=@n10, InVisibility=SelfHitTestInvisible)
   Bind Event to On Quest Progress Changed(execute=[@n14.then, @n11.then], self:Minigame HUD Quest Tracker Row Object Reference=@n2, Delegate=@n1) {is=title K2Node_AddDelegate}
   > Bind Event to On Quest Progress Changed(self=@n4, Delegate=@n3)
   > Bind Event to On Quest Progress Changed(self=@n6, Delegate=@n5)
   > SetIsRanked!(execute, self:Minigame HUD Quest Tracker Row Object Reference=[Get Row2(), Get Row3(), Get Row1()], IsRanked:bool=Get IsRanked()) -> (then)
   n12 = GetOwningPlayer()
   > Is Valid(InputObject=@n12)
   n13 = GetMinigame!(ContextActor:Actor Object Reference=@n12) -> (ReturnValue:Fort Minigame Object Reference)
   > Is Valid(InputObject=@n13)
   > Parent: OnDisplayMinigameInProgressUI!(execute, bDisplay:bool=IsInProgress!(self:Fort Minigame Object Reference=@n13) -> (ReturnValue:bool)) -> (then) {is=title K2Node_CallParentFunction}
   n14 = SetVisibility(execute=@n9.else, self=@n10, InVisibility=Collapsed)
   n15 = Cast To HorizontalBox!(execute=@n8.LoopBody, Object=GetParent!(self:Widget Object Reference=@n8) -> (ReturnValue:Panel Widget Object Reference)) -> (then, CastFailed, AsHorizontal Box:Horizontal Box Object Reference) {is=title K2Node_DynamicCast}
   > Insert!(
    execute
    TargetArray:Array of Horizontal Box Object References=Get RowParents()
    NewItem:"Horizontal Box Object Reference (by ref)"=@n15
    Index:int=@n8.Array Index
   ) -> (then) {is=title K2Node_CallArrayFunction}
   n16 = Event OnDisplayMinigameInProgressUI!() -> (OutputDelegate, then, bDisplay:bool) {is=title K2Node_Event}
   > Parent: OnDisplayMinigameInProgressUI(bDisplay=@n16.bDisplay)
   > Branch(Condition=@n16.bDisplay)
   > ResetRows(RowsToReset=Get Rows())
   n17 = Event UpdateNewScore!() -> (OutputDelegate, then, bSetOldScoreToNewScore:bool) {is=title K2Node_Event}
   > Parent: UpdateNewScore!(execute, bSetOldScoreToNewScore:bool=@n17.bSetOldScoreToNewScore) -> (then) {is=title K2Node_CallParentFunction}
   Event AnimateNewScore!() -> (OutputDelegate, then, NewScoreValue:double, bSkipAnimation:bool) {is=title K2Node_Event}
  ]}
 GetRowNewPercentage {nodes [
   n1 = Inputs!() -> (Rows:Array of Minigame HUD Quest Tracker Row Object References, RowIndex:int)
   n2 = "Get (a copy)"(Array:Array of Minigame HUD Quest Tracker Row Object References=@n1, Dimension 1=@n1.RowIndex) -> (Output:Minigame HUD Quest Tracker Row Object Reference)
   n3 = IsVisible(self=@n2)
   Outputs!(Value:double=float + float(
     A=SelectFloat(A=0, B=-10, bPickA=@n3)
     B=Get NewPercentage!(self:Minigame HUD Quest Tracker Row Object Reference=@n2) -> (NewPercentage:double)
    ) {is=title K2Node_CommutativeAssociativeBinaryOperator})
  ]}
 ResetRows {nodes [
   n1 = ResetRows!() -> (then, RowsToReset:Array of Minigame HUD Quest Tracker Row Object References) {is=title K2Node_FunctionEntry}
   > n2 = For Each Loop(Array=@n1)
   > SetToQuest!(execute, self:Fort Creative Quest HUDWidget Object Reference=@n2, questIndex:int=@n2.Array Index, bForceUpdate:bool=true) -> (then)
   > Branch(Condition=Get IsRanked())
   > SetRank!(execute, self:Minigame HUD Quest Tracker Row Object Reference=@n2, Rank:int=@n2.Array Index) -> (then)
   > SetPlacement!(
    execute
    self:Minigame HUD Quest Tracker Row Object Reference=@n2
    Placement:int=@n2.Array Index
    NewParent:Horizontal Box Object Reference="Get (a copy)"(Array:Array of Horizontal Box Object References=Get RowParents(), Dimension 1=@n2.Array Index) -> (Output:Horizontal Box Object Reference)
    PlayAnimation:bool=false
   ) -> (then)
  ]}
 SetFillPercentage {nodes [
   n1 = SetFillPercentage!() -> (then, Value:double=~) {is=title K2Node_FunctionEntry}
   > Set OriginalPercentage(OriginalPercentage:double=@n1) -> (Output_Get:double)
   n2 = Get Image_BackplateProgress() -> (Image_BackplateProgress:Image Object Reference)
   > Is Valid(InputObject=@n2)
   > n3 = GetDynamicMaterial(self=@n2)
   > Is Valid(InputObject=@n3)
   > SetScalarParameterValue(self=@n3, ParameterName=PercentFill, Value=Get OriginalPercentage() -> (OriginalPercentage:double))
  ]}
 SetPlacementForRows {nodes [
   SetPlacementForRows() {is=title K2Node_FunctionEntry}
   > Branch(Condition=Get IsRanked())
   n1 = Get Rows()
   > n2 = Reverse for Each Loop!(Exec:Exec, Array:Array of Minigame HUD Quest Tracker Row Object References=@n1) -> (LoopBody:Exec, ArrayIndex:int, ArrayElement:Minigame HUD Quest Tracker Row Object Reference, Completed:Exec) {is=title K2Node_MacroInstance}
   n3 = For Each Loop(Exec=@n2.Completed, Array=@n1)
   n4 = For Each Loop(Exec=@n3.Completed, Array=@n1)
   > SetRowPositionAndPlacement!(
    execute
    self
    RowIndex:int=@n4.Array Index
    Rows:Array of Minigame HUD Quest Tracker Row Object References=Get Rows()
    RowChangedFlags:Array of Booleans=Get HasRowsChanged() -> (HasRowsChanged:Array of Booleans)
    Parents:Array of Horizontal Box Object References=Get RowParents()
   ) -> (then)
   SortTwoElementsOfRowArray!(
    execute=@n3.LoopBody
    self
    Rows:Array of Minigame HUD Quest Tracker Row Object References=Get Rows()
    FirstIndex:int=int + int(A=@n3.Array Index, B=1) -> (ReturnValue=0) {is=title K2Node_CommutativeAssociativeBinaryOperator}
    SecondIndex:int=@n3.Array Index
    RowChangedFlags:Array of Booleans=Get HasRowsChanged()
   ) -> (then)
   SortTwoElementsOfRowArray(
    execute=@n2.LoopBody
    Rows=Get Rows()
    FirstIndex=@n2
    SecondIndex=int - int(A=@n2, B=1) {is=title K2Node_CallFunction}
    RowChangedFlags=Get HasRowsChanged()
   )
  ]}
 SetRowPositionAndPlacement {nodes [
   n1 = SetRowPositionAndPlacement!() -> (
    then
    RowIndex:int
    Rows:Array of Minigame HUD Quest Tracker Row Object References
    RowChangedFlags:Array of Booleans
    Parents:Array of Horizontal Box Object References
   ) {is=title K2Node_FunctionEntry}
   > Branch(Condition="Get (a ref)"(Array:Array of Booleans=@n1.RowChangedFlags, Dimension 1=@n1) -> (Output:"Boolean (by ref)"))
   > SetPlacement(
    self="Get (a copy)"(Array:Array of Minigame HUD Quest Tracker Row Object References=@n1.Rows, Dimension 1=@n1) -> (Output:Minigame HUD Quest Tracker Row Object Reference)
    Placement=@n1
    NewParent="Get (a copy)"(Array:Array of Horizontal Box Object References=@n1.Parents, Dimension 1=@n1) -> (Output:Horizontal Box Object Reference)
    PlayAnimation=true
   )
  ]}
 SortTwoElementsOfRowArray {nodes [
   n1 = SortTwoElementsOfRowArray!() -> (then, Rows:Array of Minigame HUD Quest Tracker Row Object References, FirstIndex:int, SecondIndex:int, RowChangedFlags:Array of Booleans) {is=title K2Node_FunctionEntry}
   n2 = Length(TargetArray:Array of Minigame HUD Quest Tracker Row Object References=@n1)
   > Branch(Condition="In Range (Integer)"!(Value:int=@n1.FirstIndex, Min:int=0, Max:int=@n2, InclusiveMin:bool=true, InclusiveMax:bool=false) -> (ReturnValue:bool))
   > Branch(Condition="In Range (Integer)"(Value=@n1.SecondIndex, Min=0, Max=@n2, InclusiveMin=true, InclusiveMax=false))
   n3 = Get Rows()
   n4 = Get SecondIndex() -> (SecondIndex:int=0)
   n5 = Get Rows()
   n6 = Get FirstIndex() -> (FirstIndex:int=0)
   > Branch(Condition="float > float"(
     A=Get Row New Percentage!(Rows:Array of Minigame HUD Quest Tracker Row Object References=@n5, RowIndex:int=@n6) -> (Value:double=~) {is=title K2Node_MacroInstance}
     B=Get Row New Percentage(Rows=@n3, RowIndex=@n4) -> (Value=~)
    ) {is=title K2Node_CallFunction})
   > Swap Array Elements!(
    execute
    TargetArray:Array of Minigame HUD Quest Tracker Row Object References=Get Rows()
    FirstIndex:int=Get FirstIndex() -> (FirstIndex=0)
    SecondIndex:int=Get SecondIndex() -> (SecondIndex=0)
   ) -> (then) {is=title K2Node_CallArrayFunction}
   > Set Array Elem!(
    execute
    TargetArray:Array of Booleans=@n1.RowChangedFlags
    Index:int=Get FirstIndex() -> (FirstIndex=0)
    Item:"Boolean (by ref)"=true
    bSizeToFit:bool=false
   ) -> (then)
   > Set Array Elem(TargetArray=@n1.RowChangedFlags, Index=Get SecondIndex() -> (SecondIndex=0), Item=true, bSizeToFit=false)
  ]}
 TickBarFill {nodes [
   TickBarFill() {is=title K2Node_FunctionEntry}
   n1 = Get NewPercentage!() -> (NewPercentage)
   > n2 = Branch(Condition="Nearly Equal (Float)"(A=@n1, B=Get OriginalPercentage(), ErrorTolerance=0.01))
   > SetFillPercentage!(execute, self, Value:double=Get NewPercentage()) -> (then) {is=title K2Node_CallFunction}
   n3 = float - float(A=@n1, B=Get OriginalPercentage()) {is=title K2Node_CallFunction}
   n4 = Get AnimateBarFill() -> (AnimateBarFill:Widget Animation Object Reference)
   n5 = "Clamp (Float)"(
    Value=float / float(A=GetAnimationCurrentTime!(self:User Widget Object Reference, InAnimation:Widget Animation Object Reference=@n4) -> (ReturnValue:float), B=1) {is=title K2Node_CallFunction}
    Min=0
    Max=1
   )
   SetFillPercentage(execute=@n2.else, Value=float + float(A=Get OriginalPercentage(), B=float * float(A=@n3, B=@n5)))
  ]}
 UpdateToLastUpdatedQuest {nodes [
   n1 = UpdateToLastUpdatedQuest!() -> (then, Row:Minigame HUD Quest Tracker Row Object Reference) {is=title K2Node_FunctionEntry}
   > n2 = Branch(Condition=IsVisible(self=Get Row() -> (Row:Minigame HUD Quest Tracker Row Object Reference)))
   > n3 = GetTargetScore!(execute, self:Minigame HUD Quest Tracker Row Object Reference=@n1) -> (then, Value:double)
   > Set ScoreToWinValue(ScoreToWinValue:double=@n3) -> (Output_Get:double)
   n4 = Get Row()
   > "SetText (Text)"(
    self=Get ScoreToWin() -> (ScoreToWin:Common Numeric Text Block Object Reference)
    InText="GetText (Text)"!(self:Text Object Reference=Get Text_TargetAmount!(self:Fort Creative Quest HUDWidget Object Reference=@n4) -> (Text_TargetAmount:Common Text Object Reference)) -> (ReturnValue:Text)
   )
   n5 = Get Row()
   > "SetText (Text)"(
    self=Get CurrentScore() -> (CurrentScore:Common Numeric Text Block Object Reference)
    InText="GetText (Text)"(self=Get Text_CurrentAmount!(self:Fort Creative Quest HUDWidget Object Reference=@n5) -> (Text_CurrentAmount:Common Text Object Reference))
   )
   > Set NewPercentage(NewPercentage:double=Get NewPercentage!(self:Minigame HUD Quest Tracker Row Object Reference=Get Row()) -> (NewPercentage)) -> (Output_Get:double)
   > PlayAnimation(InAnimation=Get AnimateBarFill())
   n6 = Get Row()
   > SetVisibility(
    self=Get ProgressBox() -> (ProgressBox:Horizontal Box Object Reference)
    InVisibility=GetVisibility!(self:Widget Object Reference=Get SizeBox_Progress!(self:Minigame HUD Quest Tracker Row Object Reference=@n6) -> (SizeBox_Progress:Size Box Object Reference)) -> (ReturnValue:ESlateVisibility Enum=Visible)
   )
   > SetPlacementForRows() {is=title K2Node_CallFunction}
   n7 = Find Item!(
    TargetArray:Array of Minigame HUD Quest Tracker Row Object References=Get Rows()
    ItemToFind:"Minigame HUD Quest Tracker Row Object Reference (by ref)"=Get Row()
   ) -> (ReturnValue)
   > SetActiveWidgetIndex(self=Get WidgetSwitcher_PlaceBackground() -> (WidgetSwitcher_PlaceBackground:Widget Switcher Object Reference), Index=@n7)
   > "SetText (Text)"(
    self=Get Text_PlaceValue() -> (Text_PlaceValue:Common Text Object Reference)
    InText="To Text (Integer)"(Value=int + int(A=@n7, B=1) -> (ReturnValue=0))
   )
   > UpdateTopProgressVisibilty()
   UpdateTopProgressVisibilty(execute=@n2.else)
  ]}
 UpdateTopProgressVisibilty {nodes [
   UpdateTopProgressVisibilty() {is=title K2Node_FunctionEntry}
   > n1 = Set TopVisibility(TopVisibility:ESlateVisibility Enum=Collapsed) -> (Output_Get:ESlateVisibility Enum=Visible)
   SetVisibility(execute=[@n2.Completed, @n1.then], self=Get TopProgress(), InVisibility=Get TopVisibility() -> (TopVisibility:ESlateVisibility Enum=Visible))
   n2 = For Each Loop with Break!(Exec:Exec, Array:Array of Minigame HUD Quest Tracker Row Object References=Get Rows(), Break:Exec=@n5.then) -> (LoopBody:Exec, Array Element:Minigame HUD Quest Tracker Row Object Reference, Array Index:int, Completed:Exec)
   n3 = IsVisible(self=@n2)
   n4 = GetVisibility(self=@n2) -> (ReturnValue=Visible)
   > Branch(Condition=@n3)
   > n5 = Set TopVisibility(TopVisibility=@n4) -> (Output_Get=Visible)
  ]}
]
variables [
 IsRanked {type=boolean}
 NewPercentage {type=number}
 OriginalPercentage {type=number}
 RowParents {type="array<HorizontalBox>", default []}
 Rows {type="array<MinigameHUD_QuestTracker_Row_C>", default []}
 ScoreToWinValue {type=number}
]
widgets {
 info {
  inheritedWidgetCount=1
  parentClass=&/MinigameHUD_StatTracker.MinigameHUD_StatTracker_C
  rootWidgetClass=&VerticalBox
  widgetCount=27
 }
 widgets [
  VerticalBox_1 {widgetClassPath=&VerticalBox, widget=VerticalBox_1}
  TopProgress {
   widgetClassPath=&Overlay
   bIsVariable=true
   parent=VerticalBox_1
   widget=TopProgress
   properties {visibility=Collapsed}
   slot {class=VerticalBoxSlot, properties {padding=<ltrb 0 8 0 0>}}
  }
  SizeBox_1 {
   widgetClassPath=&SizeBox
   parent=TopProgress
   widget=SizeBox_1
   properties {bOverride_WidthOverride=true, widthOverride=400}
   slot {class=OverlaySlot, properties {horizontalAlignment=HAlign_Center, padding=<ltrb 0 4 0 0>}}
  }
  Overlay_5 {widgetClassPath=&Overlay, parent=SizeBox_1, widget=Overlay_5, slot {class=SizeBoxSlot}}
  Image_BackplateProgress {
   widgetClassPath=&Image
   bIsVariable=true
   parent=Overlay_5
   widget=Image_BackplateProgress
   properties {
    brush {
     bIsDynamicallyLoaded=false
     drawAs=Box
     imageType=NoImage
     mirroring=NoMirror
     resourceName=None
     resourceObject=/Game/Athena/HUD/Materials/Score/M_ScoreLTM_ScoreProgress_Inst
     tiling=NoTile
     imageSize=<32 65>
     margin=<ltrb 0.09 0.09 0.09 0.09>
     outlineSettings {
      bUseBrushTransparency=false
      roundingType=HalfHeightRadius
      width=0
      color {colorUseRule=UseColor_Specified, specifiedColor=<rgba 0 0 0 0>}
      cornerRadii=<0 0 0 0>
     }
     tintColor {colorUseRule=UseColor_Specified, specifiedColor=<rgba 1 1 1 1>}
     uVRegion {bIsValid=false, max=<0 0>, min=<0 0>}
    }
    renderTransform {angle=0, scale=<1 1>, shear=<-5 0>, translation=<0 -2>}
   }
   slot {class=OverlaySlot, properties {horizontalAlignment=HAlign_Fill, verticalAlignment=VAlign_Center}}
  }
  RankBox {
   widgetClassPath=&SizeBox
   bIsVariable=true
   parent=Overlay_5
   widget=RankBox
   properties {
    bOverride_HeightOverride=true
    bOverride_WidthOverride=true
    heightOverride=64
    widthOverride=104
    renderTransform {angle=0, scale=<1 1>, shear=<0 0>, translation=<-20 0>}
   }
   slot {class=OverlaySlot, properties {verticalAlignment=VAlign_Center}}
  }
  Overlay_4 {
   widgetClassPath=&Overlay
   parent=RankBox
   widget=Overlay_4
   properties {renderTransform {angle=0, scale=<1 1>, shear=<0 0>, translation=<-4 0>}}
   slot {class=SizeBoxSlot}
  }
  WidgetSwitcher_PlaceBackground {
   widgetClassPath=&WidgetSwitcher
   bIsVariable=true
   parent=Overlay_4
   widget=WidgetSwitcher_PlaceBackground
   properties {activeWidgetIndex=2}
   slot {class=OverlaySlot, properties {horizontalAlignment=HAlign_Center, verticalAlignment=VAlign_Center}}
  }
  Image_1 {
   widgetClassPath=&Image
   parent=WidgetSwitcher_PlaceBackground
   widget=Image_1
   properties {
    brush {
     bIsDynamicallyLoaded=false
     drawAs=Image
     imageType=NoImage
     mirroring=NoMirror
     resourceName=None
     resourceObject=/Game/Athena/HUD/Materials/Score/M_ScoreLTM_PlacementPlate_1st
     tiling=NoTile
     imageSize=<104 64>
     margin=<ltrb 0 0 0 0>
     outlineSettings {
      bUseBrushTransparency=false
      roundingType=HalfHeightRadius
      width=0
      color {colorUseRule=UseColor_Specified, specifiedColor=<rgba 0 0 0 0>}
      cornerRadii=<0 0 0 0>
     }
     tintColor {colorUseRule=UseColor_Specified, specifiedColor=<rgba 1 1 1 1>}
     uVRegion {bIsValid=false, max=<0 0>, min=<0 0>}
    }
    renderTransform {angle=0, scale=<1 1>, shear=<0 0>, translation=<1 0>}
   }
   slot {class=WidgetSwitcherSlot}
  }
  Image_2 {
   widgetClassPath=&Image
   parent=WidgetSwitcher_PlaceBackground
   widget=Image_2
   properties {
    brush {
     bIsDynamicallyLoaded=false
     drawAs=Image
     imageType=NoImage
     mirroring=NoMirror
     resourceName=None
     resourceObject=/Game/Athena/HUD/Materials/Score/M_ScoreLTM_PlacementPlate_2nd
     tiling=NoTile
     imageSize=<104 62>
     margin=<ltrb 0 0 0 0>
     outlineSettings {
      bUseBrushTransparency=false
      roundingType=HalfHeightRadius
      width=0
      color {colorUseRule=UseColor_Specified, specifiedColor=<rgba 0 0 0 0>}
      cornerRadii=<0 0 0 0>
     }
     tintColor {colorUseRule=UseColor_Specified, specifiedColor=<rgba 1 1 1 1>}
     uVRegion {bIsValid=false, max=<0 0>, min=<0 0>}
    }
    renderTransform {angle=0, scale=<1 1>, shear=<0 0>, translation=<2 -2>}
   }
   slot {class=WidgetSwitcherSlot}
  }
  Image_3 {
   widgetClassPath=&Image
   parent=WidgetSwitcher_PlaceBackground
   widget=Image_3
   properties {
    brush {
     bIsDynamicallyLoaded=false
     drawAs=Image
     imageType=NoImage
     mirroring=NoMirror
     resourceName=None
     resourceObject=/Game/Athena/HUD/Materials/Score/M_ScoreLTM_PlacementPlate_3rd
     tiling=NoTile
     imageSize=<72 56>
     margin=<ltrb 0 0 0 0>
     outlineSettings {
      bUseBrushTransparency=false
      roundingType=HalfHeightRadius
      width=0
      color {colorUseRule=UseColor_Specified, specifiedColor=<rgba 0 0 0 0>}
      cornerRadii=<0 0 0 0>
     }
     tintColor {colorUseRule=UseColor_Specified, specifiedColor=<rgba 1 1 1 1>}
     uVRegion {bIsValid=false, max=<0 0>, min=<0 0>}
    }
    renderTransform {angle=0, scale=<1 1>, shear=<0 0>, translation=<0 -3>}
   }
   slot {class=WidgetSwitcherSlot}
  }
  SizeBox_0 {
   widgetClassPath=&SizeBox
   parent=Overlay_4
   widget=SizeBox_0
   properties {bOverride_HeightOverride=true, bOverride_WidthOverride=true, heightOverride=56, widthOverride=48}
   slot {class=OverlaySlot, properties {horizontalAlignment=HAlign_Center, verticalAlignment=VAlign_Center}}
  }
  Text_PlaceValue {
   widgetClassPath=&CommonTextBlock
   bIsVariable=true
   parent=SizeBox_0
   widget=Text_PlaceValue
   properties {
    text=3
    colorAndOpacity {colorUseRule=UseColor_Specified, specifiedColor=<rgba 0 0.0125 0.05 1>}
    font {
     bForceMonospaced=false
     bMaterialIsStencil=false
     fontMaterial=null
     fontObject=&Burbank
     letterSpacing=0
     monospacedWidth=1
     size=32
     skewAmount=0
     typefaceFontName=Burbank-Black
     outlineSettings {
      bApplyOutlineToDropShadows=false
      bMiteredCorners=false
      bSeparateFillAlpha=true
      outlineMaterial=null
      outlineSize=2
      outlineColor=<rgba 1 1 1 0.75>
     }
    }
    shadowOffset=<0 0>
   }
   slot {class=SizeBoxSlot, properties {horizontalAlignment=HAlign_Center, verticalAlignment=VAlign_Center, padding=<ltrb 0 -2 0 0>}}
  }
  ProgressBox {
   widgetClassPath=&HorizontalBox
   bIsVariable=true
   parent=Overlay_5
   widget=ProgressBox
   slot {class=OverlaySlot, properties {horizontalAlignment=HAlign_Center, verticalAlignment=VAlign_Center}}
  }
  CurrentScore {
   widgetClassPath=/Script/CommonUI.CommonNumericTextBlock
   bIsVariable=true
   parent=ProgressBox
   widget=CurrentScore
   properties {
    currentNumericValue=88888888
    justification=Right
    text=0
    font {
     bForceMonospaced=false
     bMaterialIsStencil=false
     fontMaterial=null
     fontObject=&NotoSans
     letterSpacing=0
     monospacedWidth=1
     size=18
     skewAmount=0
     typefaceFontName=NotoSans-Bold
     outlineSettings {
      bApplyOutlineToDropShadows=false
      bMiteredCorners=false
      bSeparateFillAlpha=false
      outlineMaterial=null
      outlineSize=1
      outlineColor=<rgba 0 0 0 0.5>
     }
    }
    shadowColorAndOpacity=<rgba 0.0156 0.0082 0.0063 0.8>
   }
   slot {class=HorizontalBoxSlot, properties {verticalAlignment=VAlign_Center}}
  }
  Slash {
   widgetClassPath=&TextBlock
   bIsVariable=true
   parent=ProgressBox
   widget=Slash
   properties {
    text=/
    font {
     bForceMonospaced=false
     bMaterialIsStencil=false
     fontMaterial=null
     fontObject=&NotoSans
     letterSpacing=0
     monospacedWidth=1
     size=18
     skewAmount=0
     typefaceFontName=NotoSans-Bold
     outlineSettings {
      bApplyOutlineToDropShadows=false
      bMiteredCorners=false
      bSeparateFillAlpha=false
      outlineMaterial=null
      outlineSize=1
      outlineColor=<rgba 0 0 0 0.5>
     }
    }
    shadowColorAndOpacity=<rgba 0.0156 0.0082 0.0063 0.5>
   }
   slot {class=HorizontalBoxSlot, properties {verticalAlignment=VAlign_Center}}
  }
  ScoreToWin {
   widgetClassPath=&CommonNumericTextBlock
   bIsVariable=true
   parent=ProgressBox
   widget=ScoreToWin
   properties {
    currentNumericValue=88888888
    justification=Right
    text="52,348"
    font {
     bForceMonospaced=false
     bMaterialIsStencil=false
     fontMaterial=null
     fontObject=&NotoSans
     letterSpacing=0
     monospacedWidth=1
     size=18
     skewAmount=0
     typefaceFontName=NotoSans-Bold
     outlineSettings {
      bApplyOutlineToDropShadows=false
      bMiteredCorners=false
      bSeparateFillAlpha=false
      outlineMaterial=null
      outlineSize=1
      outlineColor=<rgba 0 0 0 0.5>
     }
    }
    shadowColorAndOpacity=<rgba 0.0156 0.0082 0.0063 0.8>
   }
   slot {class=HorizontalBoxSlot, properties {verticalAlignment=VAlign_Center}}
  }
  MiniScoreboard {
   widgetClassPath=&Overlay
   bIsVariable=true
   parent=VerticalBox_1
   widget=MiniScoreboard
   slot {class=VerticalBoxSlot, properties {horizontalAlignment=HAlign_Center, verticalAlignment=VAlign_Bottom, padding=<ltrb 0 2 0 8>}}
  }
  SizeBox_4 {
   widgetClassPath=&SizeBox
   parent=MiniScoreboard
   widget=SizeBox_4
   properties {bOverride_HeightOverride=true, heightOverride=180, widthOverride=340}
   slot {class=OverlaySlot, properties {horizontalAlignment=HAlign_Center, verticalAlignment=VAlign_Center}}
  }
  CanvasPanel_0 {
   widgetClassPath=&CanvasPanel
   parent=SizeBox_4
   widget=CanvasPanel_0
   properties {renderTransform {angle=0, scale=<0.9 0.9>, shear=<0 0>, translation=<0 0>}}
   slot {class=SizeBoxSlot, properties {verticalAlignment=VAlign_Top}}
  }
  HorizontalBox_0 {
   widgetClassPath=&HorizontalBox
   parent=CanvasPanel_0
   widget=HorizontalBox_0
   properties {visibility=Visible}
   slot {
    class=CanvasPanelSlot
    properties {layoutData {alignment=<0 0>, anchors {maximum=<0.5 0.5>, minimum=<0.5 0.5>}, offsets=<ltrb -180 -30 400 60>}}
   }
  }
  Row1 {
   widgetClassPath=/Game/MinigameHUD_QuestTracker_Row
   bIsVariable=true
   parent=HorizontalBox_0
   widget=Row1
   properties {elementTag {tagName=Minigame.UIExtension.Trackers}, padding=<ltrb 0 0 0 0>}
   slot {class=HorizontalBoxSlot}
  }
  HorizontalBox_1 {
   widgetClassPath=&HorizontalBox
   parent=CanvasPanel_0
   widget=HorizontalBox_1
   properties {visibility=Visible}
   slot {
    class=CanvasPanelSlot
    properties {layoutData {alignment=<0 -1>, anchors {maximum=<0.5 0.5>, minimum=<0.5 0.5>}, offsets=<ltrb -180 -30 400 60>}}
   }
  }
  Row2 {
   widgetClassPath=&MinigameHUD_QuestTracker_Row
   bIsVariable=true
   parent=HorizontalBox_1
   widget=Row2
   properties {questIndex=1, elementTag {tagName=Minigame.UIExtension.Trackers}, padding=<ltrb 0 0 0 0>}
   slot {class=HorizontalBoxSlot}
  }
  HorizontalBox_2 {
   widgetClassPath=&HorizontalBox
   parent=CanvasPanel_0
   widget=HorizontalBox_2
   properties {visibility=Visible}
   slot {
    class=CanvasPanelSlot
    properties {layoutData {alignment=<0 -2>, anchors {maximum=<0.5 0.5>, minimum=<0.5 0.5>}, offsets=<ltrb -180 -30 400 60>}}
   }
  }
  Row3 {
   widgetClassPath=&MinigameHUD_QuestTracker_Row
   bIsVariable=true
   parent=HorizontalBox_2
   widget=Row3
   properties {questIndex=2, elementTag {tagName=Minigame.UIExtension.Trackers}, padding=<ltrb 0 0 0 0>}
   slot {class=HorizontalBoxSlot}
  }
  TutorialHighlight {widgetClassPath=&TutorialHighlight}
 ]
}
