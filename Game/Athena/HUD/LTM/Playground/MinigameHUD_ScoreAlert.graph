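# MinigameHUD_ScoreAlert is a minigame HUD widget showing centered score and respawn alert messages from the game state.
# It handles three center priorities: normal messages drop if busy, while important and critical ones cancel lower-priority alerts or queue behind others.
# Each alert plays an intro, rich text and sound, then a timer triggers the outro, whose completion shows the next queued message.
# Turning off the in-progress UI cancels or clears alerts.
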WidgetBlueprint /Game/Athena/HUD/LTM/Playground/MinigameHUD_ScoreAlert
animations [
 Intro {
  binding_count=1
  display_label=""
  length_seconds=0.25
  object=&self:Intro
  sequence {bindings [Feedback {
     id=#1
     display_name=OverlayFeedback
     possessed_class=HorizontalBox
     tracks [
      {
       type=MovieScene2DTransformTrack
       sections [{
         type=MovieScene2DTransformSection
         end_frame=5
         start_frame=0
         channels [
          Translation.X {type=MovieSceneScriptingFloatChannel}
          Translation.Y {type=MovieSceneScriptingFloatChannel}
          Angle {type=MovieSceneScriptingFloatChannel}
          Scale.X {type=MovieSceneScriptingFloatChannel, keys [[0,1],[3,1.282],[3.5983,1],[4.2237,1]]}
          Scale.Y {type=MovieSceneScriptingFloatChannel, keys [[0,0],[3,1.282],[3.5983,1],[4.2237,1]]}
          Shear.X {type=MovieSceneScriptingFloatChannel}
          Shear.Y {type=MovieSceneScriptingFloatChannel}
         ]
        }]
      }
      {
       type=MovieSceneFloatTrack
       display_name=Render Opacity
       sections [{type=MovieSceneFloatSection, end_frame=4, start_frame=0, channels [None {type=MovieSceneScriptingFloatChannel, keys [[0,0],[2.7623,1]]}]}]
      }
     ]
    }]}
 }
 Outro {
  binding_count=1
  display_label=""
  length_seconds=0.3397
  object=&self:Outro
  sequence {
   bindings [Feedback {
     id=#2
     display_name=OverlayFeedback
     possessed_class=HorizontalBox
     tracks [
      {
       type=MovieSceneFloatTrack
       display_name=Render Opacity
       sections [{type=MovieSceneFloatSection, end_frame=40, start_frame=0, channels [None {type=MovieSceneScriptingFloatChannel, keys [[0,1],[40.756,0]]}]}]
      }
      {
       type=MovieScene2DTransformTrack
       sections [{
         type=MovieScene2DTransformSection
         end_frame=40
         start_frame=0
         channels [
          Translation.X {type=MovieSceneScriptingFloatChannel}
          Translation.Y {type=MovieSceneScriptingFloatChannel}
          Angle {type=MovieSceneScriptingFloatChannel}
          Scale.X {type=MovieSceneScriptingFloatChannel, const [1,2]}
          Scale.Y {type=MovieSceneScriptingFloatChannel, keys [[0,1],[9,1.2783],[11,1],[40.756,0]]}
          Shear.X {type=MovieSceneScriptingFloatChannel}
          Shear.Y {type=MovieSceneScriptingFloatChannel}
         ]
        }]
      }
     ]
    }]
   display_rate {numerator=120}
  }
 }
]
graphs [
 EventGraph {nodes [
   OnOutroAnimFinished() {is=title K2Node_CustomEvent}
   > SetVisibility(self=Get Content() -> (Content:Vertical Box Object Reference), InVisibility=Collapsed)
   > Set bMsgDisplaying(bMsgDisplaying:bool=false) -> (Output_Get:bool)
   > n1 = Branch(Condition=Get bAcceptingGameMessages() -> (bAcceptingGameMessages:bool))
   n2 = Get QueuedMessages() -> (QueuedMessages:Array of Athena Game Message Data Structures)
   > n3 = Branch(Condition="integer > integer"(A=Length(TargetArray:Array of Athena Game Message Data Structures=@n2), B=0) {is=title K2Node_CallFunction})
   n4 = Get QueuedMessages()
   > Set QueuedMsg(QueuedMsg:Athena Game Message Data Structure="Get (a ref)"(Array:Array of Athena Game Message Data Structures=@n4, Dimension 1=0) -> (Output:"Athena Game Message Data Structure (by ref)")) -> (Output_Get:Athena Game Message Data Structure)
   > Remove Index!(execute, TargetArray:Array of Athena Game Message Data Structures=@n4, IndexToRemove:int=0) -> (then)
   > OnMessageRequested!(execute, self, MsgData:Athena Game Message Data Structure=Get QueuedMsg() -> (QueuedMsg:Athena Game Message Data Structure)) -> (then)
   HideGameModeMessage(
    execute=[@n5.then, @n3.else]
    self:Athena HUDContext Object Reference=Get AthenaHUDContext!() -> (ReturnValue:Athena HUDContext Object Reference) {is=title K2Node_GetBlueprintContextEx}
   )
   n5 = Clear(execute=@n1.else, TargetArray:Array of Athena Game Message Data Structures=Get QueuedMessages())
   n6 = OnMessageRequested!() -> (OutputDelegate, then, MsgData:Athena Game Message Data Structure) {is=title K2Node_CustomEvent}
   > n7 = Switch on EAthenaGameMsgType!(
    execute
    Selection:EAthenaGameMsgType Enum=Break Athena Game Message Data!(AthenaGameMessageData:"Athena Game Message Data Structure (by ref)"=@n6.MsgData) -> (
     MsgType:EAthenaGameMsgType Enum
     MsgText:Text
     MsgSound:Sound Cue Object Reference
     MsgMediaSource:Media Source Object Reference
     MsgDelay:float=~
     bIsTeamBased:bool=~
     TeamIndex:int
     DisplayTime:float=~
     MessageFlags:EAthenaGameMessageDataFlags Enum
     TargetPlayerController:Player Controller Object Reference
    ) {is=title K2Node_BreakStruct}
   ) -> (
    None:Exec
    DefaultIntro:Exec
    DefaultMessage:Exec
    DefaultCriticalMessage:Exec
    CommIntro:Exec
    CommMessage:Exec
    CommCriticalMessage:Exec
    CornerIntro:Exec
    CornerMessage:Exec
    CornerCriticalMessage:Exec
    RespawnTurningOffWarning:Exec
    RespawnOffWarning:Exec
    CenterMessage:Exec
    CenterImportantMessage:Exec
    CenterCriticalMessage:Exec
   ) {is=title K2Node_SwitchEnum}
   n8 = Branch(execute=@n7.CenterCriticalMessage, Condition=Get bMsgDisplaying() -> (bMsgDisplaying:bool))
   > n9 = Switch on EAthenaGameMsgType(Selection=Break Athena Game Message Data(AthenaGameMessageData=Get QueuedMsg()) -> (MsgDelay=~, bIsTeamBased=~, DisplayTime=~))
   n10 = CancelActiveMsg(execute=[@n9.CenterImportantMessage, @n9.CenterMessage])
   Add(
    execute=[@n15.then, @n10.then, @n14.CenterCriticalMessage, @n14.CenterImportantMessage, @n9.CenterCriticalMessage]
    TargetArray:Array of Athena Game Message Data Structures=Get QueuedMessages()
    NewItem:"Athena Game Message Data Structure (by ref)"=@n6.MsgData
   ) {is=title K2Node_CallArrayFunction}
   Set bMsgDisplaying(execute=[@n16.else, @n13.else, @n8.else], bMsgDisplaying=true)
   > n11 = Set QueuedMsg(QueuedMsg=@n6.MsgData)
   > SetVisibility(self=Get Content(), InVisibility=HitTestInvisible)
   > ShowGameModeMessage(self:Athena HUDContext Object Reference=Get AthenaHUDContext())
   > PlayAnimation(InAnimation=Get Intro())
   n12 = Break Athena Game Message Data(AthenaGameMessageData=@n11) -> (MsgDelay=~, bIsTeamBased=~, DisplayTime=~)
   > SetText(self=Get Message() -> (Message:Rich Text Block Object Reference), InText=@n12.MsgText)
   > PlaySoundForMsg!(execute, self, Sound:Sound Base Object Reference=@n12.MsgSound) -> (then)
   > Set Timer by Function Name!(
    execute
    Object:Object Reference
    FunctionName:String=OnDelayFinished
    Time:float=@n12.DisplayTime
    bLooping:bool=false
    bMaxOncePerFrame:bool=false
    InitialStartDelay:float=0
    InitialStartDelayVariance:float=0
   ) -> (then, ReturnValue:Timer Handle Structure) {is=title K2Node_CallFunction}
   n13 = Branch(execute=@n7.CenterImportantMessage, Condition=Get bMsgDisplaying())
   > n14 = Switch on EAthenaGameMsgType(Selection=Break Athena Game Message Data(AthenaGameMessageData=Get QueuedMsg()) -> (MsgDelay=~, bIsTeamBased=~, DisplayTime=~))
   n15 = CancelActiveMsg(execute=@n14.CenterMessage)
   n16 = Branch(execute=@n7.CenterMessage, Condition=Get bMsgDisplaying())
   CancelActiveMsg() {is=title K2Node_CustomEvent}
   n17 = Get Intro()
   > n18 = Branch(Condition=IsAnimationPlaying(InAnimation=@n17))
   > n19 = StopAnimation(InAnimation=@n17)
   Clear Timer by Function Name!(execute=[@n19.then, @n18.else], Object:Object Reference, FunctionName:String=OnDelayFinished) -> (then)
   > Clear(TargetArray=Get QueuedMessages())
   n20 = Get outro() -> (outro:Widget Animation Object Reference)
   > Branch(Condition=NOT Boolean(A=IsAnimationPlaying(InAnimation=@n20)))
   > OnDelayFinished()
   OnDelayFinished() {is=title K2Node_CustomEvent}
   > PlayAnimation(InAnimation=Get outro())
   n21 = Event PreConstruct()
   > Branch(Condition=@n21.IsDesignTime)
   > SetText(
    self=Get Message()
    InText=Select!(
     Option 0:Text="LOCTEXT(\"<intro>Score</> <introscore>1,000</><intro> points to win!</>\")"
     Option 1:Text="LOCTEXT(\"<enemy>An enemy is </><score>100</><enemy> points from victory!</>\")"
     Option 2:Text="LOCTEXT(\"<ally>Your team is </><score>100</><ally> points from victory!</>\")"
     Option 3:Text="LOCTEXT(\"<RespawnAlertB>NO RESPAWNS</><RespawnAlert>AFTER THE NEXT STORM CYCLE!</>\")"
     Option 4:Text="LOCTEXT(\"<RespawnAlertB>RESPAWNING IS NOW</> <RespawnAlertBLrg>OFF!</>\")"
     Index:int=Get TestMessageType() -> (TestMessageType:int=0)
    ) -> (ReturnValue:Text)
   )
   n22 = Event OnDisplayMinigameInProgressUI!() -> (OutputDelegate, then, bDisplay:bool) {is=title K2Node_Event}
   > n23 = Set bAcceptingGameMessages(bAcceptingGameMessages:bool=@n22.bDisplay) -> (Output_Get:bool)
   > n24 = Branch(Condition=@n23)
   n25 = Branch(execute=@n24.else, Condition=Get bMsgDisplaying())
   > CancelActiveMsg() {is=title K2Node_CallFunction}
   Clear(execute=@n25.else, TargetArray=Get QueuedMessages())
   Event Construct()
   > n26 = Cast To FortGameStateAthena!(execute, Object=GetGameState!() -> (ReturnValue:Game State Base Object Reference)) -> (then, CastFailed, AsFort Game State Athena:Fort Game State Athena Object Reference) {is=title K2Node_DynamicCast}
   > n27 = Assign Game Mode Message Request!(execute:Exec, self:Fort Game State Athena Object Reference=@n26, Delegate:"Delegate (by ref)"=Create Event()) -> (then:Exec) {is=title K2Node_AssignDelegate}
   SetVisibility(execute=[@n27.then, @n26.CastFailed], self=Get Content(), InVisibility=Collapsed)
   > BindToAnimationFinished(Animation=Get Outro() -> (Outro:Widget Animation Object Reference), Delegate=Create Event())
  ]}
 PlaySoundForMsg {nodes [
   n1 = PlaySoundForMsg!() -> (then, Sound:Sound Base Object Reference) {is=title K2Node_FunctionEntry}
   > Is Valid(InputObject=@n1)
   > PlaySound2D(Sound=@n1)
  ]}
]
variables [
 bAcceptingGameMessages {type=boolean}
 bMsgDisplaying {type=boolean}
 QueuedMessages {type="array<AthenaGameMessageData>", default []}
 QueuedMsg {
  type=AthenaGameMessageData
  default {bIsTeamBased=false, displayTime=5, messageFlags=None, msgDelay=0, msgMediaSource=null, msgSound=null, msgText="", msgType=None, teamIndex=0}
 }
 TestMessageType {type=integer}
]
widgets {
 info {inheritedWidgetCount=1, parentClass=/Script/FortniteUI.MinigameWidgetBase, rootWidgetClass=&VerticalBox, widgetCount=4}
 widgets [
  Content {widgetClassPath=&VerticalBox, bIsVariable=true, widget=Content}
  OverlayFeedback {
   widgetClassPath=&Overlay
   parent=Content
   widget=OverlayFeedback
   slot {class=VerticalBoxSlot, properties {horizontalAlignment=HAlign_Center}}
  }
  Message {
   widgetClassPath=/Script/UMG.RichTextBlock
   bIsVariable=true
   parent=OverlayFeedback
   widget=Message
   properties {
    justification=Center
    text="<RespawnAlert>RESPAWNING</> IS NOW <enemy>OFF!</>"
    textStyleSet=/Game/Athena/HUD/TextStyles/AthenaScoreAlertStyles
   }
   slot {class=OverlaySlot, properties {horizontalAlignment=HAlign_Center}}
  }
  TutorialHighlight {widgetClassPath=&TutorialHighlight}
 ]
}
